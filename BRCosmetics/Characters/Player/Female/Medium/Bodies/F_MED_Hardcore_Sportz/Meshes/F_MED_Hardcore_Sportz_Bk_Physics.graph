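# Physics asset for the back skirt of the female medium Hardcore Sportz body cosmetic.
# Three chains (left, center, right) of three box bodies each hang from tiny capsule master bodies locked to the pelvis.
# Chain links allow limited swing and narrow twist between segments.
# Pelvis and thigh capsules, with thighs locked to the pelvis, provide body shapes alongside the skirt.
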PhysicsAsset /BRCosmetics/Characters/Player/Female/Medium/Bodies/F_MED_Hardcore_Sportz/Meshes/F_MED_Hardcore_Sportz_Bk_Physics
bodies {
 dyn_skirt_bk_01 [{shapeType=Box, extentX=9.6, extentY=13.2, extentZ=3.8, shapeName=dyn_skirt_bk_01_box, center=<5 0 0>}]
 dyn_skirt_bk_01_MSTR [{shapeType=Capsule, length=0.01, radius=0.01}]
 dyn_skirt_bk_01_l [{shapeType=Box, extentX=9.6, extentY=13.2, extentZ=3.8, shapeName=dyn_skirt_bk_01_l_box, center=<5 0 0>}]
 dyn_skirt_bk_01_l_MSTR [{shapeType=Capsule, length=0.01, radius=0.01}]
 dyn_skirt_bk_01_r [{shapeType=Box, extentX=9.6, extentY=13.2, extentZ=3.8, shapeName=dyn_skirt_bk_01_r_box, center=<5 0 0>}]
 dyn_skirt_bk_01_r_MSTR [{shapeType=Capsule, length=0.01, radius=0.01}]
 dyn_skirt_bk_02 [{shapeType=Box, extentX=9.6, extentY=13.2, extentZ=3.8, shapeName=dyn_skirt_bk_02_box, center=<5 0 0>}]
 dyn_skirt_bk_02_l [{shapeType=Box, extentX=9.6, extentY=13.2, extentZ=3.8, shapeName=dyn_skirt_bk_02_l_box, center=<5 0 0>}]
 dyn_skirt_bk_02_r [{shapeType=Box, extentX=9.6, extentY=13.2, extentZ=3.8, shapeName=dyn_skirt_bk_02_r_box, center=<5 0 0>}]
 dyn_skirt_bk_03 [{shapeType=Box, extentX=9.6, extentY=13.2, extentZ=3.8, shapeName=dyn_skirt_bk_03_box, center=<5 0 0>}]
 dyn_skirt_bk_03_l [{shapeType=Box, extentX=9.6, extentY=13.2, extentZ=3.8, shapeName=dyn_skirt_bk_03_l_box, center=<5 0 0>}]
 dyn_skirt_bk_03_r [{shapeType=Box, extentX=9.6, extentY=13.2, extentZ=3.8, shapeName=dyn_skirt_bk_03_r_box, center=<5 0 0>}]
 pelvis [{shapeType=Capsule, length=7.5, radius=8, center=<1 0 0>}]
 thigh_l [{shapeType=Capsule, length=28.5, radius=8.4, center=<-20 0 0>, rotation=<rot 90 0 0>}]
 thigh_r [{shapeType=Capsule, length=28.5, radius=8.4, center=<20 0 0>, rotation=<rot 90 0 0>}]
}
constraints [
 {
  bone1Name=dyn_skirt_bk_01_l_MSTR
  bone2Name=pelvis
  swing1LimitDegrees=45
  swing1Motion=Locked
  swing2LimitDegrees=45
  swing2Motion=Locked
  twistLimitDegrees=45
  twistMotion=Locked
 }
 {
  bone1Name=dyn_skirt_bk_01_r_MSTR
  bone2Name=pelvis
  swing1LimitDegrees=45
  swing1Motion=Locked
  swing2LimitDegrees=45
  swing2Motion=Locked
  twistLimitDegrees=45
  twistMotion=Locked
 }
 {
  bone1Name=thigh_r
  bone2Name=pelvis
  swing1LimitDegrees=45
  swing1Motion=Locked
  swing2LimitDegrees=45
  swing2Motion=Locked
  twistLimitDegrees=45
  twistMotion=Locked
 }
 {
  bone1Name=thigh_l
  bone2Name=pelvis
  swing1LimitDegrees=45
  swing1Motion=Locked
  swing2LimitDegrees=45
  swing2Motion=Locked
  twistLimitDegrees=45
  twistMotion=Locked
 }
 {
  bone1Name=dyn_skirt_bk_01_MSTR
  bone2Name=pelvis
  swing1LimitDegrees=45
  swing1Motion=Locked
  swing2LimitDegrees=45
  swing2Motion=Locked
  twistLimitDegrees=45
  twistMotion=Locked
 }
 {bone1Name=dyn_skirt_bk_01_l, bone2Name=dyn_skirt_bk_01_l_MSTR, swing1LimitDegrees=25, swing2LimitDegrees=45, twistLimitDegrees=5}
 {bone1Name=dyn_skirt_bk_01_r, bone2Name=dyn_skirt_bk_01_r_MSTR, swing1LimitDegrees=25, swing2LimitDegrees=45, twistLimitDegrees=5}
 {bone1Name=dyn_skirt_bk_01, bone2Name=dyn_skirt_bk_01_MSTR, swing1LimitDegrees=25, swing2LimitDegrees=45, twistLimitDegrees=5}
 {bone1Name=dyn_skirt_bk_02_l, bone2Name=dyn_skirt_bk_01_l, swing1LimitDegrees=25, swing2LimitDegrees=45, twistLimitDegrees=5}
 {bone1Name=dyn_skirt_bk_02_r, bone2Name=dyn_skirt_bk_01_r, swing1LimitDegrees=25, swing2LimitDegrees=45, twistLimitDegrees=5}
 {bone1Name=dyn_skirt_bk_02, bone2Name=dyn_skirt_bk_01, swing1LimitDegrees=25, swing2LimitDegrees=45, twistLimitDegrees=5}
 {bone1Name=dyn_skirt_bk_03_l, bone2Name=dyn_skirt_bk_02_l, swing1LimitDegrees=25, swing2LimitDegrees=45, twistLimitDegrees=5}
 {bone1Name=dyn_skirt_bk_03_r, bone2Name=dyn_skirt_bk_02_r, swing1LimitDegrees=25, swing2LimitDegrees=45, twistLimitDegrees=5}
 {bone1Name=dyn_skirt_bk_03, bone2Name=dyn_skirt_bk_02, swing1LimitDegrees=25, swing2LimitDegrees=45, twistLimitDegrees=5}
]
