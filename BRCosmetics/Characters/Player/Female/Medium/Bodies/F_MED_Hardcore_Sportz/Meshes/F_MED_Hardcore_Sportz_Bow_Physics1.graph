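# Physics asset for the Hardcore Sportz female medium body's dangling bow, front and back straps, and skirt bones.
# Pelvis-rooted chains use capsules, with boxes for front straps; bow and straps hang from MSTR anchor bodies, mostly tiny capsules except the large left back-strap one.
# Anchor and skirt joints rotate freely, while bow and strap segments have angular limits.
# An extra free-rotation constraint cross-links the left and right third front-strap segments.
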PhysicsAsset /BRCosmetics/Characters/Player/Female/Medium/Bodies/F_MED_Hardcore_Sportz/Meshes/F_MED_Hardcore_Sportz_Bow_Physics1
bodies {
 dyn_bow_bk_01_l [{shapeType=Capsule, length=5, radius=3, shapeName=dyn_bow_bk_01_l_capsule, center=<1.6739 0 0>, rotation=<rot 90 0 0>}]
 dyn_bow_bk_01_l_MSTR [{shapeType=Capsule, length=0.01, radius=0.01}]
 dyn_bow_bk_01_r [{shapeType=Capsule, length=5, radius=3, shapeName=dyn_bow_bk_01_r_capsule, center=<1.6739 0 0>, rotation=<rot 90 0 0>}]
 dyn_bow_bk_01_r_MSTR [{shapeType=Capsule, length=0.01, radius=0.01}]
 dyn_bow_bk_02_l [{shapeType=Capsule, length=7, radius=2, shapeName=dyn_bow_bk_02_l_capsule, center=<0.9342 0 0>, rotation=<rot 90 0 0>}]
 dyn_bow_bk_02_r [{shapeType=Capsule, length=7, radius=2, shapeName=dyn_bow_bk_02_r_capsule, center=<0.9342 0 0>, rotation=<rot 90 0 0>}]
 dyn_bow_bk_03_l [{shapeType=Capsule, length=6, radius=2, shapeName=dyn_bow_bk_03_l_capsule, rotation=<rot 90 0 0>}]
 dyn_bow_bk_03_r [{shapeType=Capsule, length=6, radius=2, shapeName=dyn_bow_bk_03_r_capsule, rotation=<rot 90 0 0>}]
 dyn_skirt_bk_02 [{
   shapeType=Capsule
   length=10
   radius=9
   shapeName=dyn_skirt_bk_02_r_capsule
   center=<-1.0529 0.0001 -7.0101>
   rotation=<rot 76.9284 179.9999 179.9999>
  }]
 dyn_skirt_bk_02_l [{
   shapeType=Capsule
   length=10.3366
   radius=8.6634
   shapeName=dyn_skirt_bk_02_l_capsule
   center=<1.7401 0.0001 -5.9593>
   rotation=<rot 76.9209 179.9999 179.9999>
  }]
 dyn_skirt_bk_02_r [{
   shapeType=Capsule
   length=10.3366
   radius=8.6634
   shapeName=dyn_skirt_bk_02_r_capsule
   center=<1.7394 0.0001 -5.9595>
   rotation=<rot 76.9284 179.9999 179.9999>
  }]
 dyn_skirt_bk_03 [{
   shapeType=Capsule
   length=15
   radius=9
   shapeName=dyn_skirt_bk_03_r_capsule
   center=<1.7371 0.0002 -4.9533>
   rotation=<rot 86.2503 179.9997 179.9996>
  }]
 dyn_skirt_bk_03_l [{
   shapeType=Capsule
   length=15
   radius=9
   shapeName=dyn_skirt_bk_03_l_capsule
   center=<1.7375 0.0002 -4.9531>
   rotation=<rot 86.2397 -179.9992 -179.9992>
  }]
 dyn_skirt_bk_03_r [{
   shapeType=Capsule
   length=15
   radius=9
   shapeName=dyn_skirt_bk_03_r_capsule
   center=<1.7371 0.0002 -4.9533>
   rotation=<rot 86.2503 179.9997 179.9996>
  }]
 dyn_skirt_fr_01 [{shapeType=Capsule, length=15, radius=10, center=<5.72 0.0001 -5.7556>, rotation=<rot -86.3367 0.0001 0>}]
 dyn_skirt_fr_01_l [{shapeType=Capsule, length=7, radius=7, shapeName=dyn_skirt_fr_01_l_capsule, center=<8.5 0 -3>, rotation=<rot 90 0 0>}]
 dyn_skirt_fr_01_r [{shapeType=Capsule, length=7, radius=7, shapeName=dyn_skirt_fr_01_r_capsule, center=<8.5 0 -3>, rotation=<rot 90 0 0>}]
 dyn_skirt_fr_02 [{shapeType=Capsule, length=12, radius=10, center=<5.5947 0 -5.0331>, rotation=<rot -86.1507 0.0001 0>}]
 dyn_skirt_fr_02_l [{shapeType=Capsule, length=7, radius=7, shapeName=dyn_skirt_fr_02_l_capsule, center=<8.5 0 -3>, rotation=<rot 90 0 0>}]
 dyn_skirt_fr_02_r [{shapeType=Capsule, length=7, radius=7, shapeName=dyn_skirt_fr_02_r_capsule, center=<8.5 0 -3>, rotation=<rot 90 0 0>}]
 dyn_skirt_fr_03 [{shapeType=Capsule, length=12, radius=10, center=<3.0494 0 -2.6454>, rotation=<rot -85.1554 0.0001 0>}]
 dyn_skirt_fr_03_l [{shapeType=Capsule, length=7, radius=7, shapeName=dyn_skirt_fr_03_l_capsule, center=<8.5 0 -3>, rotation=<rot 90 0 0>}]
 dyn_skirt_fr_03_r [{shapeType=Capsule, length=7, radius=7, shapeName=dyn_skirt_fr_03_r_capsule, center=<8.5 0 -3>, rotation=<rot 90 0 0>}]
 dyn_strap_bk_01_l [{shapeType=Capsule, length=3.5, radius=1.2832, shapeName=dyn_strap_bk_01_l_capsule, rotation=<rot 90 0 0>}]
 dyn_strap_bk_01_l_MSTR [{shapeType=Capsule, length=14.75, radius=8, center=<4.9325 0 -9.031>, rotation=<rot 90 0 0>}]
 dyn_strap_bk_01_r [{shapeType=Capsule, length=3.5, radius=1.2832, shapeName=dyn_strap_bk_01_r_capsule, center=<0 -0.18 0>, rotation=<rot 90 0 0>}]
 dyn_strap_bk_01_r_MSTR [{shapeType=Capsule, length=0.01, radius=0.01}]
 dyn_strap_bk_02_l [{shapeType=Capsule, length=3.6867, radius=1.0966, shapeName=dyn_strap_bk_02_l_capsule, rotation=<rot 90 0 0>}]
 dyn_strap_bk_02_r [{shapeType=Capsule, length=3.6867, radius=1.0966, shapeName=dyn_strap_bk_02_r_capsule, rotation=<rot 90 0 0>}]
 dyn_strap_bk_03_l [{shapeType=Capsule, length=3.5169, radius=1.2664, shapeName=dyn_strap_bk_03_l_capsule, center=<0 -0.4 -0.2>, rotation=<rot 90 0 0>}]
 dyn_strap_bk_03_r [{shapeType=Capsule, length=3.5169, radius=1.2664, shapeName=dyn_strap_bk_03_r_capsule, center=<0 0.5 -0.2>, rotation=<rot 90 0 0>}]
 dyn_strap_bk_04_l [{shapeType=Capsule, length=8.7337, radius=1.2664, shapeName=dyn_strap_bk_04_l_capsule, center=<2.4221 -0.4 0>, rotation=<rot 90 0 0>}]
 dyn_strap_bk_04_r [{shapeType=Capsule, length=8.7337, radius=1.2664, shapeName=dyn_strap_bk_04_r_capsule, center=<2.4221 0.5 0>, rotation=<rot 90 0 0>}]
 dyn_strap_fr_01_l [{shapeType=Box, extentX=10.0235, extentY=2.1362, extentZ=7.6168, center=<5.5331 0 0>, rotation=<rot 0 0 89.999>}]
 dyn_strap_fr_01_l_MSTR [{shapeType=Capsule, length=0.01, radius=0.01}]
 dyn_strap_fr_01_r [{shapeType=Box, extentX=10.0235, extentY=2.1362, extentZ=7.6168, center=<5.5331 0 0>, rotation=<rot 0 0 89.999>}]
 dyn_strap_fr_01_r_MSTR [{shapeType=Capsule, length=0.01, radius=0.01}]
 dyn_strap_fr_02_l [{shapeType=Box, extentX=10.0235, extentY=2.1362, extentZ=7.6168, center=<5.5331 0 0>, rotation=<rot 0 0 89.999>}]
 dyn_strap_fr_02_r [{shapeType=Box, extentX=10.0235, extentY=2.1362, extentZ=7.6168, center=<5.5331 0 0>, rotation=<rot 0 0 89.999>}]
 dyn_strap_fr_03_l [{shapeType=Box, extentX=10.0235, extentY=2.1362, extentZ=7.6168, center=<5.5331 0 0>, rotation=<rot 0 0 89.999>}]
 dyn_strap_fr_03_r [{shapeType=Box, extentX=10.0235, extentY=2.1362, extentZ=7.6168, center=<5.5331 0 0>, rotation=<rot 0 0 89.999>}]
 dyn_strap_fr_04_l [{shapeType=Box, extentX=10.0235, extentY=4, extentZ=7.6168, center=<5.5331 0 0>, rotation=<rot 0 0 89.999>}]
 dyn_strap_fr_04_r [{shapeType=Box, extentX=10.0235, extentY=4, extentZ=7.6168, center=<5.5331 0 0>, rotation=<rot 0 0 89.999>}]
 pelvis [{shapeType=Capsule, length=8.8579, radius=9.0267, center=<-1.664 -2.5911 -0.4831>}]
}
constraints [
 {
  bone1Name=dyn_strap_fr_01_r_MSTR
  bone2Name=pelvis
  swing1LimitDegrees=45
  swing1Motion=Free
  swing2LimitDegrees=45
  swing2Motion=Free
  twistLimitDegrees=45
  twistMotion=Free
 }
 {
  bone1Name=dyn_strap_fr_01_l_MSTR
  bone2Name=pelvis
  swing1LimitDegrees=45
  swing1Motion=Free
  swing2LimitDegrees=45
  swing2Motion=Free
  twistLimitDegrees=45
  twistMotion=Free
 }
 {
  bone1Name=dyn_strap_bk_01_l_MSTR
  bone2Name=pelvis
  swing1LimitDegrees=45
  swing1Motion=Free
  swing2LimitDegrees=45
  swing2Motion=Free
  twistLimitDegrees=45
  twistMotion=Free
 }
 {
  bone1Name=dyn_strap_bk_01_r_MSTR
  bone2Name=pelvis
  swing1LimitDegrees=45
  swing1Motion=Free
  swing2LimitDegrees=45
  swing2Motion=Free
  twistLimitDegrees=45
  twistMotion=Free
 }
 {
  bone1Name=dyn_bow_bk_01_r_MSTR
  bone2Name=pelvis
  swing1LimitDegrees=45
  swing1Motion=Free
  swing2LimitDegrees=45
  swing2Motion=Free
  twistLimitDegrees=45
  twistMotion=Free
 }
 {
  bone1Name=dyn_bow_bk_01_l_MSTR
  bone2Name=pelvis
  swing1LimitDegrees=45
  swing1Motion=Free
  swing2LimitDegrees=45
  swing2Motion=Free
  twistLimitDegrees=45
  twistMotion=Free
 }
 {bone1Name=dyn_strap_fr_01_r, bone2Name=dyn_strap_fr_01_r_MSTR, swing1LimitDegrees=45, swing2LimitDegrees=75, twistLimitDegrees=5}
 {bone1Name=dyn_strap_fr_02_r, bone2Name=dyn_strap_fr_01_r, swing1LimitDegrees=45, swing2LimitDegrees=75, twistLimitDegrees=5}
 {bone1Name=dyn_strap_fr_03_r, bone2Name=dyn_strap_fr_02_r, swing1LimitDegrees=45, swing2LimitDegrees=75, twistLimitDegrees=5}
 {bone1Name=dyn_strap_fr_04_r, bone2Name=dyn_strap_fr_03_r, swing1LimitDegrees=45, swing2LimitDegrees=75, twistLimitDegrees=5}
 {bone1Name=dyn_strap_fr_01_l, bone2Name=dyn_strap_fr_01_l_MSTR, swing1LimitDegrees=45, swing2LimitDegrees=75, twistLimitDegrees=5}
 {bone1Name=dyn_strap_fr_02_l, bone2Name=dyn_strap_fr_01_l, swing1LimitDegrees=45, swing2LimitDegrees=75, twistLimitDegrees=5}
 {bone1Name=dyn_strap_fr_03_l, bone2Name=dyn_strap_fr_02_l, swing1LimitDegrees=45, swing2LimitDegrees=75, twistLimitDegrees=5}
 {bone1Name=dyn_strap_fr_04_l, bone2Name=dyn_strap_fr_03_l, swing1LimitDegrees=45, swing2LimitDegrees=75, twistLimitDegrees=5}
 {bone1Name=dyn_bow_bk_01_r, bone2Name=dyn_bow_bk_01_r_MSTR, swing1LimitDegrees=15, swing2LimitDegrees=15, twistLimitDegrees=5}
 {bone1Name=dyn_bow_bk_02_r, bone2Name=dyn_bow_bk_01_r, swing1LimitDegrees=15, swing2LimitDegrees=15, twistLimitDegrees=5}
 {bone1Name=dyn_bow_bk_03_r, bone2Name=dyn_bow_bk_02_r, swing1LimitDegrees=15, swing2LimitDegrees=15, twistLimitDegrees=5}
 {bone1Name=dyn_bow_bk_01_l, bone2Name=dyn_bow_bk_01_l_MSTR, swing1LimitDegrees=15, swing2LimitDegrees=15, twistLimitDegrees=5}
 {bone1Name=dyn_bow_bk_02_l, bone2Name=dyn_bow_bk_01_l, swing1LimitDegrees=15, swing2LimitDegrees=15, twistLimitDegrees=5}
 {bone1Name=dyn_bow_bk_03_l, bone2Name=dyn_bow_bk_02_l, swing1LimitDegrees=15, swing2LimitDegrees=15, twistLimitDegrees=5}
 {
  bone1Name=dyn_strap_fr_03_r
  bone2Name=dyn_strap_fr_03_l
  swing1LimitDegrees=45
  swing1Motion=Free
  swing2LimitDegrees=45
  swing2Motion=Free
  twistLimitDegrees=45
  twistMotion=Free
 }
 {
  bone1Name=dyn_skirt_bk_02_l
  bone2Name=pelvis
  swing1LimitDegrees=45
  swing1Motion=Free
  swing2LimitDegrees=45
  swing2Motion=Free
  twistLimitDegrees=45
  twistMotion=Free
 }
 {
  bone1Name=dyn_skirt_bk_03_l
  bone2Name=dyn_skirt_bk_02_l
  swing1LimitDegrees=45
  swing1Motion=Free
  swing2LimitDegrees=45
  swing2Motion=Free
  twistLimitDegrees=45
  twistMotion=Free
 }
 {
  bone1Name=dyn_skirt_bk_02_r
  bone2Name=pelvis
  swing1LimitDegrees=45
  swing1Motion=Free
  swing2LimitDegrees=45
  swing2Motion=Free
  twistLimitDegrees=45
  twistMotion=Free
 }
 {
  bone1Name=dyn_skirt_bk_03_r
  bone2Name=dyn_skirt_bk_02_r
  swing1LimitDegrees=45
  swing1Motion=Free
  swing2LimitDegrees=45
  swing2Motion=Free
  twistLimitDegrees=45
  twistMotion=Free
 }
 {
  bone1Name=dyn_skirt_fr_01
  bone2Name=pelvis
  swing1LimitDegrees=45
  swing1Motion=Free
  swing2LimitDegrees=45
  swing2Motion=Free
  twistLimitDegrees=45
  twistMotion=Free
 }
 {
  bone1Name=dyn_skirt_fr_02
  bone2Name=dyn_skirt_fr_01
  swing1LimitDegrees=45
  swing1Motion=Free
  swing2LimitDegrees=45
  swing2Motion=Free
  twistLimitDegrees=45
  twistMotion=Free
 }
 {
  bone1Name=dyn_skirt_fr_03
  bone2Name=dyn_skirt_fr_02
  swing1LimitDegrees=45
  swing1Motion=Free
  swing2LimitDegrees=45
  swing2Motion=Free
  twistLimitDegrees=45
  twistMotion=Free
 }
 {
  bone1Name=dyn_skirt_fr_01_l
  bone2Name=pelvis
  swing1LimitDegrees=45
  swing1Motion=Free
  swing2LimitDegrees=45
  swing2Motion=Free
  twistLimitDegrees=45
  twistMotion=Free
 }
 {
  bone1Name=dyn_skirt_fr_02_l
  bone2Name=dyn_skirt_fr_01_l
  swing1LimitDegrees=45
  swing1Motion=Free
  swing2LimitDegrees=45
  swing2Motion=Free
  twistLimitDegrees=45
  twistMotion=Free
 }
 {
  bone1Name=dyn_skirt_fr_03_l
  bone2Name=dyn_skirt_fr_02_l
  swing1LimitDegrees=45
  swing1Motion=Free
  swing2LimitDegrees=45
  swing2Motion=Free
  twistLimitDegrees=45
  twistMotion=Free
 }
 {
  bone1Name=dyn_skirt_fr_01_r
  bone2Name=pelvis
  swing1LimitDegrees=45
  swing1Motion=Free
  swing2LimitDegrees=45
  swing2Motion=Free
  twistLimitDegrees=45
  twistMotion=Free
 }
 {
  bone1Name=dyn_skirt_fr_02_r
  bone2Name=dyn_skirt_fr_01_r
  swing1LimitDegrees=45
  swing1Motion=Free
  swing2LimitDegrees=45
  swing2Motion=Free
  twistLimitDegrees=45
  twistMotion=Free
 }
 {
  bone1Name=dyn_skirt_fr_03_r
  bone2Name=dyn_skirt_fr_02_r
  swing1LimitDegrees=45
  swing1Motion=Free
  swing2LimitDegrees=45
  swing2Motion=Free
  twistLimitDegrees=45
  twistMotion=Free
 }
 {
  bone1Name=dyn_skirt_bk_03
  bone2Name=dyn_skirt_bk_02
  swing1LimitDegrees=45
  swing1Motion=Free
  swing2LimitDegrees=45
  swing2Motion=Free
  twistLimitDegrees=45
  twistMotion=Free
 }
 {
  bone1Name=dyn_skirt_bk_02
  bone2Name=pelvis
  swing1LimitDegrees=45
  swing1Motion=Free
  swing2LimitDegrees=45
  swing2Motion=Free
  twistLimitDegrees=45
  twistMotion=Free
 }
 {bone1Name=dyn_strap_bk_01_l, bone2Name=dyn_strap_bk_01_l_MSTR, swing1LimitDegrees=10, swing2LimitDegrees=45, twistLimitDegrees=0}
 {bone1Name=dyn_strap_bk_02_l, bone2Name=dyn_strap_bk_01_l, swing1LimitDegrees=10, swing2LimitDegrees=45, twistLimitDegrees=0}
 {bone1Name=dyn_strap_bk_03_l, bone2Name=dyn_strap_bk_02_l, swing1LimitDegrees=10, swing2LimitDegrees=45, twistLimitDegrees=0}
 {bone1Name=dyn_strap_bk_04_l, bone2Name=dyn_strap_bk_03_l, swing1LimitDegrees=10, swing2LimitDegrees=45, twistLimitDegrees=0}
 {bone1Name=dyn_strap_bk_01_r, bone2Name=dyn_strap_bk_01_r_MSTR, swing1LimitDegrees=10, swing2LimitDegrees=45, twistLimitDegrees=0}
 {bone1Name=dyn_strap_bk_02_r, bone2Name=dyn_strap_bk_01_r, swing1LimitDegrees=10, swing2LimitDegrees=45, twistLimitDegrees=0}
 {bone1Name=dyn_strap_bk_03_r, bone2Name=dyn_strap_bk_02_r, swing1LimitDegrees=10, swing2LimitDegrees=45, twistLimitDegrees=0}
 {bone1Name=dyn_strap_bk_04_r, bone2Name=dyn_strap_bk_03_r, swing1LimitDegrees=10, swing2LimitDegrees=45, twistLimitDegrees=0}
]
