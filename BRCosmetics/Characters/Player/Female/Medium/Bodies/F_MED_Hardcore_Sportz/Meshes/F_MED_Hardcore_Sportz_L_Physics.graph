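# Physics asset for the female medium Hardcore Sportz body, covering the left and front-center skirt bones with pelvis and left thigh capsules.
# Four three-bone capsule chains hang from tiny master bodies that are locked to the pelvis, as are the thigh and a chainless back master.
# Middle joints lock twist and one swing axis for hinge-like bending, while tip joints lock only twist.
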PhysicsAsset /BRCosmetics/Characters/Player/Female/Medium/Bodies/F_MED_Hardcore_Sportz/Meshes/F_MED_Hardcore_Sportz_L_Physics
bodies {
 dyn_skirt_01_l [{shapeType=Capsule, length=8, radius=4, shapeName=dyn_skirt_01_l_capsule, center=<2.2571 0 0>, rotation=<rot 80.8967 0 0>}]
 dyn_skirt_01_l_MSTR [{shapeType=Capsule, length=0.01, radius=0.01}]
 dyn_skirt_02_l [{shapeType=Capsule, length=8, radius=4, shapeName=dyn_skirt_02_l_capsule, center=<1.0204 0 0>, rotation=<rot 90 0 0>}]
 dyn_skirt_03_l [{shapeType=Capsule, length=8, radius=4, shapeName=dyn_skirt_03_l_capsule, center=<2.2571 0 0>, rotation=<rot 90 0 0>}]
 dyn_skirt_bk_01_MSTR [{shapeType=Capsule, length=0.01, radius=0.01}]
 dyn_skirt_fr2_01_l [{shapeType=Capsule, length=8, radius=4, shapeName=dyn_skirt_fr2_01_l_capsule, center=<2.2571 0 0>, rotation=<rot 90 0 0>}]
 dyn_skirt_fr2_01_l_MSTR [{shapeType=Capsule, length=0.01, radius=0.01}]
 dyn_skirt_fr2_02_l [{shapeType=Capsule, length=8, radius=4, shapeName=dyn_skirt_fr2_02_l_capsule, center=<1.022 0 0>, rotation=<rot 87.0791 -179.9995 -179.9995>}]
 dyn_skirt_fr2_03_l [{shapeType=Capsule, length=8, radius=4, shapeName=dyn_skirt_fr2_03_l_capsule, center=<2.2571 0 0>, rotation=<rot 90 0 0>}]
 dyn_skirt_fr_01 [{shapeType=Capsule, length=8, radius=4, shapeName=dyn_skirt_fr_01_capsule, center=<2.2571 0 0>, rotation=<rot 90 0 0>}]
 dyn_skirt_fr_01_MSTR [{shapeType=Capsule, length=0.01, radius=0.01}]
 dyn_skirt_fr_01_l [{shapeType=Capsule, length=8, radius=4, shapeName=dyn_skirt_fr_01_l_capsule, center=<2.2571 0 0>, rotation=<rot 90 0 0>}]
 dyn_skirt_fr_01_l_MSTR [{shapeType=Capsule, length=0.01, radius=0.01}]
 dyn_skirt_fr_02 [{shapeType=Capsule, length=8, radius=4, shapeName=dyn_skirt_fr_02_capsule, center=<1.0204 0 0>, rotation=<rot 90 0 0>}]
 dyn_skirt_fr_02_l [{shapeType=Capsule, length=8, radius=4, shapeName=dyn_skirt_fr_02_l_capsule, center=<1.0204 0 0>, rotation=<rot 90 0 0>}]
 dyn_skirt_fr_03 [{shapeType=Capsule, length=8, radius=4, shapeName=dyn_skirt_fr_03_capsule, center=<2.2571 0 0>, rotation=<rot 90 0 0>}]
 dyn_skirt_fr_03_l [{shapeType=Capsule, length=8, radius=4, shapeName=dyn_skirt_fr_03_l_capsule, center=<2.2571 0 0>, rotation=<rot 87.0077 0 0>}]
 pelvis [{shapeType=Capsule, length=17, radius=11, center=<-2.838 0 0>, rotation=<rot 90 0 0>}]
 thigh_l [{shapeType=Capsule, length=33, radius=5, center=<-25.1663 0 0.9817>, rotation=<rot 90 -2.5305 177.4696>}]
}
constraints [
 {
  bone1Name=dyn_skirt_fr_01_l_MSTR
  bone2Name=pelvis
  swing1LimitDegrees=45
  swing1Motion=Locked
  swing2LimitDegrees=45
  swing2Motion=Locked
  twistLimitDegrees=45
  twistMotion=Locked
 }
 {
  bone1Name=dyn_skirt_01_l_MSTR
  bone2Name=pelvis
  swing1LimitDegrees=45
  swing1Motion=Locked
  swing2LimitDegrees=45
  swing2Motion=Locked
  twistLimitDegrees=45
  twistMotion=Locked
 }
 {bone1Name=dyn_skirt_fr_01_l, bone2Name=dyn_skirt_fr_01_l_MSTR, swing1LimitDegrees=2, swing2LimitDegrees=10, twistLimitDegrees=2}
 {
  bone1Name=dyn_skirt_fr_02_l
  bone2Name=dyn_skirt_fr_01_l
  swing1LimitDegrees=5
  swing1Motion=Locked
  swing2LimitDegrees=9
  twistLimitDegrees=5
  twistMotion=Locked
 }
 {
  bone1Name=dyn_skirt_fr_03_l
  bone2Name=dyn_skirt_fr_02_l
  swing1LimitDegrees=5
  swing2LimitDegrees=40
  twistLimitDegrees=5
  twistMotion=Locked
 }
 {
  bone1Name=thigh_l
  bone2Name=pelvis
  swing1LimitDegrees=45
  swing1Motion=Locked
  swing2LimitDegrees=45
  swing2Motion=Locked
  twistLimitDegrees=45
  twistMotion=Locked
 }
 {
  bone1Name=dyn_skirt_bk_01_MSTR
  bone2Name=pelvis
  swing1LimitDegrees=45
  swing1Motion=Locked
  swing2LimitDegrees=45
  swing2Motion=Locked
  twistLimitDegrees=45
  twistMotion=Locked
 }
 {
  bone1Name=dyn_skirt_fr_01_MSTR
  bone2Name=pelvis
  swing1LimitDegrees=45
  swing1Motion=Locked
  swing2LimitDegrees=45
  swing2Motion=Locked
  twistLimitDegrees=45
  twistMotion=Locked
 }
 {
  bone1Name=dyn_skirt_fr2_01_l_MSTR
  bone2Name=pelvis
  swing1LimitDegrees=45
  swing1Motion=Locked
  swing2LimitDegrees=45
  swing2Motion=Locked
  twistLimitDegrees=45
  twistMotion=Locked
 }
 {bone1Name=dyn_skirt_01_l, bone2Name=dyn_skirt_01_l_MSTR, swing1LimitDegrees=2, swing2LimitDegrees=45, twistLimitDegrees=2}
 {bone1Name=dyn_skirt_fr2_01_l, bone2Name=dyn_skirt_fr2_01_l_MSTR, swing1LimitDegrees=2, swing2LimitDegrees=45, twistLimitDegrees=2}
 {bone1Name=dyn_skirt_fr_01, bone2Name=dyn_skirt_fr_01_MSTR, swing1LimitDegrees=2, swing2LimitDegrees=10, twistLimitDegrees=2}
 {
  bone1Name=dyn_skirt_02_l
  bone2Name=dyn_skirt_01_l
  swing1LimitDegrees=5
  swing1Motion=Locked
  swing2LimitDegrees=9
  twistLimitDegrees=5
  twistMotion=Locked
 }
 {
  bone1Name=dyn_skirt_fr2_02_l
  bone2Name=dyn_skirt_fr2_01_l
  swing1LimitDegrees=5
  swing1Motion=Locked
  swing2LimitDegrees=9
  twistLimitDegrees=5
  twistMotion=Locked
 }
 {
  bone1Name=dyn_skirt_fr_02
  bone2Name=dyn_skirt_fr_01
  swing1LimitDegrees=5
  swing1Motion=Locked
  swing2LimitDegrees=9
  twistLimitDegrees=5
  twistMotion=Locked
 }
 {bone1Name=dyn_skirt_03_l, bone2Name=dyn_skirt_02_l, swing1LimitDegrees=5, swing2LimitDegrees=40, twistLimitDegrees=5, twistMotion=Locked}
 {
  bone1Name=dyn_skirt_fr2_03_l
  bone2Name=dyn_skirt_fr2_02_l
  swing1LimitDegrees=5
  swing2LimitDegrees=40
  twistLimitDegrees=5
  twistMotion=Locked
 }
 {bone1Name=dyn_skirt_fr_03, bone2Name=dyn_skirt_fr_02, swing1LimitDegrees=5, swing2LimitDegrees=40, twistLimitDegrees=5, twistMotion=Locked}
]
