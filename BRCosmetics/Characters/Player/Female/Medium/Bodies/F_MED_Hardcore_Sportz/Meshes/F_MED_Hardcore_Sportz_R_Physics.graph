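# Physics asset for the right-side skirt of the female medium Hardcore Sportz body, giving its dyn_skirt bones bodies for physical motion.
# Three three-link capsule chains (side, fr, fr2) hang from tiny MSTR anchor bodies locked to the pelvis; an extra fr MSTR anchor has no chain.
# Middle joints lock twist and one swing axis, while tip joints lock only twist.
# Larger pelvis and right-thigh capsules supply the body volumes, also joined by a locked constraint.
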PhysicsAsset /BRCosmetics/Characters/Player/Female/Medium/Bodies/F_MED_Hardcore_Sportz/Meshes/F_MED_Hardcore_Sportz_R_Physics
bodies {
 dyn_skirt_01_r [{shapeType=Capsule, length=8, radius=4, shapeName=dyn_skirt_01_r_capsule, center=<2.2571 0 0>, rotation=<rot 80.1902 -0.0001 -0.0001>}]
 dyn_skirt_01_r_MSTR [{shapeType=Capsule, length=0.01, radius=0.01}]
 dyn_skirt_02_r [{shapeType=Capsule, length=8, radius=4, shapeName=dyn_skirt_02_r_capsule, center=<1.0204 0 0>, rotation=<rot 90 0 0>}]
 dyn_skirt_03_r [{shapeType=Capsule, length=8, radius=4, shapeName=dyn_skirt_03_r_capsule, center=<2.2571 0 0>, rotation=<rot 90 0 0>}]
 dyn_skirt_fr2_01_r [{shapeType=Capsule, length=8, radius=4, shapeName=dyn_skirt_fr2_01_r_capsule, center=<2.2571 0 0>, rotation=<rot 90 0 0>}]
 dyn_skirt_fr2_01_r_MSTR [{shapeType=Capsule, length=0.01, radius=0.01}]
 dyn_skirt_fr2_02_r [{shapeType=Capsule, length=8, radius=4, shapeName=dyn_skirt_fr2_02_r_capsule, center=<1.022 0 0>, rotation=<rot 87.0924 179.9999 179.9999>}]
 dyn_skirt_fr2_03_r [{shapeType=Capsule, length=8, radius=4, shapeName=dyn_skirt_fr2_03_r_capsule, center=<2.2571 0 0>, rotation=<rot 90 0 0>}]
 dyn_skirt_fr_01_MSTR [{shapeType=Capsule, length=0.01, radius=0.01}]
 dyn_skirt_fr_01_r [{shapeType=Capsule, length=8, radius=4, shapeName=dyn_skirt_fr_01_r_capsule, center=<2.2571 0 0>, rotation=<rot 90 0 0>}]
 dyn_skirt_fr_01_r_MSTR [{shapeType=Capsule, length=0.01, radius=0.01}]
 dyn_skirt_fr_02_r [{shapeType=Capsule, length=8, radius=4, shapeName=dyn_skirt_fr_02_r_capsule, center=<1.0204 0 0>, rotation=<rot 90 0 0>}]
 dyn_skirt_fr_03_r [{shapeType=Capsule, length=8, radius=4, shapeName=dyn_skirt_fr_03_r_capsule, center=<2.2571 0 0>, rotation=<rot 87.0077 -0.0001 -0.0001>}]
 pelvis [{shapeType=Capsule, length=17, radius=11, center=<-2.838 0 0>, rotation=<rot 90 0 0>}]
 thigh_r [{shapeType=Capsule, length=33, radius=5.5, center=<25.1663 -0.0001 0>, rotation=<rot 90 7.5524 7.5525>}]
}
constraints [
 {
  bone1Name=dyn_skirt_fr_01_r_MSTR
  bone2Name=pelvis
  swing1LimitDegrees=45
  swing1Motion=Locked
  swing2LimitDegrees=45
  swing2Motion=Locked
  twistLimitDegrees=45
  twistMotion=Locked
 }
 {
  bone1Name=dyn_skirt_01_r_MSTR
  bone2Name=pelvis
  swing1LimitDegrees=45
  swing1Motion=Locked
  swing2LimitDegrees=45
  swing2Motion=Locked
  twistLimitDegrees=45
  twistMotion=Locked
 }
 {
  bone1Name=thigh_r
  bone2Name=pelvis
  swing1LimitDegrees=45
  swing1Motion=Locked
  swing2LimitDegrees=45
  swing2Motion=Locked
  twistLimitDegrees=45
  twistMotion=Locked
 }
 {
  bone1Name=dyn_skirt_fr_01_MSTR
  bone2Name=pelvis
  swing1LimitDegrees=45
  swing1Motion=Locked
  swing2LimitDegrees=45
  swing2Motion=Locked
  twistLimitDegrees=45
  twistMotion=Locked
 }
 {
  bone1Name=dyn_skirt_fr2_01_r_MSTR
  bone2Name=pelvis
  swing1LimitDegrees=45
  swing1Motion=Locked
  swing2LimitDegrees=45
  swing2Motion=Locked
  twistLimitDegrees=45
  twistMotion=Locked
 }
 {bone1Name=dyn_skirt_fr_01_r, bone2Name=dyn_skirt_fr_01_r_MSTR, swing1LimitDegrees=2, swing2LimitDegrees=45, twistLimitDegrees=2}
 {bone1Name=dyn_skirt_01_r, bone2Name=dyn_skirt_01_r_MSTR, swing1LimitDegrees=2, swing2LimitDegrees=45, twistLimitDegrees=2}
 {bone1Name=dyn_skirt_fr2_01_r, bone2Name=dyn_skirt_fr2_01_r_MSTR, swing1LimitDegrees=2, swing2LimitDegrees=45, twistLimitDegrees=2}
 {
  bone1Name=dyn_skirt_fr_02_r
  bone2Name=dyn_skirt_fr_01_r
  swing1LimitDegrees=5
  swing1Motion=Locked
  swing2LimitDegrees=9
  twistLimitDegrees=5
  twistMotion=Locked
 }
 {
  bone1Name=dyn_skirt_02_r
  bone2Name=dyn_skirt_01_r
  swing1LimitDegrees=5
  swing1Motion=Locked
  swing2LimitDegrees=9
  twistLimitDegrees=5
  twistMotion=Locked
 }
 {
  bone1Name=dyn_skirt_fr2_02_r
  bone2Name=dyn_skirt_fr2_01_r
  swing1LimitDegrees=5
  swing1Motion=Locked
  swing2LimitDegrees=9
  twistLimitDegrees=5
  twistMotion=Locked
 }
 {
  bone1Name=dyn_skirt_fr_03_r
  bone2Name=dyn_skirt_fr_02_r
  swing1LimitDegrees=5
  swing2LimitDegrees=40
  twistLimitDegrees=5
  twistMotion=Locked
 }
 {bone1Name=dyn_skirt_03_r, bone2Name=dyn_skirt_02_r, swing1LimitDegrees=5, swing2LimitDegrees=40, twistLimitDegrees=5, twistMotion=Locked}
 {
  bone1Name=dyn_skirt_fr2_03_r
  bone2Name=dyn_skirt_fr2_02_r
  swing1LimitDegrees=5
  swing2LimitDegrees=40
  twistLimitDegrees=5
  twistMotion=Locked
 }
]
