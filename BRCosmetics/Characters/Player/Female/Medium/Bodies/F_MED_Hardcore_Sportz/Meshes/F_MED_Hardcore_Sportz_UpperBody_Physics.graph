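# Physics asset for the upper body mesh of the female medium Hardcore Sportz character body, covering the right-side scarf bones.
# A tiny master capsule anchors a root scarf bone, which branches into a two-bone A chain and a three-bone B chain.
# Identical small capsules on the scarf bones are linked parent to child by constraints with tight, uniform swing and twist limits, keeping any simulated sway small.
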PhysicsAsset /BRCosmetics/Characters/Player/Female/Medium/Bodies/F_MED_Hardcore_Sportz/Meshes/F_MED_Hardcore_Sportz_UpperBody_Physics
bodies {
 dyn_scarfA_01_r [{shapeType=Capsule, length=1.5, radius=0.805, center=<0.9028 0 -0.0001>, rotation=<rot 90 -0.0099 -0.0099>}]
 dyn_scarfA_02_r [{shapeType=Capsule, length=1.5, radius=0.805, center=<0.9028 0 -0.0001>, rotation=<rot 90 -0.0099 -0.0099>}]
 dyn_scarfB_01_r [{shapeType=Capsule, length=1.5, radius=0.805, center=<0.9028 0 -0.0001>, rotation=<rot 90 -0.0099 -0.0099>}]
 dyn_scarfB_02_r [{shapeType=Capsule, length=1.5, radius=0.805, center=<0.9028 0 -0.0001>, rotation=<rot 90 -0.0099 -0.0099>}]
 dyn_scarfB_03_r [{shapeType=Capsule, length=1.5, radius=0.805, center=<0.9028 0 -0.0001>, rotation=<rot 90 -0.0099 -0.0099>}]
 dyn_scarf_01_r [{shapeType=Capsule, length=1.5, radius=0.805, center=<0.9028 0 -0.0001>, rotation=<rot 90 -0.0099 -0.0099>}]
 dyn_scarf_01_r_MSTR [{shapeType=Capsule, length=0.01, radius=0.01}]
}
constraints [
 {bone1Name=dyn_scarf_01_r, bone2Name=dyn_scarf_01_r_MSTR, swing1LimitDegrees=10, swing2LimitDegrees=10, twistLimitDegrees=5}
 {bone1Name=dyn_scarfA_01_r, bone2Name=dyn_scarf_01_r, swing1LimitDegrees=10, swing2LimitDegrees=10, twistLimitDegrees=5}
 {bone1Name=dyn_scarfA_02_r, bone2Name=dyn_scarfA_01_r, swing1LimitDegrees=10, swing2LimitDegrees=10, twistLimitDegrees=5}
 {bone1Name=dyn_scarfB_01_r, bone2Name=dyn_scarf_01_r, swing1LimitDegrees=10, swing2LimitDegrees=10, twistLimitDegrees=5}
 {bone1Name=dyn_scarfB_02_r, bone2Name=dyn_scarfB_01_r, swing1LimitDegrees=10, swing2LimitDegrees=10, twistLimitDegrees=5}
 {bone1Name=dyn_scarfB_03_r, bone2Name=dyn_scarfB_02_r, swing1LimitDegrees=10, swing2LimitDegrees=10, twistLimitDegrees=5}
]
