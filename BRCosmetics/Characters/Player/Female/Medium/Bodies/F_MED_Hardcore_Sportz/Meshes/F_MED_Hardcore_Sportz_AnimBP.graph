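# Animation blueprint for the F_MED_Hardcore_Sportz female cosmetic body, adding secondary motion to its skirt, straps and upper-body pieces.
# Weighted pelvis and thigh constraints position skirt master bones before chained RigidBody simulations, which use per-region physics assets and a shared gravity variable.
# AnimDynamics springs drive strap bones, while constraints and additive offsets fit skirt and straps to thighs and each other, with a crouch-weighted front-skirt correction.
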AnimBlueprint /BRCosmetics/Characters/Player/Female/Medium/Bodies/F_MED_Hardcore_Sportz/Meshes/F_MED_Hardcore_Sportz_AnimBP
target_skeleton=&/F_MED_Hardcore_Sportz_Skeleton
graphs [
 AnimGraph {nodes [
   n1 = "Constraint - dyn_strap_fr_03_l (RR)"!(ComponentPose, Alpha=1, ConstraintWeights_0=0, ConstraintWeights_1:float=1) -> (Pose) {
    alphaBoolBlend {blendOption=Linear}
    boneToModify {boneName=dyn_strap_fr_03_l}
    constraintSetup [{transformType=Rotation, targetBone {boneName=dyn_skirt_fr_02_l}}, {transformType=Rotation, targetBone {boneName=dyn_skirt_fr_02}}]
    constraintWeights [1,1]
   }
   "Constraint - dyn_strap_fr_03_r (RR)"!(ComponentPose=@n1, Alpha=1, ConstraintWeights_0=0.5, ConstraintWeights_1:float=0.5) -> (Pose) {
    alphaBoolBlend {blendOption=Linear}
    boneToModify {boneName=dyn_strap_fr_03_r}
    constraintSetup [{transformType=Rotation, targetBone {boneName=dyn_skirt_fr_02_r}}, {transformType=Rotation, targetBone {boneName=dyn_skirt_fr_02}}]
    constraintWeights [1,1]
   }
   n2 = "Constraint - dyn_strap_fr_04_l (RR)"!(ComponentPose, Alpha=1, ConstraintWeights_0=0.5, ConstraintWeights_1:float=0.5) -> (Pose) {
    alphaBoolBlend {blendOption=Linear}
    boneToModify {boneName=dyn_strap_fr_04_l}
    constraintSetup [{transformType=Rotation, targetBone {boneName=dyn_skirt_fr_03_l}}, {transformType=Rotation, targetBone {boneName=dyn_skirt_fr_03}}]
    constraintWeights [1,1]
   }
   "Constraint - dyn_strap_fr_04_r (RR)"!(ComponentPose=@n2, Alpha=1, ConstraintWeights_0=0.5, ConstraintWeights_1:float=0.5) -> (Pose) {
    alphaBoolBlend {blendOption=Linear}
    boneToModify {boneName=dyn_strap_fr_04_r}
    constraintSetup [{transformType=Rotation, targetBone {boneName=dyn_skirt_fr_03_r}}, {transformType=Rotation, targetBone {boneName=dyn_skirt_fr_03}}]
    constraintWeights [1,1]
   }
   n3 = "Constraint - dyn_strap_fr_01_l (RRR)"!(ComponentPose, Alpha=1, ConstraintWeights_0=0, ConstraintWeights_1:float=0, ConstraintWeights_2:float=1) -> (Pose) {
    alphaBoolBlend {blendOption=Linear}
    boneToModify {boneName=dyn_strap_fr_01_l}
    constraintSetup [
     {transformType=Rotation, targetBone {boneName=dyn_skirt_fr_01_l}}
     {transformType=Rotation, targetBone {boneName=dyn_skirt_fr_01}}
     {transformType=Rotation, targetBone {boneName=thigh_l}}
    ]
    constraintWeights [1,1,1]
   }
   "Constraint - dyn_strap_fr_01_r (RRR)"!(ComponentPose=@n3, Alpha=1, ConstraintWeights_0=0, ConstraintWeights_1:float=0, ConstraintWeights_2:float=1) -> (Pose) {
    alphaBoolBlend {blendOption=Linear}
    boneToModify {boneName=dyn_strap_fr_01_r}
    constraintSetup [
     {transformType=Rotation, targetBone {boneName=dyn_skirt_fr_01_r}}
     {transformType=Rotation, targetBone {boneName=dyn_skirt_fr_01}}
     {transformType=Rotation, targetBone {boneName=thigh_r}}
    ]
    constraintWeights [1,1,1]
   }
   n4 = "Constraint - dyn_skirt_fr_03_r (P)"(Alpha=1, ConstraintWeights_0=1) {
    alphaBoolBlend {blendOption=Linear}
    alphaScaleBiasClamp {bInterpResult=true, interpSpeedDecreasing=1}
    boneToModify {boneName=dyn_skirt_fr_03_r}
    constraintSetup [{targetBone {boneName=thigh_r}}]
    constraintWeights [1]
   }
   "Transform (Modify) Bone - Bone: dyn_skirt_fr_03_r"(ComponentPose=@n4, Alpha=1, Translation="6,0,0", Rotation="0,0,0", Scale="1,1,1") {
    translationMode=BMM_Additive
    translationSpace=BCS_BoneSpace
    alphaBoolBlend {blendOption=Linear}
    alphaScaleBiasClamp {bInterpResult=true}
    boneToModify {boneName=dyn_skirt_fr_03_r}
    translation {x=6}
   }
   n5 = "Constraint - dyn_skirt_fr_03_l (P)"(Alpha=1, ConstraintWeights_0=1) {
    alphaBoolBlend {blendOption=Linear}
    alphaScaleBiasClamp {bInterpResult=true, interpSpeedDecreasing=1}
    boneToModify {boneName=dyn_skirt_fr_03_l}
    constraintSetup [{targetBone {boneName=thigh_l}}]
    constraintWeights [1]
   }
   "Transform (Modify) Bone - Bone: dyn_skirt_fr_03_l"(ComponentPose=@n5, Alpha=1, Translation="6,0,0", Rotation="0,0,0", Scale="1,1,1") {
    translationMode=BMM_Additive
    translationSpace=BCS_BoneSpace
    alphaBoolBlend {blendOption=Linear}
    alphaScaleBiasClamp {bInterpResult=true}
    boneToModify {boneName=dyn_skirt_fr_03_l}
    translation {x=6}
   }
   n6 = "Constraint - dyn_skirt_fr2_03_l (P)"(Alpha=1, ConstraintWeights_0=1) {
    alphaBoolBlend {blendOption=Linear}
    alphaScaleBiasClamp {bInterpResult=true, interpSpeedDecreasing=1}
    boneToModify {boneName=dyn_skirt_fr2_03_l}
    constraintSetup [{targetBone {boneName=thigh_l}}]
    constraintWeights [1]
   }
   "Transform (Modify) Bone - Bone: dyn_skirt_fr2_03_l"(ComponentPose=@n6, Alpha=1, Translation="6,0,0", Rotation="0,0,0", Scale="1,1,1") {
    translationMode=BMM_Additive
    translationSpace=BCS_BoneSpace
    alphaBoolBlend {blendOption=Linear}
    alphaScaleBiasClamp {bInterpResult=true}
    boneToModify {boneName=dyn_skirt_fr2_03_l}
    translation {x=6}
   }
   n7 = "Transform (Modify) Bone - Bone: dyn_strap_bk_01_l"(Alpha=1, Translation="0,0,0", Rotation="5,0,0", Scale="1,1,1") {
    rotationMode=BMM_Additive
    rotationSpace=BCS_BoneSpace
    alphaBoolBlend {blendOption=Linear}
    boneToModify {boneName=dyn_strap_bk_01_l}
    rotation {pitch=5}
   }
   "Transform (Modify) Bone - Bone: dyn_strap_bk_01_r"(ComponentPose=@n7, Alpha=1, Translation="0,0,0", Rotation="5,0,0", Scale="1,1,1") {
    rotationMode=BMM_Additive
    rotationSpace=BCS_BoneSpace
    alphaBoolBlend {blendOption=Linear}
    boneToModify {boneName=dyn_strap_bk_01_r}
    rotation {pitch=5}
   }
   n8 = "Constraint - dyn_skirt_01_l_MSTR (PP)"!(ComponentPose=Local To Component(LocalPose=Use cached pose 'presimCache'()), Alpha=1, ConstraintWeights_0=0.5, ConstraintWeights_1:float=0.5) -> (Pose) {
    alphaBoolBlend {blendOption=Linear}
    boneToModify {boneName=dyn_skirt_01_l_MSTR}
    constraintSetup [{targetBone {boneName=pelvis}}, {targetBone {boneName=dyn_thigh_01_l}}]
    constraintWeights [1,1]
   }
   n9 = "Constraint - dyn_skirt_bk2_01_l_MSTR (PP)"!(ComponentPose=@n8, Alpha=1, ConstraintWeights_0=0.7, ConstraintWeights_1:float=0.3) -> (Pose) {
    alphaBoolBlend {blendOption=Linear}
    boneToModify {boneName=dyn_skirt_bk2_01_l_MSTR}
    constraintSetup [{targetBone {boneName=pelvis}}, {targetBone {boneName=dyn_thigh_01_l}}]
    constraintWeights [1,1]
   }
   n10 = "Constraint - dyn_skirt_bk_01_l_MSTR (PP)"!(ComponentPose=@n9, Alpha=1, ConstraintWeights_0=1, ConstraintWeights_1:float=0) -> (Pose) {
    alphaBoolBlend {blendOption=Linear}
    boneToModify {boneName=dyn_skirt_bk_01_l_MSTR}
    constraintSetup [{targetBone {boneName=pelvis}}, {targetBone {boneName=dyn_thigh_01_l}}]
    constraintWeights [1,1]
   }
   n11 = "Constraint - dyn_skirt_bk_01_l_MSTR (RR)"!(ComponentPose=@n10, Alpha=1, ConstraintWeights_0=0.8, ConstraintWeights_1:float=0.2) -> (Pose) {
    alphaBoolBlend {blendOption=Linear}
    boneToModify {boneName=dyn_skirt_bk_01_l_MSTR}
    constraintSetup [{transformType=Rotation, targetBone {boneName=pelvis}}, {transformType=Rotation, targetBone {boneName=dyn_thigh_01_l}}]
    constraintWeights [1,1]
   }
   n12 = "Constraint - dyn_skirt_bk_01_MSTR (PRR)"!(ComponentPose=@n11, Alpha=1, ConstraintWeights_0=0, ConstraintWeights_1:float=0.5, ConstraintWeights_2:float=0.5) -> (Pose) {
    alphaBoolBlend {blendOption=Linear}
    boneToModify {boneName=dyn_skirt_bk_01_MSTR}
    constraintSetup [
     {targetBone {boneName=pelvis}}
     {transformType=Rotation, targetBone {boneName=dyn_thigh_01_r}}
     {transformType=Rotation, targetBone {boneName=dyn_thigh_01_l}}
    ]
    constraintWeights [1,1,1]
   }
   n13 = "Constraint - dyn_skirt_01_r_MSTR (PP)"!(ComponentPose=@n12, Alpha=1, ConstraintWeights_0=0.5, ConstraintWeights_1:float=0.5) -> (Pose) {
    alphaBoolBlend {blendOption=Linear}
    boneToModify {boneName=dyn_skirt_01_r_MSTR}
    constraintSetup [{targetBone {boneName=pelvis}}, {targetBone {boneName=dyn_thigh_01_r}}]
    constraintWeights [1,1]
   }
   n14 = "Constraint - dyn_skirt_bk2_01_r_MSTR (PP)"!(ComponentPose=@n13, Alpha=1, ConstraintWeights_0=1, ConstraintWeights_1:float=0) -> (Pose) {
    alphaBoolBlend {blendOption=Linear}
    boneToModify {boneName=dyn_skirt_bk2_01_r_MSTR}
    constraintSetup [{targetBone {boneName=pelvis}}, {targetBone {boneName=dyn_thigh_01_r}}]
    constraintWeights [1,1]
   }
   n15 = "Constraint - dyn_skirt_bk2_01_r_MSTR (PP)"(ComponentPose=@n14, Alpha=1, ConstraintWeights_0=0.7, ConstraintWeights_1=0.3) {
    alphaBoolBlend {blendOption=Linear}
    boneToModify {boneName=dyn_skirt_bk2_01_r_MSTR}
    constraintSetup [{targetBone {boneName=pelvis}}, {targetBone {boneName=dyn_thigh_01_r}}]
    constraintWeights [1,1]
   }
   n16 = "Constraint - dyn_skirt_bk_01_r_MSTR (PP)"!(ComponentPose=@n15, Alpha=1, ConstraintWeights_0=1, ConstraintWeights_1:float=0) -> (Pose) {
    alphaBoolBlend {blendOption=Linear}
    boneToModify {boneName=dyn_skirt_bk_01_r_MSTR}
    constraintSetup [{targetBone {boneName=pelvis}}, {targetBone {boneName=dyn_thigh_01_r}}]
    constraintWeights [1,1]
   }
   n17 = "Constraint - dyn_skirt_bk_01_r_MSTR (RR)"!(ComponentPose=@n16, Alpha=1, ConstraintWeights_0=0.8, ConstraintWeights_1:float=0.2) -> (Pose) {
    alphaBoolBlend {blendOption=Linear}
    boneToModify {boneName=dyn_skirt_bk_01_r_MSTR}
    constraintSetup [{transformType=Rotation, targetBone {boneName=pelvis}}, {transformType=Rotation, targetBone {boneName=dyn_thigh_01_r}}]
    constraintWeights [1,1]
   }
   n18 = Get finalGravity() -> (finalGravity:Vector)
   n19 = RigidBody!(ComponentPose=@n17, Alpha, OverrideWorldGravity:Vector=@n18, ExternalForce) -> (Pose) {
    bClampLinearTranslationLimitToRefPose=true
    bOverrideWorldGravity=true
    overridePhysicsAsset=&/F_MED_Hardcore_Sportz_L_Physics
    simulationSpace=BaseBoneSpace
    alphaBoolBlend {blendOption=Linear}
    baseBoneRef {boneName=thigh_l}
    componentAppliedLinearAccClamp=<3000 3000 2000>
    componentLinearAccScale=<0.4 0.4 0.1>
    componentLinearVelScale=<0.4 0.4 0.1>
    overrideWorldGravity {z=-980}
   }
   n20 = RigidBody(ComponentPose=@n19, OverrideWorldGravity=@n18) {
    bClampLinearTranslationLimitToRefPose=true
    bOverrideWorldGravity=true
    overridePhysicsAsset=&/F_MED_Hardcore_Sportz_R_Physics
    simulationSpace=BaseBoneSpace
    alphaBoolBlend {blendOption=Linear}
    baseBoneRef {boneName=thigh_r}
    componentAppliedLinearAccClamp=<3000 3000 2000>
    componentLinearAccScale=<0.4 0.4 0.1>
    componentLinearVelScale=<0.4 0.4 0.1>
    overrideWorldGravity {z=-980}
   }
   n21 = RigidBody(ComponentPose=@n20, OverrideWorldGravity=@n18) {
    bClampLinearTranslationLimitToRefPose=true
    bOverrideWorldGravity=true
    overridePhysicsAsset=&/F_MED_Hardcore_Sportz_Bk_Physics
    simulationSpace=BaseBoneSpace
    alphaBoolBlend {blendOption=Linear}
    baseBoneRef {boneName=pelvis}
    componentAppliedLinearAccClamp=<3000 3000 2000>
    componentLinearAccScale=<0.4 0.4 0.1>
    componentLinearVelScale=<0.4 0.4 0.1>
    overrideWorldGravity {z=-980}
   }
   n22 = RigidBody(ComponentPose=@n21, OverrideWorldGravity=@n18) {
    bClampLinearTranslationLimitToRefPose=true
    bOverrideWorldGravity=true
    overridePhysicsAsset=&/F_MED_Hardcore_Sportz_Bow_Physics1
    simulationSpace=BaseBoneSpace
    alphaBoolBlend {blendOption=Linear}
    baseBoneRef {boneName=spine_01}
    componentAppliedLinearAccClamp=<3000 3000 3000>
    componentLinearAccScale=<0.3 0.3 0.5>
    componentLinearVelScale=<0.3 0.3 0.5>
    overrideWorldGravity {z=-980}
   }
   n23 = RigidBody(ComponentPose=@n22, OverrideWorldGravity=@n18) {
    bClampLinearTranslationLimitToRefPose=true
    bOverrideWorldGravity=true
    overridePhysicsAsset=&/F_MED_Hardcore_Sportz_UpperBody_Physics
    simulationSpace=BaseBoneSpace
    alphaBoolBlend {blendOption=Linear}
    baseBoneRef {boneName=spine_05}
    componentAppliedLinearAccClamp=<3000 3000 3000>
    componentLinearAccScale=<0.3 0.3 0.5>
    componentLinearVelScale=<0.3 0.3 0.5>
    overrideWorldGravity {z=-980}
   }
   n24 = "Transform (Modify) Bone - Bone: dyn_strap_bk_01_l"(ComponentPose=@n23, Alpha=1, Translation="0,0,0", Rotation="30,0,0", Scale="1,1,1") {
    rotationMode=BMM_Additive
    rotationSpace=BCS_BoneSpace
    alphaBoolBlend {blendOption=Linear}
    alphaScaleBiasClamp {bInterpResult=true, interpSpeedDecreasing=5, interpSpeedIncreasing=5}
    boneToModify {boneName=dyn_strap_bk_01_l}
    rotation {pitch=30}
   }
   n25 = "Transform (Modify) Bone - Bone: dyn_strap_bk_01_r"(ComponentPose=@n24, Alpha=1, Translation="0,0,0", Rotation="30,0,0", Scale="1,1,1") {
    rotationMode=BMM_Additive
    rotationSpace=BCS_BoneSpace
    alphaBoolBlend {blendOption=Linear}
    alphaScaleBiasClamp {bInterpResult=true, interpSpeedDecreasing=5, interpSpeedIncreasing=5}
    boneToModify {boneName=dyn_strap_bk_01_r}
    rotation {pitch=30}
   }
   n26 = Anim Dynamics - Bone: dyn_strap_bk_02_l(ComponentPose=@n25, Alpha=1, ExternalForce="0,0,0") {
    angularDampingOverride=1
    angularSpringConstant=2000
    bAngularSpring=true
    bGravityOverrideInSimSpace=true
    bOverrideAngularDamping=true
    bUseGravityOverride=true
    simulationSpace=RootRelative
    alphaBoolBlend {blendOption=Linear}
    alphaScaleBiasClamp {bInterpResult=true, interpSpeedDecreasing=5, interpSpeedIncreasing=5}
    boundBone {boneName=dyn_strap_bk_02_l}
    componentAppliedLinearAccClamp=<3000 3000 3000>
    componentLinearAccScale=<0.4 0.3 0.3>
    componentLinearVelScale=<0.4 0.3 0.3>
    gravityOverride {z=-980}
    physicsBodyDefinitions [{
      sphereCollisionRadius=0
      boundBone {boneName=dyn_strap_bk_02_l}
      boxExtents=<10 20 20>
      constraintSetup {angularLimitsMax=<15 15 15>, angularLimitsMin=<-15 -15 -15>, angularTarget=<0 0 0>}
      localJointOffset=<5 0 0>
     }]
   }
   n27 = Anim Dynamics - Bone: dyn_strap_bk_02_r(ComponentPose=@n26, Alpha=1, ExternalForce="0,0,0") {
    angularDampingOverride=1
    angularSpringConstant=2000
    bAngularSpring=true
    bGravityOverrideInSimSpace=true
    bOverrideAngularDamping=true
    bUseGravityOverride=true
    simulationSpace=RootRelative
    alphaBoolBlend {blendOption=Linear}
    alphaScaleBiasClamp {bInterpResult=true, interpSpeedDecreasing=5, interpSpeedIncreasing=5}
    boundBone {boneName=dyn_strap_bk_02_r}
    componentAppliedLinearAccClamp=<3000 3000 3000>
    componentLinearAccScale=<0.4 0.3 0.3>
    componentLinearVelScale=<0.4 0.3 0.3>
    gravityOverride {z=-980}
    physicsBodyDefinitions [{
      sphereCollisionRadius=0
      boundBone {boneName=dyn_strap_bk_02_r}
      boxExtents=<10 20 20>
      constraintSetup {angularLimitsMax=<15 15 15>, angularLimitsMin=<-15 -15 -15>, angularTarget=<0 0 0>}
      localJointOffset=<5 0 0>
     }]
   }
   n28 = "Transform (Modify) Bone - Bone: dyn_strap_bk_01_l"(ComponentPose=@n27, Alpha=1, Translation="0,0,0", Rotation="5,0,0", Scale="1,1,1") {
    rotationMode=BMM_Additive
    rotationSpace=BCS_BoneSpace
    alphaBoolBlend {blendOption=Linear}
    boneToModify {boneName=dyn_strap_bk_01_l}
    rotation {pitch=5}
   }
   n29 = "Transform (Modify) Bone - Bone: dyn_strap_bk_01_r"(ComponentPose=@n28, Alpha=1, Translation="0,0,0", Rotation="5,0,0", Scale="1,1,1") {
    rotationMode=BMM_Additive
    rotationSpace=BCS_BoneSpace
    alphaBoolBlend {blendOption=Linear}
    boneToModify {boneName=dyn_strap_bk_01_r}
    rotation {pitch=5}
   }
   n30 = "Transform (Modify) Bone - Bone: dyn_strap_fr_01_l"(ComponentPose=@n29, Alpha=1, Translation="0,0,0", Rotation="15,0,0", Scale="1,1,1") {
    rotationMode=BMM_Additive
    rotationSpace=BCS_BoneSpace
    alphaBoolBlend {blendOption=Linear}
    alphaScaleBiasClamp {bInterpResult=true, interpSpeedDecreasing=5, interpSpeedIncreasing=5}
    boneToModify {boneName=dyn_strap_fr_01_l}
    rotation {pitch=15}
   }
   n31 = "Transform (Modify) Bone - Bone: dyn_strap_fr_01_r"(ComponentPose=@n30, Alpha=1, Translation="0,0,0", Rotation="15,0,0", Scale="1,1,1") {
    rotationMode=BMM_Additive
    rotationSpace=BCS_BoneSpace
    alphaBoolBlend {blendOption=Linear}
    alphaScaleBiasClamp {bInterpResult=true, interpSpeedDecreasing=5, interpSpeedIncreasing=5}
    boneToModify {boneName=dyn_strap_fr_01_r}
    rotation {pitch=15}
   }
   n32 = Anim Dynamics - Bone: dyn_strap_fr_02_l(ComponentPose=@n31, Alpha=1, ExternalForce="0,0,0") {
    angularDampingOverride=1
    angularSpringConstant=2000
    bAngularSpring=true
    bGravityOverrideInSimSpace=true
    bOverrideAngularDamping=true
    bUseGravityOverride=true
    simulationSpace=RootRelative
    alphaBoolBlend {blendOption=Linear}
    alphaScaleBiasClamp {bInterpResult=true, interpSpeedDecreasing=5, interpSpeedIncreasing=5}
    boundBone {boneName=dyn_strap_fr_02_l}
    componentAppliedLinearAccClamp=<3000 3000 3000>
    componentLinearAccScale=<0.4 0.3 0.3>
    componentLinearVelScale=<0.4 0.3 0.3>
    gravityOverride {z=-980}
    physicsBodyDefinitions [{
      sphereCollisionRadius=0
      boundBone {boneName=dyn_strap_fr_02_l}
      boxExtents=<10 20 20>
      constraintSetup {angularLimitsMax=<15 15 15>, angularLimitsMin=<-15 -15 -15>, angularTarget=<0 0 0>}
      localJointOffset=<5 0 0>
     }]
   }
   n33 = Anim Dynamics - Bone: dyn_strap_fr_02_r(ComponentPose=@n32, Alpha=1, ExternalForce="0,0,0") {
    angularDampingOverride=1
    angularSpringConstant=2000
    bAngularSpring=true
    bGravityOverrideInSimSpace=true
    bOverrideAngularDamping=true
    bUseGravityOverride=true
    simulationSpace=RootRelative
    alphaBoolBlend {blendOption=Linear}
    alphaScaleBiasClamp {bInterpResult=true, interpSpeedDecreasing=5, interpSpeedIncreasing=5}
    boundBone {boneName=dyn_strap_fr_02_r}
    componentAppliedLinearAccClamp=<3000 3000 3000>
    componentLinearAccScale=<0.4 0.3 0.3>
    componentLinearVelScale=<0.4 0.3 0.3>
    gravityOverride {z=-980}
    physicsBodyDefinitions [{
      sphereCollisionRadius=0
      boundBone {boneName=dyn_strap_fr_02_r}
      boxExtents=<10 20 20>
      constraintSetup {angularLimitsMax=<15 15 15>, angularLimitsMin=<-15 -15 -15>, angularTarget=<0 0 0>}
      localJointOffset=<5 0 0>
     }]
   }
   n34 = "Constraint - dyn_skirt_fr_01_l (P)"(ComponentPose=@n33, Alpha=1, ConstraintWeights_0=1) {
    alphaBoolBlend {blendOption=Linear}
    alphaScaleBiasClamp {bInterpResult=true, interpSpeedDecreasing=1}
    boneToModify {boneName=dyn_skirt_fr_01_l}
    constraintSetup [{targetBone {boneName=thigh_l}}]
    constraintWeights [1]
   }
   n35 = "Transform (Modify) Bone - Bone: dyn_skirt_fr_01_l"(ComponentPose=@n34, Alpha=1, Translation="6,0,0", Rotation="0,0,0", Scale="1,1,1") {
    translationMode=BMM_Additive
    translationSpace=BCS_BoneSpace
    alphaBoolBlend {blendOption=Linear}
    alphaScaleBiasClamp {bInterpResult=true}
    boneToModify {boneName=dyn_skirt_fr_01_l}
    translation {x=6}
   }
   n36 = "Constraint - dyn_skirt_fr_02_l (P)"(ComponentPose=@n35, Alpha=1, ConstraintWeights_0=1) {
    alphaBoolBlend {blendOption=Linear}
    alphaScaleBiasClamp {bInterpResult=true, interpSpeedDecreasing=1}
    boneToModify {boneName=dyn_skirt_fr_02_l}
    constraintSetup [{targetBone {boneName=thigh_l}}]
    constraintWeights [1]
   }
   n37 = "Transform (Modify) Bone - Bone: dyn_skirt_fr_02_l"(ComponentPose=@n36, Alpha=1, Translation="6,0,0", Rotation="0,0,0", Scale="1,1,1") {
    translationMode=BMM_Additive
    translationSpace=BCS_BoneSpace
    alphaBoolBlend {blendOption=Linear}
    alphaScaleBiasClamp {bInterpResult=true}
    boneToModify {boneName=dyn_skirt_fr_02_l}
    translation {x=6}
   }
   n38 = "Constraint - dyn_skirt_fr2_01_l (P)"(ComponentPose=@n37, Alpha=1, ConstraintWeights_0=1) {
    alphaBoolBlend {blendOption=Linear}
    alphaScaleBiasClamp {bInterpResult=true, interpSpeedDecreasing=1}
    boneToModify {boneName=dyn_skirt_fr2_01_l}
    constraintSetup [{targetBone {boneName=thigh_l}}]
    constraintWeights [1]
   }
   n39 = "Transform (Modify) Bone - Bone: dyn_skirt_fr2_01_l"(ComponentPose=@n38, Alpha=1, Translation="6,0,0", Rotation="0,0,0", Scale="1,1,1") {
    translationMode=BMM_Additive
    translationSpace=BCS_BoneSpace
    alphaBoolBlend {blendOption=Linear}
    alphaScaleBiasClamp {bInterpResult=true}
    boneToModify {boneName=dyn_skirt_fr2_01_l}
    translation {x=6}
   }
   n40 = "Constraint - dyn_skirt_fr2_02_l (P)"(ComponentPose=@n39, Alpha=1, ConstraintWeights_0=1) {
    alphaBoolBlend {blendOption=Linear}
    alphaScaleBiasClamp {bInterpResult=true, interpSpeedDecreasing=1}
    boneToModify {boneName=dyn_skirt_fr2_02_l}
    constraintSetup [{targetBone {boneName=thigh_l}}]
    constraintWeights [1]
   }
   n41 = "Transform (Modify) Bone - Bone: dyn_skirt_fr2_02_l"(ComponentPose=@n40, Alpha=1, Translation="6,0,0", Rotation="0,0,0", Scale="1,1,1") {
    translationMode=BMM_Additive
    translationSpace=BCS_BoneSpace
    alphaBoolBlend {blendOption=Linear}
    alphaScaleBiasClamp {bInterpResult=true}
    boneToModify {boneName=dyn_skirt_fr2_02_l}
    translation {x=6}
   }
   n42 = "Constraint - dyn_skirt_fr_01_r (P)"(ComponentPose=@n41, Alpha=1, ConstraintWeights_0=1) {
    alphaBoolBlend {blendOption=Linear}
    alphaScaleBiasClamp {bInterpResult=true, interpSpeedDecreasing=1}
    boneToModify {boneName=dyn_skirt_fr_01_r}
    constraintSetup [{targetBone {boneName=thigh_r}}]
    constraintWeights [1]
   }
   n43 = "Transform (Modify) Bone - Bone: dyn_skirt_fr_01_r"(ComponentPose=@n42, Alpha=1, Translation="6,0,0", Rotation="0,0,0", Scale="1,1,1") {
    translationMode=BMM_Additive
    translationSpace=BCS_BoneSpace
    alphaBoolBlend {blendOption=Linear}
    alphaScaleBiasClamp {bInterpResult=true}
    boneToModify {boneName=dyn_skirt_fr_01_r}
    translation {x=6}
   }
   n44 = "Constraint - dyn_skirt_fr_02_r (P)"(ComponentPose=@n43, Alpha=1, ConstraintWeights_0=1) {
    alphaBoolBlend {blendOption=Linear}
    alphaScaleBiasClamp {bInterpResult=true, interpSpeedDecreasing=1}
    boneToModify {boneName=dyn_skirt_fr_02_r}
    constraintSetup [{targetBone {boneName=thigh_r}}]
    constraintWeights [1]
   }
   n45 = "Transform (Modify) Bone - Bone: dyn_skirt_fr_02_r"(ComponentPose=@n44, Alpha=1, Translation="6,0,0", Rotation="0,0,0", Scale="1,1,1") {
    translationMode=BMM_Additive
    translationSpace=BCS_BoneSpace
    alphaBoolBlend {blendOption=Linear}
    alphaScaleBiasClamp {bInterpResult=true}
    boneToModify {boneName=dyn_skirt_fr_02_r}
    translation {x=6}
   }
   n46 = "Constraint - dyn_skirt_fr2_01_r (P)"(ComponentPose=@n45, Alpha=1, ConstraintWeights_0=1) {
    alphaBoolBlend {blendOption=Linear}
    alphaScaleBiasClamp {bInterpResult=true, interpSpeedDecreasing=1}
    boneToModify {boneName=dyn_skirt_fr2_01_r}
    constraintSetup [{targetBone {boneName=thigh_r}}]
    constraintWeights [1]
   }
   n47 = "Transform (Modify) Bone - Bone: dyn_skirt_fr2_01_r"(ComponentPose=@n46, Alpha=1, Translation="6,0,0", Rotation="0,0,0", Scale="1,1,1") {
    translationMode=BMM_Additive
    translationSpace=BCS_BoneSpace
    alphaBoolBlend {blendOption=Linear}
    alphaScaleBiasClamp {bInterpResult=true}
    boneToModify {boneName=dyn_skirt_fr2_01_r}
    translation {x=6}
   }
   n48 = "Constraint - dyn_skirt_fr2_02_r (P)"(ComponentPose=@n47, Alpha=1, ConstraintWeights_0=1) {
    alphaBoolBlend {blendOption=Linear}
    alphaScaleBiasClamp {bInterpResult=true, interpSpeedDecreasing=1}
    boneToModify {boneName=dyn_skirt_fr2_02_r}
    constraintSetup [{targetBone {boneName=thigh_r}}]
    constraintWeights [1]
   }
   n49 = "Transform (Modify) Bone - Bone: dyn_skirt_fr2_02_r"(ComponentPose=@n48, Alpha=1, Translation="6,0,0", Rotation="0,0,0", Scale="1,1,1") {
    translationMode=BMM_Additive
    translationSpace=BCS_BoneSpace
    alphaBoolBlend {blendOption=Linear}
    alphaScaleBiasClamp {bInterpResult=true}
    boneToModify {boneName=dyn_skirt_fr2_02_r}
    translation {x=6}
   }
   n50 = "Constraint - dyn_skirt_01_r (P)"(ComponentPose=@n49, Alpha=1, ConstraintWeights_0=1) {
    alphaBoolBlend {blendOption=Linear}
    alphaScaleBiasClamp {bInterpResult=true}
    boneToModify {boneName=dyn_skirt_01_r}
    constraintSetup [{targetBone {boneName=thigh_r}}]
    constraintWeights [1]
   }
   n51 = "Transform (Modify) Bone - Bone: dyn_skirt_01_r"(ComponentPose=@n50, Alpha=1, Translation="0,0,0", Rotation="0,0,0", Scale="1,1,1") {
    translationMode=BMM_Additive
    translationSpace=BCS_BoneSpace
    alphaBoolBlend {blendOption=Linear}
    alphaScaleBiasClamp {bInterpResult=true}
    boneToModify {boneName=dyn_skirt_01_r}
   }
   n52 = "Constraint - dyn_skirt_01_l (P)"(ComponentPose=@n51, Alpha=1, ConstraintWeights_0=1) {
    alphaBoolBlend {blendOption=Linear}
    alphaScaleBiasClamp {bInterpResult=true}
    boneToModify {boneName=dyn_skirt_01_l}
    constraintSetup [{targetBone {boneName=thigh_l}}]
    constraintWeights [1]
   }
   n53 = "Transform (Modify) Bone - Bone: dyn_skirt_01_l"(ComponentPose=@n52, Alpha=1, Translation="0,0,0", Rotation="0,0,0", Scale="1,1,1") {
    translationMode=BMM_Additive
    translationSpace=BCS_BoneSpace
    alphaBoolBlend {blendOption=Linear}
    alphaScaleBiasClamp {bInterpResult=true}
    boneToModify {boneName=dyn_skirt_01_l}
   }
   n54 = "Constraint - dyn_skirt_bk_03 (PP)"!(ComponentPose=@n53, Alpha=1, ConstraintWeights_0=0.5, ConstraintWeights_1:float=0.5) -> (Pose) {
    alphaBoolBlend {blendOption=Linear}
    boneToModify {boneName=dyn_skirt_bk_03}
    constraintSetup [{targetBone {boneName=dyn_skirt_bk_03_l}}, {targetBone {boneName=dyn_skirt_bk_03_r}}]
    constraintWeights [1,1]
   }
   n55 = "Transform (Modify) Bone - Bone: dyn_skirt_bk_01"(ComponentPose=@n54, Alpha=1, Translation="0,0,1", Rotation="0,0,0", Scale="1,1,1") {
    translationMode=BMM_Additive
    translationSpace=BCS_BoneSpace
    alphaBoolBlend {blendOption=Linear}
    alphaScaleBiasClamp {bInterpResult=true, interpSpeedDecreasing=5, interpSpeedIncreasing=5}
    boneToModify {boneName=dyn_skirt_bk_01}
    translation {z=1}
   }
   n56 = "Transform (Modify) Bone - Bone: dyn_skirt_bk_01_l"(ComponentPose=@n55, Alpha=1, Translation="0,0,1", Rotation="0,0,0", Scale="1,1,1") {
    translationMode=BMM_Additive
    translationSpace=BCS_BoneSpace
    alphaBoolBlend {blendOption=Linear}
    alphaScaleBiasClamp {bInterpResult=true, interpSpeedDecreasing=5, interpSpeedIncreasing=5}
    boneToModify {boneName=dyn_skirt_bk_01_l}
    translation {z=1}
   }
   n57 = "Transform (Modify) Bone - Bone: dyn_skirt_bk_01_r"(ComponentPose=@n56, Alpha=1, Translation="0,0,1", Rotation="0,0,0", Scale="1,1,1") {
    translationMode=BMM_Additive
    translationSpace=BCS_BoneSpace
    alphaBoolBlend {blendOption=Linear}
    alphaScaleBiasClamp {bInterpResult=true, interpSpeedDecreasing=5, interpSpeedIncreasing=5}
    boneToModify {boneName=dyn_skirt_bk_01_r}
    translation {z=1}
   }
   n58 = "Transform (Modify) Bone - Bone: dyn_skirt_bk_02"(ComponentPose=@n57, Alpha=1, Translation="0,0,2", Rotation="0,0,0", Scale="1,1,1") {
    translationMode=BMM_Additive
    translationSpace=BCS_BoneSpace
    alphaBoolBlend {blendOption=Linear}
    alphaScaleBiasClamp {bInterpResult=true, interpSpeedDecreasing=5, interpSpeedIncreasing=5}
    boneToModify {boneName=dyn_skirt_bk_02}
    translation {z=2}
   }
   n59 = "Transform (Modify) Bone - Bone: dyn_skirt_bk_02_l"(ComponentPose=@n58, Alpha=1, Translation="0,0,2", Rotation="0,0,0", Scale="1,1,1") {
    translationMode=BMM_Additive
    translationSpace=BCS_BoneSpace
    alphaBoolBlend {blendOption=Linear}
    alphaScaleBiasClamp {bInterpResult=true, interpSpeedDecreasing=5, interpSpeedIncreasing=5}
    boneToModify {boneName=dyn_skirt_bk_02_l}
    translation {z=2}
   }
   n60 = "Transform (Modify) Bone - Bone: dyn_skirt_bk_02_r"(ComponentPose=@n59, Alpha=1, Translation="0,0,2", Rotation="0,0,0", Scale="1,1,1") {
    translationMode=BMM_Additive
    translationSpace=BCS_BoneSpace
    alphaBoolBlend {blendOption=Linear}
    alphaScaleBiasClamp {bInterpResult=true, interpSpeedDecreasing=5, interpSpeedIncreasing=5}
    boneToModify {boneName=dyn_skirt_bk_02_r}
    translation {z=2}
   }
   n61 = "Constraint - dyn_skirt_fr_01 (PP)"!(ComponentPose=@n60, Alpha=1, ConstraintWeights_0=0.5, ConstraintWeights_1:float=0.5) -> (Pose) {
    alphaBoolBlend {blendOption=Linear}
    alphaScaleBiasClamp {bInterpResult=true}
    boneToModify {boneName=dyn_skirt_fr_01}
    constraintSetup [{targetBone {boneName=dyn_skirt_fr_01_l}}, {targetBone {boneName=dyn_skirt_fr_01_r}}]
    constraintWeights [1,1]
   }
   n62 = "Constraint - dyn_skirt_fr_02 (PP)"!(ComponentPose=@n61, Alpha=1, ConstraintWeights_0=0.5, ConstraintWeights_1:float=0.5) -> (Pose) {
    alphaBoolBlend {blendOption=Linear}
    alphaScaleBiasClamp {bInterpResult=true}
    boneToModify {boneName=dyn_skirt_fr_02}
    constraintSetup [{targetBone {boneName=dyn_skirt_fr_02_l}}, {targetBone {boneName=dyn_skirt_fr_02_r}}]
    constraintWeights [1,1]
   }
   n63 = "Constraint - dyn_skirt_fr_03 (PP)"!(ComponentPose=@n62, Alpha=1, ConstraintWeights_0=0.5, ConstraintWeights_1:float=0.5) -> (Pose) {
    alphaBoolBlend {blendOption=Linear}
    alphaScaleBiasClamp {bInterpResult=true}
    boneToModify {boneName=dyn_skirt_fr_03}
    constraintSetup [{targetBone {boneName=dyn_skirt_fr_03_l}}, {targetBone {boneName=dyn_skirt_fr_03_r}}]
    constraintWeights [1,1]
   }
   n64 = "Constraint - dyn_strap_fr_01_r (PP)"!(ComponentPose=@n63, Alpha=1, ConstraintWeights_0=1, ConstraintWeights_1:float=0) -> (Pose) {
    alphaBoolBlend {blendOption=Linear}
    alphaScaleBiasClamp {bInterpResult=true}
    boneToModify {boneName=dyn_strap_fr_01_r}
    constraintSetup [{targetBone {boneName=spine_01}}, {targetBone {boneName=None}}]
    constraintWeights [1,1]
   }
   n65 = "Constraint - dyn_strap_fr_02_r (PP)"!(ComponentPose=@n64, Alpha=1, ConstraintWeights_0=0.5, ConstraintWeights_1:float=0.5) -> (Pose) {
    alphaBoolBlend {blendOption=Linear}
    alphaScaleBiasClamp {bInterpResult=true}
    boneToModify {boneName=dyn_strap_fr_02_r}
    constraintSetup [{targetBone {boneName=dyn_skirt_fr2_01_r}}, {targetBone {boneName=dyn_skirt_fr_01}}]
    constraintWeights [1,1]
   }
   n66 = "Transform (Modify) Bone - Bone: dyn_strap_fr_02_r"(ComponentPose=@n65, Alpha=1, Translation="10,0,0", Rotation="0,0,0", Scale="1,1,1") {
    translationMode=BMM_Additive
    translationSpace=BCS_BoneSpace
    alphaBoolBlend {blendOption=Linear}
    alphaScaleBiasClamp {bInterpResult=true}
    boneToModify {boneName=dyn_strap_fr_02_r}
    translation {x=10}
   }
   n67 = "Constraint - dyn_strap_fr_03_r (PP)"!(ComponentPose=@n66, Alpha=1, ConstraintWeights_0=0.5, ConstraintWeights_1:float=0.5) -> (Pose) {
    alphaBoolBlend {blendOption=Linear}
    alphaScaleBiasClamp {bInterpResult=true}
    boneToModify {boneName=dyn_strap_fr_03_r}
    constraintSetup [{targetBone {boneName=dyn_skirt_fr2_02_r}}, {targetBone {boneName=dyn_skirt_fr_02}}]
    constraintWeights [1,1]
   }
   n68 = "Transform (Modify) Bone - Bone: dyn_strap_fr_03_r"(ComponentPose=@n67, Alpha=1, Translation="10,0,0", Rotation="0,0,0", Scale="1,1,1") {
    translationMode=BMM_Additive
    translationSpace=BCS_BoneSpace
    alphaBoolBlend {blendOption=Linear}
    alphaScaleBiasClamp {bInterpResult=true}
    boneToModify {boneName=dyn_strap_fr_03_r}
    translation {x=10}
   }
   n69 = "Constraint - dyn_strap_fr_04_r (PP)"!(ComponentPose=@n68, Alpha=1, ConstraintWeights_0=0.5, ConstraintWeights_1:float=0.5) -> (Pose) {
    alphaBoolBlend {blendOption=Linear}
    alphaScaleBiasClamp {bInterpResult=true}
    boneToModify {boneName=dyn_strap_fr_04_r}
    constraintSetup [{targetBone {boneName=dyn_skirt_fr2_03_r}}, {targetBone {boneName=dyn_skirt_fr_03}}]
    constraintWeights [1,1]
   }
   n70 = "Transform (Modify) Bone - Bone: dyn_strap_fr_04_r"(ComponentPose=@n69, Alpha=1, Translation="10,0,0", Rotation="0,0,0", Scale="1,1,1") {
    translationMode=BMM_Additive
    translationSpace=BCS_BoneSpace
    alphaBoolBlend {blendOption=Linear}
    alphaScaleBiasClamp {bInterpResult=true}
    boneToModify {boneName=dyn_strap_fr_04_r}
    translation {x=10}
   }
   n71 = "Constraint - dyn_strap_fr_01_l (PP)"!(ComponentPose=@n70, Alpha=1, ConstraintWeights_0=1, ConstraintWeights_1:float=0) -> (Pose) {
    alphaBoolBlend {blendOption=Linear}
    alphaScaleBiasClamp {bInterpResult=true}
    boneToModify {boneName=dyn_strap_fr_01_l}
    constraintSetup [{targetBone {boneName=spine_01}}, {targetBone {boneName=None}}]
    constraintWeights [1,1]
   }
   n72 = "Constraint - dyn_strap_fr_02_l (PP)"!(ComponentPose=@n71, Alpha=1, ConstraintWeights_0=0.3, ConstraintWeights_1:float=0.7) -> (Pose) {
    alphaBoolBlend {blendOption=Linear}
    alphaScaleBiasClamp {bInterpResult=true}
    boneToModify {boneName=dyn_strap_fr_02_l}
    constraintSetup [{targetBone {boneName=dyn_skirt_fr2_01_l}}, {targetBone {boneName=dyn_skirt_fr_01}}]
    constraintWeights [1,1]
   }
   n73 = "Transform (Modify) Bone - Bone: dyn_strap_fr_02_l"(ComponentPose=@n72, Alpha=1, Translation="15,10,0", Rotation="0,0,90", Scale="1,1,1") {
    rotationMode=BMM_Additive
    rotationSpace=BCS_BoneSpace
    translationMode=BMM_Additive
    translationSpace=BCS_BoneSpace
    alphaBoolBlend {blendOption=Linear}
    alphaScaleBiasClamp {bInterpResult=true}
    boneToModify {boneName=dyn_strap_fr_02_l}
    rotation {roll=90}
    translation=<15 10>
   }
   n74 = "Constraint - dyn_strap_fr_03_l (PP)"!(ComponentPose=@n73, Alpha=1, ConstraintWeights_0=0.3, ConstraintWeights_1:float=0.7) -> (Pose) {
    alphaBoolBlend {blendOption=Linear}
    alphaScaleBiasClamp {bInterpResult=true}
    boneToModify {boneName=dyn_strap_fr_03_l}
    constraintSetup [{targetBone {boneName=dyn_skirt_fr2_02_l}}, {targetBone {boneName=dyn_skirt_fr_02}}]
    constraintWeights [1,1]
   }
   n75 = "Transform (Modify) Bone - Bone: dyn_strap_fr_03_l"(ComponentPose=@n74, Alpha=1, Translation="15,10,0", Rotation="0,0,45", Scale="1,1,1") {
    rotationMode=BMM_Additive
    rotationSpace=BCS_BoneSpace
    translationMode=BMM_Additive
    translationSpace=BCS_BoneSpace
    alphaBoolBlend {blendOption=Linear}
    alphaScaleBiasClamp {bInterpResult=true}
    boneToModify {boneName=dyn_strap_fr_03_l}
    rotation {roll=45}
    translation=<15 10>
   }
   n76 = "Constraint - dyn_strap_fr_04_l (PP)"!(ComponentPose=@n75, Alpha=1, ConstraintWeights_0=0.3, ConstraintWeights_1:float=0.7) -> (Pose) {
    alphaBoolBlend {blendOption=Linear}
    alphaScaleBiasClamp {bInterpResult=true}
    boneToModify {boneName=dyn_strap_fr_04_l}
    constraintSetup [{targetBone {boneName=dyn_skirt_fr2_03_l}}, {targetBone {boneName=dyn_skirt_fr_03}}]
    constraintWeights [1,1]
   }
   n77 = "Transform (Modify) Bone - Bone: dyn_strap_fr_04_l"(ComponentPose=@n76, Alpha=1, Translation="15,10,0", Rotation="0,0,45", Scale="1,1,1") {
    rotationMode=BMM_Additive
    rotationSpace=BCS_BoneSpace
    translationMode=BMM_Additive
    translationSpace=BCS_BoneSpace
    alphaBoolBlend {blendOption=Linear}
    alphaScaleBiasClamp {bInterpResult=true}
    boneToModify {boneName=dyn_strap_fr_04_l}
    rotation {roll=45}
    translation=<15 10>
   }
   n78 = "Transform (Modify) Bone - Bone: dyn_skirt_fr_01_l_MSTR"(
    ComponentPose=@n77
    Alpha=Get isCrouchingAlpha() -> (isCrouchingAlpha:double)
    Translation="4.104,0.723,0.000"
    Rotation="0.000,10.000,0.000"
    Scale="1,1,1"
   ) {
    rotationMode=BMM_Additive
    rotationSpace=BCS_BoneSpace
    translationMode=BMM_Additive
    translationSpace=BCS_BoneSpace
    alphaBoolBlend {blendOption=Linear}
    boneToModify {boneName=dyn_skirt_fr_01_l_MSTR}
    rotation {yaw=10.0002}
    translation=<4.104 0.723>
   }
   n79 = "Transform (Modify) Bone - Bone: dyn_skirt_bk_03_r"(ComponentPose=@n78, Alpha=1, Translation="0.000,-0.008,2.7", Rotation="0,0,0", Scale="1,1,1") {
    translationMode=BMM_Additive
    translationSpace=BCS_BoneSpace
    alphaBoolBlend {blendOption=Linear}
    boneToModify {boneName=dyn_skirt_bk_03_r}
    translation {y=-0.008, z=2.7}
   }
   n80 = "Transform (Modify) Bone - Bone: dyn_skirt_03_r"(ComponentPose=@n79, Alpha=1, Translation="0,0,2.7", Rotation="0,0,0", Scale="1,1,1") {
    translationMode=BMM_Additive
    translationSpace=BCS_BoneSpace
    alphaBoolBlend {blendOption=Linear}
    boneToModify {boneName=dyn_skirt_03_r}
    translation {z=2.7}
   }
   n81 = "Transform (Modify) Bone - Bone: dyn_skirt_bk_03_l"(ComponentPose=@n80, Alpha=1, Translation="0.000,-0.008,2.7", Rotation="0,0,0", Scale="1,1,1") {
    translationMode=BMM_Additive
    translationSpace=BCS_BoneSpace
    alphaBoolBlend {blendOption=Linear}
    boneToModify {boneName=dyn_skirt_bk_03_l}
    translation {y=-0.008, z=2.7}
   }
   "Transform (Modify) Bone - Bone: dyn_skirt_03_l"(ComponentPose=@n81, Alpha=1, Translation="0,0,2.7", Rotation="0,0,0", Scale="1,1,1") {
    translationMode=BMM_Additive
    translationSpace=BCS_BoneSpace
    alphaBoolBlend {blendOption=Linear}
    boneToModify {boneName=dyn_skirt_03_l}
    translation {z=2.7}
   }
   n82 = "Transform (Modify) Bone - Bone: dyn_skirt_fr_03"(Alpha=1, Translation="0,0,0", Rotation="45,0,0", Scale="1,1,1") {
    rotationMode=BMM_Additive
    rotationSpace=BCS_BoneSpace
    translationMode=BMM_Additive
    translationSpace=BCS_BoneSpace
    alphaBoolBlend {blendOption=Linear}
    boneToModify {boneName=dyn_skirt_fr_03}
    rotation {pitch=45}
   }
   "Transform (Modify) Bone - Bone: dyn_skirt_fr_03"(ComponentPose=@n82, Alpha=1, Translation="0,0,0", Rotation="-45,0,0", Scale="1,1,1") {
    rotationMode=BMM_Additive
    rotationSpace=BCS_BoneSpace
    translationMode=BMM_Additive
    translationSpace=BCS_BoneSpace
    alphaBoolBlend {blendOption=Linear}
    boneToModify {boneName=dyn_skirt_fr_03}
    rotation {pitch=-45}
   }
   n83 = "Constraint - dyn_skirt_fr2_03_r (P)"(Alpha=1, ConstraintWeights_0=1) {
    alphaBoolBlend {blendOption=Linear}
    alphaScaleBiasClamp {bInterpResult=true, interpSpeedDecreasing=1}
    boneToModify {boneName=dyn_skirt_fr2_03_r}
    constraintSetup [{targetBone {boneName=thigh_r}}]
    constraintWeights [1]
   }
   "Transform (Modify) Bone - Bone: dyn_skirt_fr2_03_r"(ComponentPose=@n83, Alpha=1, Translation="6,0,0", Rotation="0,0,0", Scale="1,1,1") {
    translationMode=BMM_Additive
    translationSpace=BCS_BoneSpace
    alphaBoolBlend {blendOption=Linear}
    alphaScaleBiasClamp {bInterpResult=true}
    boneToModify {boneName=dyn_skirt_fr2_03_r}
    translation {x=6}
   }
   n84 = Get finalGravity()
   n85 = RigidBody(ComponentPose=Local To Component(LocalPose=Use cached pose 'PostDressCache'()), OverrideWorldGravity=@n84) {
    bClampLinearTranslationLimitToRefPose=true
    bOverrideWorldGravity=true
    overridePhysicsAsset=&F_MED_Hardcore_Sportz_Bow_Physics1
    simulationSpace=BaseBoneSpace
    alphaBoolBlend {blendOption=Linear}
    baseBoneRef {boneName=spine_01}
    componentAppliedLinearAccClamp=<3000 3000 3000>
    componentLinearAccScale=<0.3 0.3 0.5>
    componentLinearVelScale=<0.3 0.3 0.5>
    overrideWorldGravity {z=-980}
   }
   n86 = RigidBody(ComponentPose=@n85, OverrideWorldGravity=@n84) {
    bClampLinearTranslationLimitToRefPose=true
    bOverrideWorldGravity=true
    overridePhysicsAsset=&F_MED_Hardcore_Sportz_UpperBody_Physics
    simulationSpace=BaseBoneSpace
    alphaBoolBlend {blendOption=Linear}
    baseBoneRef {boneName=spine_05}
    componentAppliedLinearAccClamp=<3000 3000 3000>
    componentLinearAccScale=<0.3 0.3 0.5>
    componentLinearVelScale=<0.3 0.3 0.5>
    overrideWorldGravity {z=-980}
   }
   n87 = "Constraint - dyn_strap_fr_01_l (RRR)"(ComponentPose=@n86, Alpha=1, ConstraintWeights_0=0.3, ConstraintWeights_1=0.7, ConstraintWeights_2=0) {
    alphaBoolBlend {blendOption=Linear}
    alphaScaleBiasClamp {bInterpResult=true}
    boneToModify {boneName=dyn_strap_fr_01_l}
    constraintSetup [
     {transformType=Rotation, targetBone {boneName=dyn_skirt_fr_01_l}}
     {transformType=Rotation, targetBone {boneName=dyn_skirt_fr_01}}
     {transformType=Rotation, targetBone {boneName=spine_02}}
    ]
    constraintWeights [1,1,1]
   }
   n88 = "Constraint - dyn_strap_fr_01_r (RRR)"(ComponentPose=@n87, Alpha=1, ConstraintWeights_0=0.7, ConstraintWeights_1=0.3, ConstraintWeights_2=0) {
    alphaBoolBlend {blendOption=Linear}
    alphaScaleBiasClamp {bInterpResult=true}
    boneToModify {boneName=dyn_strap_fr_01_r}
    constraintSetup [
     {transformType=Rotation, targetBone {boneName=dyn_skirt_fr_01_r}}
     {transformType=Rotation, targetBone {boneName=dyn_skirt_fr_01}}
     {transformType=Rotation, targetBone {boneName=spine_02}}
    ]
    constraintWeights [1,1,1]
   }
   n89 = "Constraint - dyn_strap_fr_02_l (PPRRP)"!(
    ComponentPose=@n88
    Alpha=1
    ConstraintWeights_0=0
    ConstraintWeights_1:float=0
    ConstraintWeights_2:float=0
    ConstraintWeights_3:float=0
    ConstraintWeights_4:float=1
   ) -> (Pose) {
    alphaBoolBlend {blendOption=Linear}
    alphaScaleBiasClamp {bInterpResult=true}
    boneToModify {boneName=dyn_strap_fr_02_l}
    constraintSetup [
     {targetBone {boneName=dyn_skirt_fr_01_l}}
     {targetBone {boneName=dyn_skirt_fr_01}}
     {transformType=Rotation, targetBone {boneName=thigh_l}}
     {transformType=Rotation, targetBone {boneName=thigh_r}}
     {targetBone {boneName=dyn_strap_fr_01_l}}
    ]
    constraintWeights [1,1,1,1,1]
   }
   n90 = "Constraint - dyn_strap_fr_02_r (TTRRP)"!(
    ComponentPose=@n89
    Alpha=1
    ConstraintWeights_0=0
    ConstraintWeights_1:float=0
    ConstraintWeights_2:float=0
    ConstraintWeights_3:float=0
    ConstraintWeights_4:float=1
   ) -> (Pose) {
    alphaBoolBlend {blendOption=Linear}
    alphaScaleBiasClamp {bInterpResult=true}
    boneToModify {boneName=dyn_strap_fr_02_r}
    constraintSetup [
     {transformType=Translation, targetBone {boneName=dyn_skirt_fr_01_r}}
     {transformType=Translation, targetBone {boneName=dyn_skirt_fr_01}}
     {transformType=Rotation, targetBone {boneName=thigh_r}}
     {transformType=Rotation, targetBone {boneName=thigh_l}}
     {targetBone {boneName=dyn_strap_fr_01_r}}
    ]
    constraintWeights [1,1,1,1,1]
   }
   n91 = "Transform (Modify) Bone - Bone: dyn_strap_fr_01_l"(ComponentPose=@n90, Alpha=1, Translation="0,0,0", Rotation="0,0,0", Scale="1,1,1") {
    rotationMode=BMM_Additive
    rotationSpace=BCS_BoneSpace
    translationMode=BMM_Additive
    translationSpace=BCS_BoneSpace
    alphaBoolBlend {blendOption=Linear}
    boneToModify {boneName=dyn_strap_fr_01_l}
   }
   n92 = "Transform (Modify) Bone - Bone: dyn_strap_fr_01_r"(ComponentPose=@n91, Alpha=1, Translation="0,0,0", Rotation="0,0,0", Scale="1,1,1") {
    rotationMode=BMM_Additive
    rotationSpace=BCS_BoneSpace
    translationMode=BMM_Additive
    translationSpace=BCS_BoneSpace
    alphaBoolBlend {blendOption=Linear}
    boneToModify {boneName=dyn_strap_fr_01_r}
   }
   n93 = "Transform (Modify) Bone - Bone: dyn_strap_fr_02_l"(ComponentPose=@n92, Alpha=1, Translation="0,0,0", Rotation="0,0,0", Scale="1,1,1") {
    rotationMode=BMM_Additive
    rotationSpace=BCS_BoneSpace
    translationMode=BMM_Additive
    translationSpace=BCS_BoneSpace
    alphaBoolBlend {blendOption=Linear}
    boneToModify {boneName=dyn_strap_fr_02_l}
   }
   n94 = "Transform (Modify) Bone - Bone: dyn_strap_fr_02_r"(ComponentPose=@n93, Alpha=1, Translation="0,0,0", Rotation="0,0,0", Scale="1,1,1") {
    rotationMode=BMM_Additive
    rotationSpace=BCS_BoneSpace
    translationMode=BMM_Additive
    translationSpace=BCS_BoneSpace
    alphaBoolBlend {blendOption=Linear}
    boneToModify {boneName=dyn_strap_fr_02_r}
   }
   n95 = "Transform (Modify) Bone - Bone: dyn_strap_bk_01_l"!(ComponentPose=@n94, bAlphaBoolEnabled:bool=True, Translation="0,0,0", Rotation="5,5,0", Scale="1,1,1") -> (Pose) {
    alphaInputType=Bool
    rotationMode=BMM_Additive
    rotationSpace=BCS_BoneSpace
    alphaBoolBlend {blendOption=Linear}
    boneToModify {boneName=dyn_strap_bk_01_l}
    rotation {pitch=5, yaw=5}
   }
   n96 = "Transform (Modify) Bone - Bone: dyn_strap_bk_01_r"!(ComponentPose=@n95, bAlphaBoolEnabled:bool=True, Translation="0,0,0", Rotation="5,-5,0", Scale="1,1,1") -> (Pose) {
    alphaInputType=Bool
    rotationMode=BMM_Additive
    rotationSpace=BCS_BoneSpace
    alphaBoolBlend {blendOption=Linear}
    boneToModify {boneName=dyn_strap_bk_01_r}
    rotation {pitch=5, yaw=-5}
   }
   n97 = "Transform (Modify) Bone - Bone: dyn_strap_bk_03_l"!(ComponentPose=@n96, bAlphaBoolEnabled:bool=True, Translation="0,0,0", Rotation="0,5,0", Scale="1,1,1") -> (Pose) {
    alphaInputType=Bool
    rotationMode=BMM_Additive
    rotationSpace=BCS_BoneSpace
    alphaBoolBlend {blendOption=Linear}
    boneToModify {boneName=dyn_strap_bk_03_l}
    rotation {yaw=5}
   }
   n98 = "Transform (Modify) Bone - Bone: dyn_strap_bk_03_r"!(ComponentPose=@n97, bAlphaBoolEnabled:bool=True, Translation="0,0,0", Rotation="0,-5,0", Scale="1,1,1") -> (Pose) {
    alphaInputType=Bool
    rotationMode=BMM_Additive
    rotationSpace=BCS_BoneSpace
    alphaBoolBlend {blendOption=Linear}
    boneToModify {boneName=dyn_strap_bk_03_r}
    rotation {yaw=-5}
   }
   Output Pose(Result=Component To Local(ComponentPose=@n98)) {layerGroup=None}
   n99 = "Constraint - dyn_thigh_01_l (RR)"!(
    ComponentPose=Local To Component(LocalPose=Copy Pose From Mesh() {bUseAttachedParent=true})
    Alpha=1
    ConstraintWeights_0=1
    ConstraintWeights_1:float=0
   ) -> (Pose) {
    alphaBoolBlend {blendOption=Linear}
    boneToModify {boneName=dyn_thigh_01_l}
    constraintSetup [{transformType=Rotation, targetBone {boneName=thigh_l}}, {transformType=Rotation, targetBone {boneName=pelvis}}]
    constraintWeights [1,1]
   }
   n100 = "Constraint - dyn_thigh_01_r (RR)"!(ComponentPose=@n99, Alpha=1, ConstraintWeights_0=1, ConstraintWeights_1:float=0) -> (Pose) {
    alphaBoolBlend {blendOption=Linear}
    boneToModify {boneName=dyn_thigh_01_r}
    constraintSetup [{transformType=Rotation, targetBone {boneName=thigh_r}}, {transformType=Rotation, targetBone {boneName=pelvis}}]
    constraintWeights [1,1]
   }
   n101 = "Constraint - dyn_skirt_fr_01_MSTR (PP)"!(ComponentPose=@n100, Alpha=1, ConstraintWeights_0=0.5, ConstraintWeights_1:float=0.5) -> (Pose) {
    alphaBoolBlend {blendOption=Linear}
    boneToModify {boneName=dyn_skirt_fr_01_MSTR}
    constraintSetup [{targetBone {boneName=dyn_skirt_fr_01_l_MSTR}}, {targetBone {boneName=dyn_skirt_fr_01_r_MSTR}}]
    constraintWeights [1,1]
   }
   n102 = "Constraint - dyn_skirt_fr_01 (PP)"(ComponentPose=@n101, Alpha=1, ConstraintWeights_0=0.5, ConstraintWeights_1=0.5) {
    alphaBoolBlend {blendOption=Linear}
    boneToModify {boneName=dyn_skirt_fr_01}
    constraintSetup [{targetBone {boneName=dyn_skirt_fr_01_l}}, {targetBone {boneName=dyn_skirt_fr_01_r}}]
    constraintWeights [1,1]
   }
   n103 = "Constraint - dyn_skirt_fr_02 (PP)"(ComponentPose=@n102, Alpha=1, ConstraintWeights_0=0.5, ConstraintWeights_1=0.5) {
    alphaBoolBlend {blendOption=Linear}
    boneToModify {boneName=dyn_skirt_fr_02}
    constraintSetup [{targetBone {boneName=dyn_skirt_fr_02_l}}, {targetBone {boneName=dyn_skirt_fr_02_r}}]
    constraintWeights [1,1]
   }
   n104 = "Constraint - dyn_skirt_fr_03 (PP)"(ComponentPose=@n103, Alpha=1, ConstraintWeights_0=0.5, ConstraintWeights_1=0.5) {
    alphaBoolBlend {blendOption=Linear}
    boneToModify {boneName=dyn_skirt_fr_03}
    constraintSetup [{targetBone {boneName=dyn_skirt_fr_03_l}}, {targetBone {boneName=dyn_skirt_fr_03_r}}]
    constraintWeights [1,1]
   }
   n105 = "Constraint - kneepad_01_l (RR)"!(ComponentPose=@n104, Alpha=1, ConstraintWeights_0=0.3, ConstraintWeights_1:float=0.7) -> (Pose) {
    alphaBoolBlend {blendOption=Linear}
    boneToModify {boneName=kneepad_01_l}
    constraintSetup [{transformType=Rotation, targetBone {boneName=thigh_l}}, {transformType=Rotation, targetBone {boneName=calf_l}}]
    constraintWeights [1,1]
   }
   n106 = "Transform (Modify) Bone - Bone: kneepad_01_l"(ComponentPose=@n105, Alpha=Get isCrouchingAlpha(), Translation="-2.217,0.270,-2.178", Rotation="-10.000,0.000,0.000", Scale="1,1,1") {
    rotationMode=BMM_Additive
    rotationSpace=BCS_BoneSpace
    translationMode=BMM_Additive
    translationSpace=BCS_BoneSpace
    alphaBoolBlend {blendOption=Linear}
    alphaScaleBiasClamp {bInterpResult=true, interpSpeedDecreasing=20, interpSpeedIncreasing=20}
    boneToModify {boneName=kneepad_01_l}
    rotation {pitch=-10}
    translation=<-2.217 0.27 -2.178>
   }
   n107 = Flap Node - Bone: dyn_shoe_01_l!(ComponentPose:Component Space Pose Link Structure=@n106, Alpha:float=1) -> (Pose:Component Space Pose Link Structure) {
    is=title FortAnimGraphNode_Flap
    bUseClamp=true
    targetClampMax=1
    targetComponent=RotationZ
    alphaBoolBlend {blendOption=Linear}
    sourceDrives [{
      bUseRange=false
      drivingCurve=null
      multiplier=0.8
      rangeMax=1
      rangeMin=-1
      remappedMax=1
      remappedMin=0
      sourceComponent=RotationZ
      useQuaternion=true
      sourceBone {boneName=foot_l}
     }]
    targetBone {boneName=dyn_shoe_01_l}
   }
   n108 = Flap Node - Bone: dyn_shoe_01_r!(ComponentPose:Component Space Pose Link Structure=@n107, Alpha:float=1) -> (Pose:Component Space Pose Link Structure) {
    is=title FortAnimGraphNode_Flap
    bUseClamp=true
    targetClampMax=1
    targetComponent=RotationZ
    alphaBoolBlend {blendOption=Linear}
    sourceDrives [{
      bUseRange=false
      drivingCurve=null
      multiplier=0.8
      rangeMax=1
      rangeMin=-1
      remappedMax=1
      remappedMin=0
      sourceComponent=RotationZ
      useQuaternion=true
      sourceBone {boneName=foot_r}
     }]
    targetBone {boneName=dyn_shoe_01_r}
   }
   n109 = Anim Dynamics - Bone: dyn_collar_01(ComponentPose=@n108, Alpha=1, ExternalForce="0,0,0") {
    angularDampingOverride=1
    angularSpringConstant=20
    bAngularSpring=true
    bGravityOverrideInSimSpace=true
    gravityScale=0.97
    simulationSpace=RootRelative
    alphaBoolBlend {blendOption=Linear}
    boundBone {boneName=dyn_collar_01}
    componentLinearAccScale=<1 1 1>
    physicsBodyDefinitions [{
      sphereCollisionRadius=0
      boundBone {boneName=dyn_collar_01}
      boxExtents=<10 20 2>
      constraintSetup {angularLimitsMax=<0 5 0>, angularLimitsMin=<0 -60 0>, angularTarget=<0 0 0>}
      localJointOffset=<2 0 0>
     }]
   }
   Save cached pose 'PreDressRBAN'(Pose=Component To Local(ComponentPose=@n109)) {cacheName=PreDressRBAN}
   n110 = "Constraint - dyn_thigh_01_l (RR)"(Alpha=1, ConstraintWeights_0=1, ConstraintWeights_1=0) {
    alphaBoolBlend {blendOption=Linear}
    boneToModify {boneName=dyn_thigh_01_l}
    constraintSetup [{transformType=Rotation, targetBone {boneName=thigh_l}}, {transformType=Rotation, targetBone {boneName=pelvis}}]
    constraintWeights [1,1]
   }
   n111 = "Constraint - dyn_thigh_01_r (RR)"(ComponentPose=@n110, Alpha=1, ConstraintWeights_0=1, ConstraintWeights_1=0) {
    alphaBoolBlend {blendOption=Linear}
    boneToModify {boneName=dyn_thigh_01_r}
    constraintSetup [{transformType=Rotation, targetBone {boneName=thigh_r}}, {transformType=Rotation, targetBone {boneName=pelvis}}]
    constraintWeights [1,1]
   }
   n112 = "Constraint - dyn_skirt_fr_01_MSTR (RR)"!(ComponentPose=@n111, Alpha=1, ConstraintWeights_0=0.5, ConstraintWeights_1:float=0.5) -> (Pose) {
    alphaBoolBlend {blendOption=Linear}
    boneToModify {boneName=dyn_skirt_fr_01_MSTR}
    constraintSetup [
     {transformType=Rotation, targetBone {boneName=dyn_skirt_fr_01_l_MSTR}}
     {transformType=Rotation, targetBone {boneName=dyn_skirt_fr_01_r_MSTR}}
    ]
    constraintWeights [1,1]
   }
   n113 = "Constraint - kneepad_01_l (RR)"(ComponentPose=@n112, Alpha=1, ConstraintWeights_0=0.3, ConstraintWeights_1=0.7) {
    alphaBoolBlend {blendOption=Linear}
    boneToModify {boneName=kneepad_01_l}
    constraintSetup [{transformType=Rotation, targetBone {boneName=thigh_l}}, {transformType=Rotation, targetBone {boneName=calf_l}}]
    constraintWeights [1,1]
   }
   n114 = "Transform (Modify) Bone - Bone: kneepad_01_l"(ComponentPose=@n113, Alpha=Get isCrouchingAlpha(), Translation="-2.217,0.270,-2.178", Rotation="-10.000,0.000,0.000", Scale="1,1,1") {
    rotationMode=BMM_Additive
    rotationSpace=BCS_BoneSpace
    translationMode=BMM_Additive
    translationSpace=BCS_BoneSpace
    alphaBoolBlend {blendOption=Linear}
    alphaScaleBiasClamp {bInterpResult=true, interpSpeedDecreasing=20, interpSpeedIncreasing=20}
    boneToModify {boneName=kneepad_01_l}
    rotation {pitch=-10}
    translation=<-2.217 0.27 -2.178>
   }
   n115 = Flap Node - Bone: dyn_shoe_01_l(ComponentPose=@n114, Alpha=1) {
    bUseClamp=true
    targetClampMax=1
    targetComponent=RotationZ
    alphaBoolBlend {blendOption=Linear}
    sourceDrives [{
      bUseRange=false
      drivingCurve=null
      multiplier=0.8
      rangeMax=1
      rangeMin=-1
      remappedMax=1
      remappedMin=0
      sourceComponent=RotationZ
      useQuaternion=true
      sourceBone {boneName=foot_l}
     }]
    targetBone {boneName=dyn_shoe_01_l}
   }
   n116 = Flap Node - Bone: dyn_shoe_01_r(ComponentPose=@n115, Alpha=1) {
    bUseClamp=true
    targetClampMax=1
    targetComponent=RotationZ
    alphaBoolBlend {blendOption=Linear}
    sourceDrives [{
      bUseRange=false
      drivingCurve=null
      multiplier=0.8
      rangeMax=1
      rangeMin=-1
      remappedMax=1
      remappedMin=0
      sourceComponent=RotationZ
      useQuaternion=true
      sourceBone {boneName=foot_r}
     }]
    targetBone {boneName=dyn_shoe_01_r}
   }
   n117 = Anim Dynamics - Bone: dyn_collar_01(ComponentPose=@n116, Alpha=1, ExternalForce="0,0,0") {
    angularDampingOverride=1
    angularSpringConstant=20
    bAngularSpring=true
    bGravityOverrideInSimSpace=true
    gravityScale=0.97
    simulationSpace=RootRelative
    alphaBoolBlend {blendOption=Linear}
    boundBone {boneName=dyn_collar_01}
    componentLinearAccScale=<1 1 1>
    physicsBodyDefinitions [{
      sphereCollisionRadius=0
      boundBone {boneName=dyn_collar_01}
      boxExtents=<10 20 2>
      constraintSetup {angularLimitsMax=<0 5 0>, angularLimitsMin=<0 -60 0>, angularTarget=<0 0 0>}
      localJointOffset=<2 0 0>
     }]
   }
   Save cached pose 'presimCache'(Pose=Component To Local(ComponentPose=@n117)) {cacheName=presimCache}
   n118 = "Constraint - dyn_skirt_bk_01 (RPP)"!(
    ComponentPose=Local To Component(LocalPose=Use cached pose 'PreDressRBAN'())
    Alpha=1
    ConstraintWeights_0=0
    ConstraintWeights_1:float=0.5
    ConstraintWeights_2:float=0.5
   ) -> (Pose) {
    alphaBoolBlend {blendOption=Linear}
    boneToModify {boneName=dyn_skirt_bk_01}
    constraintSetup [{transformType=Rotation, targetBone {boneName=pelvis}}, {targetBone {boneName=dyn_skirt_bk_01_l}}, {targetBone {boneName=dyn_skirt_bk_01_r}}]
    constraintWeights [1,1,1]
   }
   n119 = "Constraint - dyn_skirt_bk_02 (RPP)"!(ComponentPose=@n118, Alpha=1, ConstraintWeights_0=0, ConstraintWeights_1:float=0.5, ConstraintWeights_2:float=0.5) -> (Pose) {
    alphaBoolBlend {blendOption=Linear}
    boneToModify {boneName=dyn_skirt_bk_02}
    constraintSetup [{transformType=Rotation, targetBone {boneName=pelvis}}, {targetBone {boneName=dyn_skirt_bk_02_l}}, {targetBone {boneName=dyn_skirt_bk_02_r}}]
    constraintWeights [1,1,1]
   }
   n120 = "Constraint - dyn_skirt_bk_03 (RPP)"!(ComponentPose=@n119, Alpha=1, ConstraintWeights_0=0, ConstraintWeights_1:float=0.5, ConstraintWeights_2:float=0.5) -> (Pose) {
    alphaBoolBlend {blendOption=Linear}
    boneToModify {boneName=dyn_skirt_bk_03}
    constraintSetup [{transformType=Rotation, targetBone {boneName=pelvis}}, {targetBone {boneName=dyn_skirt_bk_03_l}}, {targetBone {boneName=dyn_skirt_bk_03_r}}]
    constraintWeights [1,1,1]
   }
   n121 = "Constraint - dyn_skirt_bk2_01_l (P)"(ComponentPose=@n120, Alpha=1, ConstraintWeights_0=1) {
    alphaBoolBlend {blendOption=Linear}
    boneToModify {boneName=dyn_skirt_bk2_01_l}
    constraintSetup [{targetBone {boneName=pelvis}}]
    constraintWeights [1]
   }
   n122 = "Constraint - dyn_skirt_bk2_02_l (P)"(ComponentPose=@n121, Alpha=0.3, ConstraintWeights_0=1) {
    alphaBoolBlend {blendOption=Linear}
    boneToModify {boneName=dyn_skirt_bk2_02_l}
    constraintSetup [{targetBone {boneName=pelvis}}]
    constraintWeights [1]
   }
   n123 = "Constraint - dyn_skirt_bk2_03_l (P)"(ComponentPose=@n122, Alpha=1, ConstraintWeights_0=1) {
    alphaBoolBlend {blendOption=Linear}
    boneToModify {boneName=dyn_skirt_bk2_03_l}
    constraintSetup [{targetBone {boneName=pelvis}}]
    constraintWeights [1]
   }
   n124 = "Constraint - dyn_skirt_bk_01_l (RRR)"!(ComponentPose=@n123, Alpha=1, ConstraintWeights_0=0, ConstraintWeights_1:float=0.5, ConstraintWeights_2:float=0.5) -> (Pose) {
    alphaBoolBlend {blendOption=Linear}
    boneToModify {boneName=dyn_skirt_bk_01_l}
    constraintSetup [
     {transformType=Rotation, targetBone {boneName=pelvis}}
     {transformType=Rotation, targetBone {boneName=dyn_skirt_bk_01}}
     {transformType=Rotation, targetBone {boneName=dyn_skirt_bk2_01_l}}
    ]
    constraintWeights [1,1,1]
   }
   n125 = "Constraint - dyn_skirt_bk_02_l (RRR)"!(ComponentPose=@n124, Alpha=1, ConstraintWeights_0=0, ConstraintWeights_1:float=0.5, ConstraintWeights_2:float=0.5) -> (Pose) {
    alphaBoolBlend {blendOption=Linear}
    boneToModify {boneName=dyn_skirt_bk_02_l}
    constraintSetup [
     {transformType=Rotation, targetBone {boneName=pelvis}}
     {transformType=Rotation, targetBone {boneName=dyn_skirt_bk_02}}
     {transformType=Rotation, targetBone {boneName=dyn_skirt_bk2_02_l}}
    ]
    constraintWeights [1,1,1]
   }
   n126 = "Constraint - dyn_skirt_bk_03_l (RRR)"!(ComponentPose=@n125, Alpha=1, ConstraintWeights_0=0, ConstraintWeights_1:float=0.5, ConstraintWeights_2:float=0.5) -> (Pose) {
    alphaBoolBlend {blendOption=Linear}
    boneToModify {boneName=dyn_skirt_bk_03_l}
    constraintSetup [
     {transformType=Rotation, targetBone {boneName=pelvis}}
     {transformType=Rotation, targetBone {boneName=dyn_skirt_bk_03}}
     {transformType=Rotation, targetBone {boneName=dyn_skirt_bk2_03_l}}
    ]
    constraintWeights [1,1,1]
   }
   n127 = "Constraint - dyn_skirt_bk_01_r (RRR)"!(ComponentPose=@n126, Alpha=1, ConstraintWeights_0=0, ConstraintWeights_1:float=0.5, ConstraintWeights_2:float=0.5) -> (Pose) {
    alphaBoolBlend {blendOption=Linear}
    boneToModify {boneName=dyn_skirt_bk_01_r}
    constraintSetup [
     {transformType=Rotation, targetBone {boneName=pelvis}}
     {transformType=Rotation, targetBone {boneName=dyn_skirt_bk_01}}
     {transformType=Rotation, targetBone {boneName=dyn_skirt_bk2_01_r}}
    ]
    constraintWeights [1,1,1]
   }
   n128 = "Constraint - dyn_skirt_bk_02_r (RRR)"!(ComponentPose=@n127, Alpha=1, ConstraintWeights_0=0, ConstraintWeights_1:float=0.5, ConstraintWeights_2:float=0.5) -> (Pose) {
    alphaBoolBlend {blendOption=Linear}
    boneToModify {boneName=dyn_skirt_bk_02_r}
    constraintSetup [
     {transformType=Rotation, targetBone {boneName=pelvis}}
     {transformType=Rotation, targetBone {boneName=dyn_skirt_bk_02}}
     {transformType=Rotation, targetBone {boneName=dyn_skirt_bk2_02_r}}
    ]
    constraintWeights [1,1,1]
   }
   n129 = "Constraint - dyn_skirt_bk_03_r (RRR)"!(ComponentPose=@n128, Alpha=1, ConstraintWeights_0=0, ConstraintWeights_1:float=0.5, ConstraintWeights_2:float=0.5) -> (Pose) {
    alphaBoolBlend {blendOption=Linear}
    boneToModify {boneName=dyn_skirt_bk_03_r}
    constraintSetup [
     {transformType=Rotation, targetBone {boneName=pelvis}}
     {transformType=Rotation, targetBone {boneName=dyn_skirt_bk_03}}
     {transformType=Rotation, targetBone {boneName=dyn_skirt_bk2_03_r}}
    ]
    constraintWeights [1,1,1]
   }
   n130 = "Constraint - dyn_skirt_bk2_01_r (P)"(ComponentPose=@n129, Alpha=1, ConstraintWeights_0=1) {
    alphaBoolBlend {blendOption=Linear}
    boneToModify {boneName=dyn_skirt_bk2_01_r}
    constraintSetup [{targetBone {boneName=pelvis}}]
    constraintWeights [1]
   }
   n131 = "Constraint - dyn_skirt_bk2_02_r (P)"(ComponentPose=@n130, Alpha=1, ConstraintWeights_0=1) {
    alphaBoolBlend {blendOption=Linear}
    boneToModify {boneName=dyn_skirt_bk2_02_r}
    constraintSetup [{targetBone {boneName=pelvis}}]
    constraintWeights [1]
   }
   n132 = "Constraint - dyn_skirt_bk2_03_r (P)"(ComponentPose=@n131, Alpha=1, ConstraintWeights_0=1) {
    alphaBoolBlend {blendOption=Linear}
    boneToModify {boneName=dyn_skirt_bk2_03_r}
    constraintSetup [{targetBone {boneName=pelvis}}]
    constraintWeights [1]
   }
   n133 = Anim Dynamics - Bone: dyn_skirt_bk_01_l(ComponentPose=@n132, Alpha=0.5, ExternalForce="0,0,0") {
    angularDampingOverride=0.9999
    angularSpringConstant=75
    bAngularSpring=true
    bGravityOverrideInSimSpace=true
    bOverrideAngularDamping=true
    simulationSpace=RootRelative
    alphaBoolBlend {blendOption=Linear}
    boundBone {boneName=dyn_skirt_bk_01_l}
    componentLinearAccScale=<1.2 1.2 1.2>
    componentLinearVelScale=<1.2 1.2 1.2>
    physicsBodyDefinitions [{
      sphereCollisionRadius=0
      boundBone {boneName=dyn_skirt_bk_01_l}
      boxExtents=<20 20 10>
      constraintSetup {angularLimitsMax=<25 0 25>, angularLimitsMin=<-25 -25 -25>, angularTarget=<0 0 0>}
      localJointOffset=<2 0 0>
     }]
   }
   n134 = Anim Dynamics - Bone: dyn_skirt_bk_02_l(ComponentPose=@n133, Alpha=0.5, ExternalForce="0,0,0") {
    angularDampingOverride=0.9999
    angularSpringConstant=75
    bAngularSpring=true
    bGravityOverrideInSimSpace=true
    bOverrideAngularDamping=true
    simulationSpace=RootRelative
    alphaBoolBlend {blendOption=Linear}
    boundBone {boneName=dyn_skirt_bk_02_l}
    componentLinearAccScale=<1.2 1.2 1.2>
    componentLinearVelScale=<1.2 1.2 1.2>
    physicsBodyDefinitions [{
      sphereCollisionRadius=0
      boundBone {boneName=dyn_skirt_bk_02_l}
      boxExtents=<20 20 10>
      constraintSetup {angularLimitsMax=<25 0 25>, angularLimitsMin=<-25 -25 -25>, angularTarget=<0 0 0>}
      localJointOffset=<2 0 0>
     }]
   }
   n135 = Anim Dynamics - Bone: dyn_skirt_bk_03_l(ComponentPose=@n134, Alpha=0.5, ExternalForce="0,0,0") {
    angularDampingOverride=0.9999
    angularSpringConstant=75
    bAngularSpring=true
    bGravityOverrideInSimSpace=true
    bOverrideAngularDamping=true
    simulationSpace=RootRelative
    alphaBoolBlend {blendOption=Linear}
    boundBone {boneName=dyn_skirt_bk_03_l}
    componentLinearAccScale=<1.2 1.2 1.2>
    componentLinearVelScale=<1.2 1.2 1.2>
    physicsBodyDefinitions [{
      sphereCollisionRadius=0
      boundBone {boneName=dyn_skirt_bk_03_l}
      boxExtents=<20 20 10>
      constraintSetup {angularLimitsMax=<25 0 25>, angularLimitsMin=<-25 -25 -25>, angularTarget=<0 0 0>}
      localJointOffset=<2 0 0>
     }]
   }
   n136 = Anim Dynamics - Bone: dyn_skirt_bk_01(ComponentPose=@n135, Alpha=0.5, ExternalForce="0,0,0") {
    angularDampingOverride=0.9999
    angularSpringConstant=75
    bAngularSpring=true
    bGravityOverrideInSimSpace=true
    bOverrideAngularDamping=true
    simulationSpace=RootRelative
    alphaBoolBlend {blendOption=Linear}
    boundBone {boneName=dyn_skirt_bk_01}
    componentLinearAccScale=<1.2 1.2 1.2>
    componentLinearVelScale=<1.2 1.2 1.2>
    physicsBodyDefinitions [{
      sphereCollisionRadius=0
      boundBone {boneName=dyn_skirt_bk_01}
      boxExtents=<20 20 10>
      constraintSetup {angularLimitsMax=<25 0 25>, angularLimitsMin=<-25 -25 -25>, angularTarget=<0 0 0>}
      localJointOffset=<2 0 0>
     }]
   }
   n137 = Anim Dynamics - Bone: dyn_skirt_bk_02(ComponentPose=@n136, Alpha=0.5, ExternalForce="0,0,0") {
    angularDampingOverride=0.9999
    angularSpringConstant=75
    bAngularSpring=true
    bGravityOverrideInSimSpace=true
    bOverrideAngularDamping=true
    simulationSpace=RootRelative
    alphaBoolBlend {blendOption=Linear}
    boundBone {boneName=dyn_skirt_bk_02}
    componentLinearAccScale=<1.2 1.2 1.2>
    componentLinearVelScale=<1.2 1.2 1.2>
    physicsBodyDefinitions [{
      sphereCollisionRadius=0
      boundBone {boneName=dyn_skirt_bk_02}
      boxExtents=<20 20 10>
      constraintSetup {angularLimitsMax=<25 0 25>, angularLimitsMin=<-25 -25 -25>, angularTarget=<0 0 0>}
      localJointOffset=<2 0 0>
     }]
   }
   n138 = Anim Dynamics - Bone: dyn_skirt_bk_03(ComponentPose=@n137, Alpha=0.5, ExternalForce="0,0,0") {
    angularDampingOverride=0.9999
    angularSpringConstant=75
    bAngularSpring=true
    bGravityOverrideInSimSpace=true
    bOverrideAngularDamping=true
    simulationSpace=RootRelative
    alphaBoolBlend {blendOption=Linear}
    boundBone {boneName=dyn_skirt_bk_03}
    componentLinearAccScale=<1.2 1.2 1.2>
    componentLinearVelScale=<1.2 1.2 1.2>
    physicsBodyDefinitions [{
      sphereCollisionRadius=0
      boundBone {boneName=dyn_skirt_bk_03}
      boxExtents=<20 20 10>
      constraintSetup {angularLimitsMax=<25 0 25>, angularLimitsMin=<-25 -25 -25>, angularTarget=<0 0 0>}
      localJointOffset=<2 0 0>
     }]
   }
   n139 = Anim Dynamics - Bone: dyn_skirt_bk_01_r(ComponentPose=@n138, Alpha=0.5, ExternalForce="0,0,0") {
    angularDampingOverride=0.9999
    angularSpringConstant=75
    bAngularSpring=true
    bGravityOverrideInSimSpace=true
    bOverrideAngularDamping=true
    simulationSpace=RootRelative
    alphaBoolBlend {blendOption=Linear}
    boundBone {boneName=dyn_skirt_bk_01_r}
    componentLinearAccScale=<1.2 1.2 1.2>
    componentLinearVelScale=<1.2 1.2 1.2>
    physicsBodyDefinitions [{
      sphereCollisionRadius=0
      boundBone {boneName=dyn_skirt_bk_01_r}
      boxExtents=<20 20 10>
      constraintSetup {angularLimitsMax=<25 0 25>, angularLimitsMin=<-25 -25 -25>, angularTarget=<0 0 0>}
      localJointOffset=<2 0 0>
     }]
   }
   n140 = Anim Dynamics - Bone: dyn_skirt_bk_02_r(ComponentPose=@n139, Alpha=0.5, ExternalForce="0,0,0") {
    angularDampingOverride=0.9999
    angularSpringConstant=75
    bAngularSpring=true
    bGravityOverrideInSimSpace=true
    bOverrideAngularDamping=true
    simulationSpace=RootRelative
    alphaBoolBlend {blendOption=Linear}
    boundBone {boneName=dyn_skirt_bk_02_r}
    componentLinearAccScale=<1.2 1.2 1.2>
    componentLinearVelScale=<1.2 1.2 1.2>
    physicsBodyDefinitions [{
      sphereCollisionRadius=0
      boundBone {boneName=dyn_skirt_bk_02_r}
      boxExtents=<20 20 10>
      constraintSetup {angularLimitsMax=<25 0 25>, angularLimitsMin=<-25 -25 -25>, angularTarget=<0 0 0>}
      localJointOffset=<2 0 0>
     }]
   }
   n141 = Anim Dynamics - Bone: dyn_skirt_bk_03_r(ComponentPose=@n140, Alpha=0.5, ExternalForce="0,0,0") {
    angularDampingOverride=0.9999
    angularSpringConstant=75
    bAngularSpring=true
    bGravityOverrideInSimSpace=true
    bOverrideAngularDamping=true
    simulationSpace=RootRelative
    alphaBoolBlend {blendOption=Linear}
    boundBone {boneName=dyn_skirt_bk_03_r}
    componentLinearAccScale=<1.2 1.2 1.2>
    componentLinearVelScale=<1.2 1.2 1.2>
    physicsBodyDefinitions [{
      sphereCollisionRadius=0
      boundBone {boneName=dyn_skirt_bk_03_r}
      boxExtents=<20 20 10>
      constraintSetup {angularLimitsMax=<25 0 25>, angularLimitsMin=<-25 -25 -25>, angularTarget=<0 0 0>}
      localJointOffset=<2 0 0>
     }]
   }
   n142 = Component To Local(ComponentPose=@n141)
   n143 = RigidBody(ComponentPose=Local To Component(LocalPose=Use cached pose 'PreDressRBAN'()), OverrideWorldGravity="0,0,0") {
    bClampLinearTranslationLimitToRefPose=true
    bOverrideWorldGravity=true
    overridePhysicsAsset=&F_MED_Hardcore_Sportz_Bk_Physics
    simulationSpace=BaseBoneSpace
    alphaBoolBlend {blendOption=Linear}
    baseBoneRef {boneName=pelvis}
    componentLinearAccScale=<0.28 0.28 0.28>
    overrideWorldGravity {z=-980}
   }
   n144 = Component To Local(ComponentPose=@n143)
   n145 = Get skyDivingBool() -> (skyDivingBool:bool)
   n146 = "Constraint - dyn_skirt_fr_01_l_MSTR (PP)"!(
    ComponentPose=Local To Component(LocalPose=Blend Poses by bool(BlendPose_0=@n142, BlendPose_1=@n144, BlendTime_0=0.01, BlendTime_1=0.01, bActiveValue=@n145) {blendType=Linear})
    Alpha=0.95
    ConstraintWeights_0=0
    ConstraintWeights_1:float=1
   ) -> (Pose) {
    alphaBoolBlend {blendOption=Linear}
    boneToModify {boneName=dyn_skirt_fr_01_l_MSTR}
    constraintSetup [{targetBone {boneName=pelvis}}, {targetBone {boneName=dyn_thigh_01_l}}]
    constraintWeights [1,1]
   }
   n147 = "Constraint - dyn_skirt_fr_01_l (PP)"!(ComponentPose=@n146, Alpha=0.95, ConstraintWeights_0=0, ConstraintWeights_1:float=1) -> (Pose) {
    alphaBoolBlend {blendOption=Linear}
    boneToModify {boneName=dyn_skirt_fr_01_l}
    constraintSetup [{targetBone {boneName=pelvis}}, {targetBone {boneName=dyn_thigh_01_l}}]
    constraintWeights [1,1]
   }
   n148 = "Constraint - dyn_skirt_fr_02_l (PP)"!(ComponentPose=@n147, Alpha=0.85, ConstraintWeights_0=0, ConstraintWeights_1:float=1) -> (Pose) {
    alphaBoolBlend {blendOption=Linear}
    boneToModify {boneName=dyn_skirt_fr_02_l}
    constraintSetup [{targetBone {boneName=pelvis}}, {targetBone {boneName=dyn_thigh_01_l}}]
    constraintWeights [1,1]
   }
   n149 = "Constraint - dyn_skirt_fr_03_l (PP)"!(ComponentPose=@n148, Alpha=0.6, ConstraintWeights_0=0, ConstraintWeights_1:float=1) -> (Pose) {
    alphaBoolBlend {blendOption=Linear}
    boneToModify {boneName=dyn_skirt_fr_03_l}
    constraintSetup [{targetBone {boneName=pelvis}}, {targetBone {boneName=dyn_thigh_01_l}}]
    constraintWeights [1,1]
   }
   n150 = "Constraint - dyn_skirt_fr2_01_l_MSTR (PP)"!(ComponentPose=@n149, Alpha=0.9, ConstraintWeights_0=0, ConstraintWeights_1:float=1) -> (Pose) {
    alphaBoolBlend {blendOption=Linear}
    boneToModify {boneName=dyn_skirt_fr2_01_l_MSTR}
    constraintSetup [{targetBone {boneName=pelvis}}, {targetBone {boneName=dyn_thigh_01_l}}]
    constraintWeights [1,1]
   }
   n151 = "Constraint - dyn_skirt_fr2_01_l (PP)"!(ComponentPose=@n150, Alpha=0.9, ConstraintWeights_0=0, ConstraintWeights_1:float=1) -> (Pose) {
    alphaBoolBlend {blendOption=Linear}
    boneToModify {boneName=dyn_skirt_fr2_01_l}
    constraintSetup [{targetBone {boneName=pelvis}}, {targetBone {boneName=dyn_thigh_01_l}}]
    constraintWeights [1,1]
   }
   n152 = "Constraint - dyn_skirt_fr2_02_l (PP)"!(ComponentPose=@n151, Alpha=0.8, ConstraintWeights_0=0, ConstraintWeights_1:float=1) -> (Pose) {
    alphaBoolBlend {blendOption=Linear}
    boneToModify {boneName=dyn_skirt_fr2_02_l}
    constraintSetup [{targetBone {boneName=pelvis}}, {targetBone {boneName=dyn_thigh_01_l}}]
    constraintWeights [1,1]
   }
   n153 = "Constraint - dyn_skirt_fr2_03_l (PP)"!(ComponentPose=@n152, Alpha=0.6, ConstraintWeights_0=0, ConstraintWeights_1:float=1) -> (Pose) {
    alphaBoolBlend {blendOption=Linear}
    boneToModify {boneName=dyn_skirt_fr2_03_l}
    constraintSetup [{targetBone {boneName=pelvis}}, {targetBone {boneName=dyn_thigh_01_l}}]
    constraintWeights [1,1]
   }
   n154 = "Constraint - dyn_skirt_fr_01_r_MSTR (PP)"!(ComponentPose=@n153, Alpha=0.9, ConstraintWeights_0=0, ConstraintWeights_1:float=1) -> (Pose) {
    alphaBoolBlend {blendOption=Linear}
    boneToModify {boneName=dyn_skirt_fr_01_r_MSTR}
    constraintSetup [{targetBone {boneName=pelvis}}, {targetBone {boneName=dyn_thigh_01_r}}]
    constraintWeights [1,1]
   }
   n155 = "Constraint - dyn_skirt_fr_01_r (PP)"!(ComponentPose=@n154, Alpha=0.9, ConstraintWeights_0=0, ConstraintWeights_1:float=1) -> (Pose) {
    alphaBoolBlend {blendOption=Linear}
    boneToModify {boneName=dyn_skirt_fr_01_r}
    constraintSetup [{targetBone {boneName=pelvis}}, {targetBone {boneName=dyn_thigh_01_r}}]
    constraintWeights [1,1]
   }
   n156 = "Constraint - dyn_skirt_fr_02_r (PP)"!(ComponentPose=@n155, Alpha=0.5, ConstraintWeights_0=0, ConstraintWeights_1:float=1) -> (Pose) {
    alphaBoolBlend {blendOption=Linear}
    boneToModify {boneName=dyn_skirt_fr_02_r}
    constraintSetup [{targetBone {boneName=pelvis}}, {targetBone {boneName=dyn_thigh_01_r}}]
    constraintWeights [1,1]
   }
   n157 = "Constraint - dyn_skirt_fr_03_r (PP)"!(ComponentPose=@n156, Alpha=0.5, ConstraintWeights_0=0, ConstraintWeights_1:float=1) -> (Pose) {
    alphaBoolBlend {blendOption=Linear}
    boneToModify {boneName=dyn_skirt_fr_03_r}
    constraintSetup [{targetBone {boneName=pelvis}}, {targetBone {boneName=dyn_thigh_01_r}}]
    constraintWeights [1,1]
   }
   n158 = "Constraint - dyn_skirt_fr2_01_r_MSTR (PP)"!(ComponentPose=@n157, Alpha=0.9, ConstraintWeights_0=0, ConstraintWeights_1:float=1) -> (Pose) {
    alphaBoolBlend {blendOption=Linear}
    boneToModify {boneName=dyn_skirt_fr2_01_r_MSTR}
    constraintSetup [{targetBone {boneName=pelvis}}, {targetBone {boneName=dyn_thigh_01_r}}]
    constraintWeights [1,1]
   }
   n159 = "Constraint - dyn_skirt_fr2_01_r (PP)"!(ComponentPose=@n158, Alpha=0.9, ConstraintWeights_0=0, ConstraintWeights_1:float=1) -> (Pose) {
    alphaBoolBlend {blendOption=Linear}
    boneToModify {boneName=dyn_skirt_fr2_01_r}
    constraintSetup [{targetBone {boneName=pelvis}}, {targetBone {boneName=dyn_thigh_01_r}}]
    constraintWeights [1,1]
   }
   n160 = "Constraint - dyn_skirt_fr2_02_r (PP)"!(ComponentPose=@n159, Alpha=0.5, ConstraintWeights_0=0, ConstraintWeights_1:float=1) -> (Pose) {
    alphaBoolBlend {blendOption=Linear}
    boneToModify {boneName=dyn_skirt_fr2_02_r}
    constraintSetup [{targetBone {boneName=pelvis}}, {targetBone {boneName=dyn_thigh_01_r}}]
    constraintWeights [1,1]
   }
   n161 = "Constraint - dyn_skirt_fr2_03_r (PP)"!(ComponentPose=@n160, Alpha=0.5, ConstraintWeights_0=0, ConstraintWeights_1:float=1) -> (Pose) {
    alphaBoolBlend {blendOption=Linear}
    boneToModify {boneName=dyn_skirt_fr2_03_r}
    constraintSetup [{targetBone {boneName=pelvis}}, {targetBone {boneName=dyn_thigh_01_r}}]
    constraintWeights [1,1]
   }
   n162 = "Constraint - dyn_skirt_fr_01_MSTR (PP)"(ComponentPose=@n161, Alpha=1, ConstraintWeights_0=0.5, ConstraintWeights_1=0.5) {
    alphaBoolBlend {blendOption=Linear}
    boneToModify {boneName=dyn_skirt_fr_01_MSTR}
    constraintSetup [{targetBone {boneName=dyn_skirt_fr_01_l_MSTR}}, {targetBone {boneName=dyn_skirt_fr_01_r_MSTR}}]
    constraintWeights [1,1]
   }
   n163 = "Constraint - dyn_skirt_fr_01 (PP)"(ComponentPose=@n162, Alpha=1, ConstraintWeights_0=0.5, ConstraintWeights_1=0.5) {
    alphaBoolBlend {blendOption=Linear}
    boneToModify {boneName=dyn_skirt_fr_01}
    constraintSetup [{targetBone {boneName=dyn_skirt_fr_01_l}}, {targetBone {boneName=dyn_skirt_fr_01_r}}]
    constraintWeights [1,1]
   }
   n164 = "Constraint - dyn_skirt_fr_02 (PP)"(ComponentPose=@n163, Alpha=1, ConstraintWeights_0=0.5, ConstraintWeights_1=0.5) {
    alphaBoolBlend {blendOption=Linear}
    boneToModify {boneName=dyn_skirt_fr_02}
    constraintSetup [{targetBone {boneName=dyn_skirt_fr_02_l}}, {targetBone {boneName=dyn_skirt_fr_02_r}}]
    constraintWeights [1,1]
   }
   n165 = "Constraint - dyn_skirt_fr_03 (PP)"(ComponentPose=@n164, Alpha=1, ConstraintWeights_0=0.5, ConstraintWeights_1=0.5) {
    alphaBoolBlend {blendOption=Linear}
    boneToModify {boneName=dyn_skirt_fr_03}
    constraintSetup [{targetBone {boneName=dyn_skirt_fr_03_l}}, {targetBone {boneName=dyn_skirt_fr_03_r}}]
    constraintWeights [1,1]
   }
   n166 = "Transform (Modify) Bone - Bone: dyn_skirt_fr_01"(ComponentPose=@n165, Alpha=1, Translation="3,0,0", Rotation="0,0,0", Scale="1,1,1") {
    rotationMode=BMM_Additive
    rotationSpace=BCS_BoneSpace
    translationMode=BMM_Additive
    translationSpace=BCS_BoneSpace
    alphaBoolBlend {blendOption=Linear}
    boneToModify {boneName=dyn_skirt_fr_01}
    translation {x=3}
   }
   n167 = "Transform (Modify) Bone - Bone: dyn_skirt_fr_02"(ComponentPose=@n166, Alpha=1, Translation="2,0,0", Rotation="0,0,0", Scale="1,1,1") {
    rotationMode=BMM_Additive
    rotationSpace=BCS_BoneSpace
    translationMode=BMM_Additive
    translationSpace=BCS_BoneSpace
    alphaBoolBlend {blendOption=Linear}
    boneToModify {boneName=dyn_skirt_fr_02}
    translation {x=2}
   }
   n168 = "Transform (Modify) Bone - Bone: dyn_skirt_fr_03"(ComponentPose=@n167, Alpha=1, Translation="1,0,0", Rotation="0,0,0", Scale="1,1,1") {
    rotationMode=BMM_Additive
    rotationSpace=BCS_BoneSpace
    translationMode=BMM_Additive
    translationSpace=BCS_BoneSpace
    alphaBoolBlend {blendOption=Linear}
    boneToModify {boneName=dyn_skirt_fr_03}
    translation {x=1}
   }
   n169 = "Constraint - dyn_skirt_01_l_MSTR (P)"(ComponentPose=@n168, Alpha=0.5, ConstraintWeights_0=1) {
    alphaBoolBlend {blendOption=Linear}
    boneToModify {boneName=dyn_skirt_01_l_MSTR}
    constraintSetup [{targetBone {boneName=dyn_skirt_fr2_01_l_MSTR}}]
    constraintWeights [1]
   }
   n170 = "Constraint - dyn_skirt_01_l (P)"(ComponentPose=@n169, Alpha=0.5, ConstraintWeights_0=1) {
    alphaBoolBlend {blendOption=Linear}
    boneToModify {boneName=dyn_skirt_01_l}
    constraintSetup [{targetBone {boneName=dyn_skirt_fr2_01_l}}]
    constraintWeights [1]
   }
   n171 = "Constraint - dyn_skirt_02_l (P)"(ComponentPose=@n170, Alpha=0.5, ConstraintWeights_0=1) {
    alphaBoolBlend {blendOption=Linear}
    boneToModify {boneName=dyn_skirt_02_l}
    constraintSetup [{targetBone {boneName=dyn_skirt_fr2_02_l}}]
    constraintWeights [1]
   }
   n172 = "Constraint - dyn_skirt_03_l (P)"(ComponentPose=@n171, Alpha=0.5, ConstraintWeights_0=1) {
    alphaBoolBlend {blendOption=Linear}
    boneToModify {boneName=dyn_skirt_03_l}
    constraintSetup [{targetBone {boneName=dyn_skirt_fr2_03_l}}]
    constraintWeights [1]
   }
   n173 = "Constraint - dyn_skirt_bk2_01_l_MSTR (P)"(ComponentPose=@n172, Alpha=0.5, ConstraintWeights_0=1) {
    alphaBoolBlend {blendOption=Linear}
    boneToModify {boneName=dyn_skirt_bk2_01_l_MSTR}
    constraintSetup [{targetBone {boneName=dyn_skirt_01_l_MSTR}}]
    constraintWeights [1]
   }
   n174 = "Constraint - dyn_skirt_bk2_01_l (P)"(ComponentPose=@n173, Alpha=0.5, ConstraintWeights_0=1) {
    alphaBoolBlend {blendOption=Linear}
    boneToModify {boneName=dyn_skirt_bk2_01_l}
    constraintSetup [{targetBone {boneName=dyn_skirt_01_l}}]
    constraintWeights [1]
   }
   n175 = "Constraint - dyn_skirt_bk2_02_l (P)"(ComponentPose=@n174, Alpha=0.5, ConstraintWeights_0=1) {
    alphaBoolBlend {blendOption=Linear}
    boneToModify {boneName=dyn_skirt_bk2_02_l}
    constraintSetup [{targetBone {boneName=dyn_skirt_02_l}}]
    constraintWeights [1]
   }
   n176 = "Constraint - dyn_skirt_bk2_03_l (P)"(ComponentPose=@n175, Alpha=0.5, ConstraintWeights_0=1) {
    alphaBoolBlend {blendOption=Linear}
    boneToModify {boneName=dyn_skirt_bk2_03_l}
    constraintSetup [{targetBone {boneName=dyn_skirt_03_l}}]
    constraintWeights [1]
   }
   n177 = "Constraint - dyn_skirt_01_r_MSTR (P)"(ComponentPose=@n176, Alpha=0.5, ConstraintWeights_0=1) {
    alphaBoolBlend {blendOption=Linear}
    boneToModify {boneName=dyn_skirt_01_r_MSTR}
    constraintSetup [{targetBone {boneName=dyn_skirt_fr2_01_r_MSTR}}]
    constraintWeights [1]
   }
   n178 = "Constraint - dyn_skirt_01_r (P)"(ComponentPose=@n177, Alpha=0.5, ConstraintWeights_0=1) {
    alphaBoolBlend {blendOption=Linear}
    boneToModify {boneName=dyn_skirt_01_r}
    constraintSetup [{targetBone {boneName=dyn_skirt_fr2_01_r}}]
    constraintWeights [1]
   }
   n179 = "Constraint - dyn_skirt_02_r (P)"(ComponentPose=@n178, Alpha=0.5, ConstraintWeights_0=1) {
    alphaBoolBlend {blendOption=Linear}
    boneToModify {boneName=dyn_skirt_02_r}
    constraintSetup [{targetBone {boneName=dyn_skirt_fr2_02_r}}]
    constraintWeights [1]
   }
   n180 = "Constraint - dyn_skirt_03_r (P)"(ComponentPose=@n179, Alpha=0.5, ConstraintWeights_0=1) {
    alphaBoolBlend {blendOption=Linear}
    boneToModify {boneName=dyn_skirt_03_r}
    constraintSetup [{targetBone {boneName=dyn_skirt_fr2_03_r}}]
    constraintWeights [1]
   }
   n181 = "Constraint - dyn_skirt_bk2_01_r_MSTR (P)"(ComponentPose=@n180, Alpha=0.5, ConstraintWeights_0=1) {
    alphaBoolBlend {blendOption=Linear}
    boneToModify {boneName=dyn_skirt_bk2_01_r_MSTR}
    constraintSetup [{targetBone {boneName=dyn_skirt_01_r_MSTR}}]
    constraintWeights [1]
   }
   n182 = "Constraint - dyn_skirt_bk2_01_r (P)"(ComponentPose=@n181, Alpha=0.5, ConstraintWeights_0=1) {
    alphaBoolBlend {blendOption=Linear}
    boneToModify {boneName=dyn_skirt_bk2_01_r}
    constraintSetup [{targetBone {boneName=dyn_skirt_01_r}}]
    constraintWeights [1]
   }
   n183 = "Constraint - dyn_skirt_bk2_02_r (P)"(ComponentPose=@n182, Alpha=0.5, ConstraintWeights_0=1) {
    alphaBoolBlend {blendOption=Linear}
    boneToModify {boneName=dyn_skirt_bk2_02_r}
    constraintSetup [{targetBone {boneName=dyn_skirt_02_r}}]
    constraintWeights [1]
   }
   n184 = "Constraint - dyn_skirt_bk2_03_r (P)"(ComponentPose=@n183, Alpha=0.5, ConstraintWeights_0=1) {
    alphaBoolBlend {blendOption=Linear}
    boneToModify {boneName=dyn_skirt_bk2_03_r}
    constraintSetup [{targetBone {boneName=dyn_skirt_03_r}}]
    constraintWeights [1]
   }
   n185 = "Constraint - dyn_skirt_bk2_01_l_MSTR (PP)"(ComponentPose=@n184, Alpha=1, ConstraintWeights_0=0.5, ConstraintWeights_1=0.5) {
    alphaBoolBlend {blendOption=Linear}
    boneToModify {boneName=dyn_skirt_bk2_01_l_MSTR}
    constraintSetup [{targetBone {boneName=dyn_skirt_bk_01_l_MSTR}}, {targetBone {boneName=dyn_skirt_01_l_MSTR}}]
    constraintWeights [1,1]
   }
   n186 = "Constraint - dyn_skirt_bk2_01_l (PP)"!(ComponentPose=@n185, Alpha=1, ConstraintWeights_0=0.5, ConstraintWeights_1:float=0.5) -> (Pose) {
    alphaBoolBlend {blendOption=Linear}
    boneToModify {boneName=dyn_skirt_bk2_01_l}
    constraintSetup [{targetBone {boneName=dyn_skirt_bk_01_l}}, {targetBone {boneName=dyn_skirt_01_l}}]
    constraintWeights [1,1]
   }
   n187 = "Constraint - dyn_skirt_bk2_02_l (PP)"!(ComponentPose=@n186, Alpha=1, ConstraintWeights_0=0.5, ConstraintWeights_1:float=0.5) -> (Pose) {
    alphaBoolBlend {blendOption=Linear}
    boneToModify {boneName=dyn_skirt_bk2_02_l}
    constraintSetup [{targetBone {boneName=dyn_skirt_bk_02_l}}, {targetBone {boneName=dyn_skirt_02_l}}]
    constraintWeights [1,1]
   }
   n188 = "Constraint - dyn_skirt_bk2_03_l (PP)"!(ComponentPose=@n187, Alpha=1, ConstraintWeights_0=0.5, ConstraintWeights_1:float=0.5) -> (Pose) {
    alphaBoolBlend {blendOption=Linear}
    boneToModify {boneName=dyn_skirt_bk2_03_l}
    constraintSetup [{targetBone {boneName=dyn_skirt_bk_03_l}}, {targetBone {boneName=dyn_skirt_03_l}}]
    constraintWeights [1,1]
   }
   n189 = "Transform (Modify) Bone - Bone: dyn_skirt_bk2_01_l_MSTR"(ComponentPose=@n188, Alpha=1, Translation="2,0,0", Rotation="0,0,10", Scale="1,1,1") {
    rotationMode=BMM_Additive
    scaleMode=BMM_Additive
    translationMode=BMM_Additive
    alphaBoolBlend {blendOption=Linear}
    boneToModify {boneName=dyn_skirt_bk2_01_l_MSTR}
    rotation {roll=10}
    translation {x=2}
   }
   n190 = "Constraint - dyn_skirt_01_l_MSTR (PP)"(ComponentPose=@n189, Alpha=1, ConstraintWeights_0=0.5, ConstraintWeights_1=0.5) {
    alphaBoolBlend {blendOption=Linear}
    boneToModify {boneName=dyn_skirt_01_l_MSTR}
    constraintSetup [{targetBone {boneName=dyn_skirt_bk2_01_l_MSTR}}, {targetBone {boneName=dyn_skirt_fr2_01_l_MSTR}}]
    constraintWeights [1,1]
   }
   n191 = "Constraint - dyn_skirt_01_l (PP)"!(ComponentPose=@n190, Alpha=1, ConstraintWeights_0=0.5, ConstraintWeights_1:float=0.5) -> (Pose) {
    alphaBoolBlend {blendOption=Linear}
    boneToModify {boneName=dyn_skirt_01_l}
    constraintSetup [{targetBone {boneName=dyn_skirt_bk2_01_l}}, {targetBone {boneName=dyn_skirt_fr2_01_l_MSTR}}]
    constraintWeights [1,1]
   }
   n192 = "Constraint - dyn_skirt_02_l (PP)"!(ComponentPose=@n191, Alpha=1, ConstraintWeights_0=0.5, ConstraintWeights_1:float=0.5) -> (Pose) {
    alphaBoolBlend {blendOption=Linear}
    boneToModify {boneName=dyn_skirt_02_l}
    constraintSetup [{targetBone {boneName=dyn_skirt_bk2_02_l}}, {targetBone {boneName=dyn_skirt_fr2_02_l}}]
    constraintWeights [1,1]
   }
   n193 = "Constraint - dyn_skirt_03_l (PP)"!(ComponentPose=@n192, Alpha=1, ConstraintWeights_0=0.5, ConstraintWeights_1:float=0.5) -> (Pose) {
    alphaBoolBlend {blendOption=Linear}
    boneToModify {boneName=dyn_skirt_03_l}
    constraintSetup [{targetBone {boneName=dyn_skirt_bk2_03_l}}, {targetBone {boneName=dyn_skirt_fr2_03_l}}]
    constraintWeights [1,1]
   }
   n194 = "Transform (Modify) Bone - Bone: dyn_skirt_01_l_MSTR"(ComponentPose=@n193, Alpha=1, Translation="0,0,3", Rotation="0,0,0", Scale="1,1,1") {
    rotationMode=BMM_Additive
    rotationSpace=BCS_BoneSpace
    translationMode=BMM_Additive
    alphaBoolBlend {blendOption=Linear}
    boneToModify {boneName=dyn_skirt_01_l_MSTR}
    translation {z=3}
   }
   n195 = "Constraint - dyn_skirt_bk2_01_r_MSTR (PP)"(ComponentPose=@n194, Alpha=1, ConstraintWeights_0=0.5, ConstraintWeights_1=0.5) {
    alphaBoolBlend {blendOption=Linear}
    boneToModify {boneName=dyn_skirt_bk2_01_r_MSTR}
    constraintSetup [{targetBone {boneName=dyn_skirt_bk_01_r_MSTR}}, {targetBone {boneName=dyn_skirt_01_r_MSTR}}]
    constraintWeights [1,1]
   }
   n196 = "Constraint - dyn_skirt_bk2_01_r (PP)"!(ComponentPose=@n195, Alpha=1, ConstraintWeights_0=0.5, ConstraintWeights_1:float=0.5) -> (Pose) {
    alphaBoolBlend {blendOption=Linear}
    boneToModify {boneName=dyn_skirt_bk2_01_r}
    constraintSetup [{targetBone {boneName=dyn_skirt_bk_01_r}}, {targetBone {boneName=dyn_skirt_01_r}}]
    constraintWeights [1,1]
   }
   n197 = "Constraint - dyn_skirt_bk2_02_r (PP)"!(ComponentPose=@n196, Alpha=1, ConstraintWeights_0=0.5, ConstraintWeights_1:float=0.5) -> (Pose) {
    alphaBoolBlend {blendOption=Linear}
    boneToModify {boneName=dyn_skirt_bk2_02_r}
    constraintSetup [{targetBone {boneName=dyn_skirt_bk_02_r}}, {targetBone {boneName=dyn_skirt_02_r}}]
    constraintWeights [1,1]
   }
   n198 = "Constraint - dyn_skirt_bk2_03_r (PP)"!(ComponentPose=@n197, Alpha=1, ConstraintWeights_0=0.5, ConstraintWeights_1:float=0.5) -> (Pose) {
    alphaBoolBlend {blendOption=Linear}
    boneToModify {boneName=dyn_skirt_bk2_03_r}
    constraintSetup [{targetBone {boneName=dyn_skirt_bk_03_r}}, {targetBone {boneName=dyn_skirt_03_r}}]
    constraintWeights [1,1]
   }
   n199 = "Constraint - dyn_skirt_01_r_MSTR (PP)"(ComponentPose=@n198, Alpha=1, ConstraintWeights_0=0.5, ConstraintWeights_1=0.5) {
    alphaBoolBlend {blendOption=Linear}
    boneToModify {boneName=dyn_skirt_01_r_MSTR}
    constraintSetup [{targetBone {boneName=dyn_skirt_bk2_01_r_MSTR}}, {targetBone {boneName=dyn_skirt_fr2_01_r_MSTR}}]
    constraintWeights [1,1]
   }
   n200 = "Constraint - dyn_skirt_01_r (PP)"!(ComponentPose=@n199, Alpha=1, ConstraintWeights_0=0.5, ConstraintWeights_1:float=0.5) -> (Pose) {
    alphaBoolBlend {blendOption=Linear}
    boneToModify {boneName=dyn_skirt_01_r}
    constraintSetup [{targetBone {boneName=dyn_skirt_bk2_01_r}}, {targetBone {boneName=dyn_skirt_fr2_01_r}}]
    constraintWeights [1,1]
   }
   n201 = "Constraint - dyn_skirt_02_r (PP)"!(ComponentPose=@n200, Alpha=1, ConstraintWeights_0=0.5, ConstraintWeights_1:float=0.5) -> (Pose) {
    alphaBoolBlend {blendOption=Linear}
    boneToModify {boneName=dyn_skirt_02_r}
    constraintSetup [{targetBone {boneName=dyn_skirt_bk2_02_r}}, {targetBone {boneName=dyn_skirt_fr2_02_r}}]
    constraintWeights [1,1]
   }
   n202 = "Constraint - dyn_skirt_03_r (PP)"!(ComponentPose=@n201, Alpha=1, ConstraintWeights_0=0.5, ConstraintWeights_1:float=0.5) -> (Pose) {
    alphaBoolBlend {blendOption=Linear}
    boneToModify {boneName=dyn_skirt_03_r}
    constraintSetup [{targetBone {boneName=dyn_skirt_bk2_03_r}}, {targetBone {boneName=dyn_skirt_fr2_03_r}}]
    constraintWeights [1,1]
   }
   n203 = Anim Dynamics - Bone: dyn_skirt_fr_02_l(ComponentPose=@n202, Alpha=1, ExternalForce="0,0,0") {
    angularDampingOverride=0.9999
    angularSpringConstant=75
    bAngularSpring=true
    bGravityOverrideInSimSpace=true
    bOverrideAngularDamping=true
    simulationSpace=RootRelative
    alphaBoolBlend {blendOption=Linear}
    boundBone {boneName=dyn_skirt_fr_02_l}
    componentLinearAccScale=<1.2 1.2 1.2>
    componentLinearVelScale=<1.2 1.2 1.2>
    physicsBodyDefinitions [{
      sphereCollisionRadius=0
      boundBone {boneName=dyn_skirt_fr_02_l}
      boxExtents=<20 10 10>
      constraintSetup {angularLimitsMax=<25 0 10>, angularLimitsMin=<-25 -25 0>, angularTarget=<0 0 0>}
      localJointOffset=<2 0 0>
     }]
   }
   n204 = Anim Dynamics - Bone: dyn_skirt_fr_03_l(ComponentPose=@n203, Alpha=1, ExternalForce="0,0,0") {
    angularDampingOverride=0.9999
    angularSpringConstant=75
    bAngularSpring=true
    bGravityOverrideInSimSpace=true
    bOverrideAngularDamping=true
    simulationSpace=RootRelative
    alphaBoolBlend {blendOption=Linear}
    boundBone {boneName=dyn_skirt_fr_03_l}
    componentLinearAccScale=<1.2 1.2 1.2>
    componentLinearVelScale=<1.2 1.2 1.2>
    physicsBodyDefinitions [{
      sphereCollisionRadius=0
      boundBone {boneName=dyn_skirt_fr_03_l}
      boxExtents=<20 10 10>
      constraintSetup {angularLimitsMax=<25 0 10>, angularLimitsMin=<-25 -25 -10>, angularTarget=<0 0 0>}
      localJointOffset=<2 0 0>
     }]
   }
   n205 = Anim Dynamics - Bone: dyn_skirt_fr_02_r(ComponentPose=@n204, Alpha=1, ExternalForce="0,0,0") {
    angularDampingOverride=0.9999
    angularSpringConstant=75
    bAngularSpring=true
    bGravityOverrideInSimSpace=true
    bOverrideAngularDamping=true
    simulationSpace=RootRelative
    alphaBoolBlend {blendOption=Linear}
    boundBone {boneName=dyn_skirt_fr_02_r}
    componentLinearAccScale=<1.2 1.2 1.2>
    componentLinearVelScale=<1.2 1.2 1.2>
    physicsBodyDefinitions [{
      sphereCollisionRadius=0
      boundBone {boneName=dyn_skirt_fr_02_r}
      boxExtents=<20 20 10>
      constraintSetup {angularLimitsMax=<25 0 25>, angularLimitsMin=<-25 -25 -25>, angularTarget=<0 0 0>}
      localJointOffset=<2 0 0>
     }]
   }
   n206 = Anim Dynamics - Bone: dyn_skirt_fr_03_r(ComponentPose=@n205, Alpha=1, ExternalForce="0,0,0") {
    angularDampingOverride=0.9999
    angularSpringConstant=75
    bAngularSpring=true
    bGravityOverrideInSimSpace=true
    bOverrideAngularDamping=true
    simulationSpace=RootRelative
    alphaBoolBlend {blendOption=Linear}
    boundBone {boneName=dyn_skirt_fr_03_r}
    componentLinearAccScale=<1.2 1.2 1.2>
    componentLinearVelScale=<1.2 1.2 1.2>
    physicsBodyDefinitions [{
      sphereCollisionRadius=0
      boundBone {boneName=dyn_skirt_fr_03_r}
      boxExtents=<20 20 10>
      constraintSetup {angularLimitsMax=<25 0 25>, angularLimitsMin=<-25 -25 -25>, angularTarget=<0 0 0>}
      localJointOffset=<2 0 0>
     }]
   }
   n207 = Anim Dynamics - Bone: dyn_skirt_fr2_02_l(ComponentPose=@n206, Alpha=1, ExternalForce="0,0,0") {
    angularDampingOverride=0.9999
    angularSpringConstant=75
    bAngularSpring=true
    bGravityOverrideInSimSpace=true
    bOverrideAngularDamping=true
    simulationSpace=RootRelative
    alphaBoolBlend {blendOption=Linear}
    boundBone {boneName=dyn_skirt_fr2_02_l}
    componentLinearAccScale=<1.2 1.2 1.2>
    componentLinearVelScale=<1.2 1.2 1.2>
    physicsBodyDefinitions [{
      sphereCollisionRadius=0
      boundBone {boneName=dyn_skirt_fr2_02_l}
      boxExtents=<20 10 10>
      constraintSetup {angularLimitsMax=<25 0 10>, angularLimitsMin=<-25 -25 -10>, angularTarget=<0 0 0>}
      localJointOffset=<2 0 0>
     }]
   }
   n208 = Anim Dynamics - Bone: dyn_skirt_fr2_03_l(ComponentPose=@n207, Alpha=1, ExternalForce="0,0,0") {
    angularDampingOverride=0.9999
    angularSpringConstant=75
    bAngularSpring=true
    bGravityOverrideInSimSpace=true
    bOverrideAngularDamping=true
    simulationSpace=RootRelative
    alphaBoolBlend {blendOption=Linear}
    boundBone {boneName=dyn_skirt_fr2_03_l}
    componentLinearAccScale=<1.2 1.2 1.2>
    componentLinearVelScale=<1.2 1.2 1.2>
    physicsBodyDefinitions [{
      sphereCollisionRadius=0
      boundBone {boneName=dyn_skirt_fr2_03_l}
      boxExtents=<20 10 10>
      constraintSetup {angularLimitsMax=<25 0 10>, angularLimitsMin=<-25 -25 -10>, angularTarget=<0 0 0>}
      localJointOffset=<2 0 0>
     }]
   }
   n209 = Anim Dynamics - Bone: dyn_skirt_fr2_02_r(ComponentPose=@n208, Alpha=1, ExternalForce="0,0,0") {
    angularDampingOverride=0.9999
    angularSpringConstant=75
    bAngularSpring=true
    bGravityOverrideInSimSpace=true
    bOverrideAngularDamping=true
    simulationSpace=RootRelative
    alphaBoolBlend {blendOption=Linear}
    boundBone {boneName=dyn_skirt_fr2_02_r}
    componentLinearAccScale=<1.2 1.2 1.2>
    componentLinearVelScale=<1.2 1.2 1.2>
    physicsBodyDefinitions [{
      sphereCollisionRadius=0
      boundBone {boneName=dyn_skirt_fr2_02_r}
      boxExtents=<20 20 10>
      constraintSetup {angularLimitsMax=<25 0 25>, angularLimitsMin=<-25 -25 -25>, angularTarget=<0 0 0>}
      localJointOffset=<2 0 0>
     }]
   }
   n210 = Anim Dynamics - Bone: dyn_skirt_fr2_03_r(ComponentPose=@n209, Alpha=1, ExternalForce="0,0,0") {
    angularDampingOverride=0.9999
    angularSpringConstant=75
    bAngularSpring=true
    bGravityOverrideInSimSpace=true
    bOverrideAngularDamping=true
    simulationSpace=RootRelative
    alphaBoolBlend {blendOption=Linear}
    boundBone {boneName=dyn_skirt_fr2_03_r}
    componentLinearAccScale=<1.2 1.2 1.2>
    componentLinearVelScale=<1.2 1.2 1.2>
    physicsBodyDefinitions [{
      sphereCollisionRadius=0
      boundBone {boneName=dyn_skirt_fr2_03_r}
      boxExtents=<20 20 10>
      constraintSetup {angularLimitsMax=<25 0 25>, angularLimitsMin=<-25 -25 -25>, angularTarget=<0 0 0>}
      localJointOffset=<2 0 0>
     }]
   }
   n211 = Anim Dynamics - Bone: dyn_skirt_02_l(ComponentPose=@n210, Alpha=0.8, ExternalForce="0,0,0") {
    angularDampingOverride=0.9999
    angularSpringConstant=75
    bAngularSpring=true
    bGravityOverrideInSimSpace=true
    bOverrideAngularDamping=true
    simulationSpace=RootRelative
    alphaBoolBlend {blendOption=Linear}
    boundBone {boneName=dyn_skirt_02_l}
    componentLinearAccScale=<1.2 1.2 1.2>
    componentLinearVelScale=<1.2 1.2 1.2>
    physicsBodyDefinitions [{
      sphereCollisionRadius=0
      boundBone {boneName=dyn_skirt_02_l}
      boxExtents=<20 10 10>
      constraintSetup {angularLimitsMax=<25 0 25>, angularLimitsMin=<-25 -25 -25>, angularTarget=<0 0 0>}
      localJointOffset=<2 0 0>
     }]
   }
   n212 = Anim Dynamics - Bone: dyn_skirt_03_l(ComponentPose=@n211, Alpha=0.8, ExternalForce="0,0,0") {
    angularDampingOverride=0.9999
    angularSpringConstant=75
    bAngularSpring=true
    bGravityOverrideInSimSpace=true
    bOverrideAngularDamping=true
    simulationSpace=RootRelative
    alphaBoolBlend {blendOption=Linear}
    boundBone {boneName=dyn_skirt_03_l}
    componentLinearAccScale=<1.2 1.2 1.2>
    componentLinearVelScale=<1.2 1.2 1.2>
    physicsBodyDefinitions [{
      sphereCollisionRadius=0
      boundBone {boneName=dyn_skirt_03_l}
      boxExtents=<20 10 10>
      constraintSetup {angularLimitsMax=<25 0 25>, angularLimitsMin=<-25 -25 -25>, angularTarget=<0 0 0>}
      localJointOffset=<2 0 0>
     }]
   }
   n213 = Anim Dynamics - Bone: dyn_skirt_01_r(ComponentPose=@n212, Alpha=0.8, ExternalForce="0,0,0") {
    angularDampingOverride=0.9999
    angularSpringConstant=75
    bAngularSpring=true
    bGravityOverrideInSimSpace=true
    bOverrideAngularDamping=true
    simulationSpace=RootRelative
    alphaBoolBlend {blendOption=Linear}
    boundBone {boneName=dyn_skirt_01_r}
    componentLinearAccScale=<1.2 1.2 1.2>
    componentLinearVelScale=<1.2 1.2 1.2>
    physicsBodyDefinitions [{
      sphereCollisionRadius=0
      boundBone {boneName=dyn_skirt_01_r}
      boxExtents=<20 20 10>
      constraintSetup {angularLimitsMax=<25 0 25>, angularLimitsMin=<-25 -25 -25>, angularTarget=<0 0 0>}
      localJointOffset=<2 0 0>
     }]
   }
   n214 = Anim Dynamics - Bone: dyn_skirt_03_r(ComponentPose=@n213, Alpha=0.7, ExternalForce="0,0,0") {
    angularDampingOverride=0.9999
    angularSpringConstant=75
    bAngularSpring=true
    bGravityOverrideInSimSpace=true
    bOverrideAngularDamping=true
    simulationSpace=RootRelative
    alphaBoolBlend {blendOption=Linear}
    boundBone {boneName=dyn_skirt_03_r}
    componentLinearAccScale=<1.2 1.2 1.2>
    componentLinearVelScale=<1.2 1.2 1.2>
    physicsBodyDefinitions [{
      sphereCollisionRadius=0
      boundBone {boneName=dyn_skirt_03_r}
      boxExtents=<20 20 10>
      constraintSetup {angularLimitsMax=<25 0 25>, angularLimitsMin=<-25 -25 -25>, angularTarget=<0 0 0>}
      localJointOffset=<2 0 0>
     }]
   }
   n215 = Anim Dynamics - Bone: dyn_skirt_bk2_02_r(ComponentPose=@n214, Alpha=0.8, ExternalForce="0,0,0") {
    angularDampingOverride=0.9999
    angularSpringConstant=75
    bAngularSpring=true
    bGravityOverrideInSimSpace=true
    bOverrideAngularDamping=true
    simulationSpace=RootRelative
    alphaBoolBlend {blendOption=Linear}
    boundBone {boneName=dyn_skirt_bk2_02_r}
    componentLinearAccScale=<1.2 1.2 1.2>
    componentLinearVelScale=<1.2 1.2 1.2>
    physicsBodyDefinitions [{
      sphereCollisionRadius=0
      boundBone {boneName=dyn_skirt_bk2_02_r}
      boxExtents=<20 20 10>
      constraintSetup {angularLimitsMax=<25 0 25>, angularLimitsMin=<-25 -25 -25>, angularTarget=<0 0 0>}
      localJointOffset=<2 0 0>
     }]
   }
   n216 = Anim Dynamics - Bone: dyn_skirt_bk2_02_l(ComponentPose=@n215, Alpha=0.8, ExternalForce="0,0,0") {
    angularDampingOverride=0.9999
    angularSpringConstant=75
    bAngularSpring=true
    bGravityOverrideInSimSpace=true
    bOverrideAngularDamping=true
    simulationSpace=RootRelative
    alphaBoolBlend {blendOption=Linear}
    boundBone {boneName=dyn_skirt_bk2_02_l}
    componentLinearAccScale=<1.2 1.2 1.2>
    componentLinearVelScale=<1.2 1.2 1.2>
    physicsBodyDefinitions [{
      sphereCollisionRadius=0
      boundBone {boneName=dyn_skirt_bk2_02_l}
      boxExtents=<20 20 10>
      constraintSetup {angularLimitsMax=<25 0 25>, angularLimitsMin=<-25 -25 -25>, angularTarget=<0 0 0>}
      localJointOffset=<2 0 0>
     }]
   }
   n217 = Anim Dynamics - Bone: dyn_skirt_bk2_03_r(ComponentPose=@n216, Alpha=0.8, ExternalForce="0,0,0") {
    angularDampingOverride=0.9999
    angularSpringConstant=75
    bAngularSpring=true
    bGravityOverrideInSimSpace=true
    bOverrideAngularDamping=true
    simulationSpace=RootRelative
    alphaBoolBlend {blendOption=Linear}
    boundBone {boneName=dyn_skirt_bk2_03_r}
    componentLinearAccScale=<1.2 1.2 1.2>
    componentLinearVelScale=<1.2 1.2 1.2>
    physicsBodyDefinitions [{
      sphereCollisionRadius=0
      boundBone {boneName=dyn_skirt_bk2_03_r}
      boxExtents=<20 20 10>
      constraintSetup {angularLimitsMax=<25 0 25>, angularLimitsMin=<-25 -25 -25>, angularTarget=<0 0 0>}
      localJointOffset=<2 0 0>
     }]
   }
   n218 = Anim Dynamics - Bone: dyn_skirt_bk2_03_l(ComponentPose=@n217, Alpha=0.8, ExternalForce="0,0,0") {
    angularDampingOverride=0.9999
    angularSpringConstant=75
    bAngularSpring=true
    bGravityOverrideInSimSpace=true
    bOverrideAngularDamping=true
    simulationSpace=RootRelative
    alphaBoolBlend {blendOption=Linear}
    boundBone {boneName=dyn_skirt_bk2_03_l}
    componentLinearAccScale=<1.2 1.2 1.2>
    componentLinearVelScale=<1.2 1.2 1.2>
    physicsBodyDefinitions [{
      sphereCollisionRadius=0
      boundBone {boneName=dyn_skirt_bk2_03_l}
      boxExtents=<20 20 10>
      constraintSetup {angularLimitsMax=<25 0 25>, angularLimitsMin=<-25 -25 -25>, angularTarget=<0 0 0>}
      localJointOffset=<2 0 0>
     }]
   }
   n219 = "Transform (Modify) Bone - Bone: dyn_skirt_01_r_MSTR"(ComponentPose=@n218, Alpha=1, Translation="0,0,2", Rotation="0,0,0", Scale="1,1,1") {
    rotationMode=BMM_Additive
    rotationSpace=BCS_BoneSpace
    scaleMode=BMM_Additive
    scaleSpace=BCS_BoneSpace
    translationMode=BMM_Additive
    translationSpace=BCS_BoneSpace
    alphaBoolBlend {blendOption=Linear}
    boneToModify {boneName=dyn_skirt_01_r_MSTR}
    translation {z=2}
   }
   n220 = "Constraint - dyn_skirt_fr2_01_l (P)"(ComponentPose=@n219, Alpha=1, ConstraintWeights_0=1) {
    alphaBoolBlend {blendOption=Linear}
    alphaScaleBiasClamp {bInterpResult=true, interpSpeedIncreasing=20}
    boneToModify {boneName=dyn_skirt_fr2_01_l}
    constraintSetup [{targetBone {boneName=thigh_l}}]
    constraintWeights [1]
   }
   n221 = "Constraint - dyn_skirt_fr2_02_l (P)"(ComponentPose=@n220, Alpha=1, ConstraintWeights_0=1) {
    alphaBoolBlend {blendOption=Linear}
    alphaScaleBiasClamp {bInterpResult=true, interpSpeedIncreasing=20}
    boneToModify {boneName=dyn_skirt_fr2_02_l}
    constraintSetup [{targetBone {boneName=thigh_l}}]
    constraintWeights [1]
   }
   n222 = "Constraint - dyn_skirt_fr2_03_l (P)"(ComponentPose=@n221, Alpha=1, ConstraintWeights_0=1) {
    alphaBoolBlend {blendOption=Linear}
    alphaScaleBiasClamp {bInterpResult=true, interpSpeedIncreasing=20}
    boneToModify {boneName=dyn_skirt_fr2_03_l}
    constraintSetup [{targetBone {boneName=thigh_l}}]
    constraintWeights [1]
   }
   n223 = "Constraint - dyn_skirt_fr_01_l (P)"(ComponentPose=@n222, Alpha=1, ConstraintWeights_0=1) {
    alphaBoolBlend {blendOption=Linear}
    alphaScaleBiasClamp {bInterpResult=true, interpSpeedIncreasing=20}
    boneToModify {boneName=dyn_skirt_fr_01_l}
    constraintSetup [{targetBone {boneName=thigh_l}}]
    constraintWeights [1]
   }
   n224 = "Constraint - dyn_skirt_fr_02_l (P)"(ComponentPose=@n223, Alpha=1, ConstraintWeights_0=1) {
    alphaBoolBlend {blendOption=Linear}
    alphaScaleBiasClamp {bInterpResult=true, interpSpeedIncreasing=20}
    boneToModify {boneName=dyn_skirt_fr_02_l}
    constraintSetup [{targetBone {boneName=thigh_l}}]
    constraintWeights [1]
   }
   n225 = "Constraint - dyn_skirt_fr_03_l (P)"(ComponentPose=@n224, Alpha=1, ConstraintWeights_0=1) {
    alphaBoolBlend {blendOption=Linear}
    alphaScaleBiasClamp {bInterpResult=true, interpSpeedIncreasing=20}
    boneToModify {boneName=dyn_skirt_fr_03_l}
    constraintSetup [{targetBone {boneName=thigh_l}}]
    constraintWeights [1]
   }
   n226 = "Transform (Modify) Bone - Bone: dyn_skirt_fr2_01_l"(ComponentPose=@n225, Alpha=1, Translation="1,0,0", Rotation="0,0,0", Scale="1,1,1") {
    rotationMode=BMM_Additive
    rotationSpace=BCS_BoneSpace
    translationMode=BMM_Additive
    translationSpace=BCS_BoneSpace
    alphaBoolBlend {blendOption=Linear}
    boneToModify {boneName=dyn_skirt_fr2_01_l}
    translation {x=1}
   }
   n227 = "Transform (Modify) Bone - Bone: dyn_skirt_fr2_02_l"(ComponentPose=@n226, Alpha=1, Translation="1,0,0", Rotation="0,0,0", Scale="1,1,1") {
    rotationMode=BMM_Additive
    rotationSpace=BCS_BoneSpace
    translationMode=BMM_Additive
    translationSpace=BCS_BoneSpace
    alphaBoolBlend {blendOption=Linear}
    boneToModify {boneName=dyn_skirt_fr2_02_l}
    translation {x=1}
   }
   n228 = "Transform (Modify) Bone - Bone: dyn_skirt_fr2_03_l"(ComponentPose=@n227, Alpha=1, Translation="1,0,0", Rotation="0,0,0", Scale="1,1,1") {
    rotationMode=BMM_Additive
    rotationSpace=BCS_BoneSpace
    translationMode=BMM_Additive
    translationSpace=BCS_BoneSpace
    alphaBoolBlend {blendOption=Linear}
    boneToModify {boneName=dyn_skirt_fr2_03_l}
    translation {x=1}
   }
   n229 = "Transform (Modify) Bone - Bone: dyn_skirt_fr_01_l"(ComponentPose=@n228, Alpha=1, Translation="1,0,0", Rotation="0,0,0", Scale="1,1,1") {
    rotationMode=BMM_Additive
    rotationSpace=BCS_BoneSpace
    translationMode=BMM_Additive
    translationSpace=BCS_BoneSpace
    alphaBoolBlend {blendOption=Linear}
    boneToModify {boneName=dyn_skirt_fr_01_l}
    translation {x=1}
   }
   n230 = "Transform (Modify) Bone - Bone: dyn_skirt_fr_02_l"(ComponentPose=@n229, Alpha=1, Translation="1,0,0", Rotation="0,0,0", Scale="1,1,1") {
    rotationMode=BMM_Additive
    rotationSpace=BCS_BoneSpace
    translationMode=BMM_Additive
    translationSpace=BCS_BoneSpace
    alphaBoolBlend {blendOption=Linear}
    boneToModify {boneName=dyn_skirt_fr_02_l}
    translation {x=1}
   }
   n231 = "Transform (Modify) Bone - Bone: dyn_skirt_fr_03_l"(ComponentPose=@n230, Alpha=1, Translation="1,0,0", Rotation="0,0,0", Scale="1,1,1") {
    rotationMode=BMM_Additive
    rotationSpace=BCS_BoneSpace
    translationMode=BMM_Additive
    translationSpace=BCS_BoneSpace
    alphaBoolBlend {blendOption=Linear}
    boneToModify {boneName=dyn_skirt_fr_03_l}
    translation {x=1}
   }
   n232 = "Constraint - dyn_skirt_fr2_01_r (P)"(ComponentPose=@n231, Alpha=1, ConstraintWeights_0=1) {
    alphaBoolBlend {blendOption=Linear}
    alphaScaleBiasClamp {bInterpResult=true, interpSpeedIncreasing=20}
    boneToModify {boneName=dyn_skirt_fr2_01_r}
    constraintSetup [{targetBone {boneName=thigh_r}}]
    constraintWeights [1]
   }
   n233 = "Constraint - dyn_skirt_fr2_02_r (P)"(ComponentPose=@n232, Alpha=1, ConstraintWeights_0=1) {
    alphaBoolBlend {blendOption=Linear}
    alphaScaleBiasClamp {bInterpResult=true, interpSpeedIncreasing=20}
    boneToModify {boneName=dyn_skirt_fr2_02_r}
    constraintSetup [{targetBone {boneName=thigh_r}}]
    constraintWeights [1]
   }
   n234 = "Constraint - dyn_skirt_fr2_03_r (P)"(ComponentPose=@n233, Alpha=1, ConstraintWeights_0=1) {
    alphaBoolBlend {blendOption=Linear}
    alphaScaleBiasClamp {bInterpResult=true, interpSpeedIncreasing=20}
    boneToModify {boneName=dyn_skirt_fr2_03_r}
    constraintSetup [{targetBone {boneName=thigh_r}}]
    constraintWeights [1]
   }
   n235 = "Constraint - dyn_skirt_fr_01_r (P)"(ComponentPose=@n234, Alpha=1, ConstraintWeights_0=1) {
    alphaBoolBlend {blendOption=Linear}
    alphaScaleBiasClamp {bInterpResult=true, interpSpeedIncreasing=20}
    boneToModify {boneName=dyn_skirt_fr_01_r}
    constraintSetup [{targetBone {boneName=thigh_r}}]
    constraintWeights [1]
   }
   n236 = "Constraint - dyn_skirt_fr_02_r (P)"(ComponentPose=@n235, Alpha=1, ConstraintWeights_0=1) {
    alphaBoolBlend {blendOption=Linear}
    alphaScaleBiasClamp {bInterpResult=true, interpSpeedIncreasing=20}
    boneToModify {boneName=dyn_skirt_fr_02_r}
    constraintSetup [{targetBone {boneName=thigh_r}}]
    constraintWeights [1]
   }
   n237 = "Constraint - dyn_skirt_fr_03_r (P)"(ComponentPose=@n236, Alpha=1, ConstraintWeights_0=1) {
    alphaBoolBlend {blendOption=Linear}
    alphaScaleBiasClamp {bInterpResult=true, interpSpeedIncreasing=20}
    boneToModify {boneName=dyn_skirt_fr_03_r}
    constraintSetup [{targetBone {boneName=thigh_r}}]
    constraintWeights [1]
   }
   n238 = "Transform (Modify) Bone - Bone: dyn_skirt_fr2_01_r"(ComponentPose=@n237, Alpha=1, Translation="1,0,0", Rotation="0,0,0", Scale="1,1,1") {
    rotationMode=BMM_Additive
    rotationSpace=BCS_BoneSpace
    translationMode=BMM_Additive
    translationSpace=BCS_BoneSpace
    alphaBoolBlend {blendOption=Linear}
    boneToModify {boneName=dyn_skirt_fr2_01_r}
    translation {x=1}
   }
   n239 = "Transform (Modify) Bone - Bone: dyn_skirt_fr2_02_r"(ComponentPose=@n238, Alpha=1, Translation="1,0,0", Rotation="0,0,0", Scale="1,1,1") {
    rotationMode=BMM_Additive
    rotationSpace=BCS_BoneSpace
    translationMode=BMM_Additive
    translationSpace=BCS_BoneSpace
    alphaBoolBlend {blendOption=Linear}
    boneToModify {boneName=dyn_skirt_fr2_02_r}
    translation {x=1}
   }
   n240 = "Transform (Modify) Bone - Bone: dyn_skirt_fr2_03_r"(ComponentPose=@n239, Alpha=1, Translation="1,0,0", Rotation="0,0,0", Scale="1,1,1") {
    rotationMode=BMM_Additive
    rotationSpace=BCS_BoneSpace
    translationMode=BMM_Additive
    translationSpace=BCS_BoneSpace
    alphaBoolBlend {blendOption=Linear}
    boneToModify {boneName=dyn_skirt_fr2_03_r}
    translation {x=1}
   }
   n241 = "Transform (Modify) Bone - Bone: dyn_skirt_fr_01_r"(ComponentPose=@n240, Alpha=1, Translation="1,0,0", Rotation="0,0,0", Scale="1,1,1") {
    rotationMode=BMM_Additive
    rotationSpace=BCS_BoneSpace
    translationMode=BMM_Additive
    translationSpace=BCS_BoneSpace
    alphaBoolBlend {blendOption=Linear}
    boneToModify {boneName=dyn_skirt_fr_01_r}
    translation {x=1}
   }
   n242 = "Transform (Modify) Bone - Bone: dyn_skirt_fr_02_r"(ComponentPose=@n241, Alpha=1, Translation="1,0,0", Rotation="0,0,0", Scale="1,1,1") {
    rotationMode=BMM_Additive
    rotationSpace=BCS_BoneSpace
    translationMode=BMM_Additive
    translationSpace=BCS_BoneSpace
    alphaBoolBlend {blendOption=Linear}
    boneToModify {boneName=dyn_skirt_fr_02_r}
    translation {x=1}
   }
   n243 = "Transform (Modify) Bone - Bone: dyn_skirt_fr_03_r"(ComponentPose=@n242, Alpha=1, Translation="1,0,0", Rotation="0,0,0", Scale="1,1,1") {
    rotationMode=BMM_Additive
    rotationSpace=BCS_BoneSpace
    translationMode=BMM_Additive
    translationSpace=BCS_BoneSpace
    alphaBoolBlend {blendOption=Linear}
    boneToModify {boneName=dyn_skirt_fr_03_r}
    translation {x=1}
   }
   Save cached pose 'PostDressCache'(Pose=Component To Local(ComponentPose=@n243)) {cacheName=PostDressCache}
   SelectFloat(A=1, B=1, bPickA=Get bIsCrouching())
   SelectFloat(A=1, B=1, bPickA=Get bIsCrouching())
   SelectFloat(A=1, B=1, bPickA=Get bIsCrouching())
   SelectFloat(A=1, B=1, bPickA=Get bIsCrouching())
   SelectFloat(A=1, B=1, bPickA=Get bIsCrouching())
  ]}
 EventGraph {nodes [
   Event BlueprintUpdateAnimation()
   > n1 = Sequence!(execute) -> (then_0, then_1, then_2:Exec, then_3:Exec)
   n2 = Break Fort Anim Input Skydiving External Force(FortAnimInput_SkydivingExternalForce=Get SkydivingExternalForceInput()) -> (
    bUseSkydivingVectorForce=~
    SkydivingVectorForce=~
    HeadToPelvisDirection=~
    FloatingMultiplier=~
    FloatingAdditive=~
    DivingMultiplier=~
    DivingAdditive=~
    ParachutingMultiplier=~
    ParachutingAdditive=~
   )
   > n3 = Set windVector(windVector:Vector=SelectVector(A=vector * float(A=@n2.SkydivingVectorForce, B=0.01), B="0, 0,-980", bPickA=Get bIsSkyDiving())) -> (Output_Get:Vector)
   n4 = GetOwnerSkeletalMeshComponent()
   n5 = BreakRotator(InRot=BreakTransform(InTransform=GetDeltaTransformFromRefPose(self=@n4, BoneName=pelvis, BaseName=root)).Rotation)
   n6 = SelectFloat(A=980, B=-980, bPickA="float > float"(A="Absolute (Float)"(A=@n5.Yaw), B=90) {is=title K2Node_CallFunction})
   > Set finalGravity(finalGravity:Vector=SelectVector(A=MakeVector(Y=0, Z=@n6), B=@n3, bPickA=Get bIsPlayingEmote())) -> (Output_Get:Vector)
   > n7 = Branch(Condition=Get bIsSkyDiving())
   > Set dressAlpha(dressAlpha:double=0.3) -> (Output_Get:double)
   > Set strapAlpha(strapAlpha:double=0) -> (Output_Get:double)
   Set dressAlpha(execute=@n7.else, dressAlpha=0.7)
   > Set strapAlpha(strapAlpha=7)
   n8 = Gravity Override State Function DEPRECATED!(
    execute=@n1.then_1
    CharacterPartAnimInstance=Self-Reference()
    GameTimeInSeconds=GetGameTimeInSeconds()
    GravityOverrideParams_Skydive_GoingUp_28_#1_WindFrequency_28_#2:Vector="1.5,7,8"
    GravityOverrideParams_Skydive_GoingUp_28_#1_WindAmplitude_31_#3:Vector="200,-500,400"
    GravityOverrideParams_Skydive_GoingUp_28_#1_WindOffset_32_#4:Vector="0,0,0"
    GravityOverrideParams_Skydive_GoingUp_28_#1_GravityOverride_15_#5:Vector="0,0,-100"
    GravityOverrideParams_Skydive_GoingUp_28_#1_JointName_27_#6:Name=root
    GravityOverrideParams_Skydive_FloatingDown_29_#7_WindFrequency_28_#2:Vector="1.5,7,8"
    GravityOverrideParams_Skydive_FloatingDown_29_#7_WindAmplitude_31_#3:Vector="200,-500,400"
    GravityOverrideParams_Skydive_FloatingDown_29_#7_WindOffset_32_#4:Vector="0,-600,400"
    GravityOverrideParams_Skydive_FloatingDown_29_#7_GravityOverride_15_#5:Vector="0,0,0"
    GravityOverrideParams_Skydive_FloatingDown_29_#7_JointName_27_#6:Name=root
    GravityOverrideParams_Skydive_DivingDown_30_#8_WindFrequency_28_#2:Vector="1.5,7,8"
    GravityOverrideParams_Skydive_DivingDown_30_#8_WindAmplitude_31_#3:Vector="50,300,100"
    GravityOverrideParams_Skydive_DivingDown_30_#8_WindOffset_32_#4:Vector="0,0,0"
    GravityOverrideParams_Skydive_DivingDown_30_#8_GravityOverride_15_#5:Vector="0,0,0"
    GravityOverrideParams_Skydive_DivingDown_30_#8_JointName_27_#6:Name=pelvis
    GravityOverrideParams_Skydive_Parachuting_31_#9_WindFrequency_28_#2:Vector="1.5,7,8"
    GravityOverrideParams_Skydive_Parachuting_31_#9_WindAmplitude_31_#3:Vector="200,-500,400"
    GravityOverrideParams_Skydive_Parachuting_31_#9_WindOffset_32_#4:Vector="0,0,0"
    GravityOverrideParams_Skydive_Parachuting_31_#9_GravityOverride_15_#5:Vector="0,-800,-980"
    GravityOverrideParams_Skydive_Parachuting_31_#9_JointName_27_#6:Name=root
    GravityOverrideParams_OnGround_Standing_32_#10:Gravity Override Setting Struct Structure=(WindFrequency_28_#2=(X=0,Y=0,Z=0),WindAmplitude_31_#3=(X=0,Y=0,Z=0),WindOffset_32_#4=(X=0,Y=0,Z=0),GravityOverride_15_#5=(X=0,Y=0,Z=-980),JointName_27_#6="root")
    GravityOverrideParams_OnGround_Moving_33_#11:Gravity Override Setting Struct Structure=(WindFrequency_28_#2=(X=0,Y=0,Z=0),WindAmplitude_31_#3=(X=0,Y=0,Z=0),WindOffset_32_#4=(X=0,Y=0,Z=0),GravityOverride_15_#5=(X=0,Y=0,Z=-980),JointName_27_#6="root")
    GravityOverrideParams_EmoteOrMelee_34_#12:Gravity Override Setting Struct Structure=(WindFrequency_28_#2=(X=0,Y=0,Z=0),WindAmplitude_31_#3=(X=0,Y=0,Z=0),WindOffset_32_#4=(X=0,Y=0,Z=0),GravityOverride_15_#5=(X=0,Y=0,Z=-980),JointName_27_#6="root")
    GravityOverrideParams_Swimming_35_#13:Gravity Override Setting Struct Structure=(WindFrequency_28_#2=(X=0,Y=0,Z=0),WindAmplitude_31_#3=(X=0,Y=0,Z=0),WindOffset_32_#4=(X=0,Y=0,Z=0),GravityOverride_15_#5=(X=-800,Y=0,Z=0),JointName_27_#6="pelvis")
    GravityOverrideParams_Front_End_36_#14:Gravity Override Setting Struct Structure=(WindFrequency_28_#2=(X=0,Y=0,Z=0),WindAmplitude_31_#3=(X=0,Y=0,Z=0),WindOffset_32_#4=(X=0,Y=0,Z=0),GravityOverride_15_#5=(X=0,Y=0,Z=-980),JointName_27_#6="root")
    GravityOverrideParams_DBNO_40_#15:Gravity Override Setting Struct Structure=(WindFrequency_28_#2=(X=0,Y=0,Z=0),WindAmplitude_31_#3=(X=0,Y=0,Z=0),WindOffset_32_#4=(X=0,Y=0,Z=0),GravityOverride_15_#5=(X=-980,Y=400,Z=0),JointName_27_#6="pelvis")
    GravityOverrideParams_DBNO_Carried_41_#16:Gravity Override Setting Struct Structure=(WindFrequency_28_#2=(X=0,Y=0,Z=0),WindAmplitude_31_#3=(X=0,Y=0,Z=0),WindOffset_32_#4=(X=0,Y=0,Z=0),GravityOverride_15_#5=(X=-980,Y=500,Z=-500),JointName_27_#6="pelvis")
    GravityOverrideParams_OnGround_Falling_45_#17:Gravity Override Setting Struct Structure=(WindFrequency_28_#2=(X=0,Y=0,Z=0),WindAmplitude_31_#3=(X=0,Y=0,Z=0),WindOffset_32_#4=(X=0,Y=0,Z=0),GravityOverride_15_#5=(X=0,Y=0,Z=-980),JointName_27_#6="root")
    GravityOverrideParams_Ziplining_46_#18:Gravity Override Setting Struct Structure=(WindFrequency_28_#2=(X=0,Y=0,Z=0),WindAmplitude_31_#3=(X=0,Y=0,Z=0),WindOffset_32_#4=(X=0,Y=0,Z=0),GravityOverride_15_#5=(X=0,Y=0,Z=-980),JointName_27_#6="root")
    GravityOverrideParams_Skydive_Surfing_49_#19:Gravity Override Setting Struct Structure=(WindFrequency_28_#2=(X=1.5,Y=7,Z=8),WindAmplitude_31_#3=(X=200,Y=-500,Z=400),WindOffset_32_#4=(X=0,Y=0,Z=0),GravityOverride_15_#5=(X=500,Y=-300,Z=-980),JointName_27_#6="root")
    GravityOverrideParams_Riding_Standing_54_#20:Gravity Override Setting Struct Structure=(WindFrequency_28_#2=(X=0,Y=0,Z=0),WindAmplitude_31_#3=(X=0,Y=0,Z=0),WindOffset_32_#4=(X=0,Y=0,Z=0),GravityOverride_15_#5=(X=0,Y=-350,Z=-980),JointName_27_#6="root")
    GravityOverrideParams_Riding_Moving_55_#21:Gravity Override Setting Struct Structure=(WindFrequency_28_#2=(X=0,Y=0,Z=0),WindAmplitude_31_#3=(X=0,Y=0,Z=0),WindOffset_32_#4=(X=0,Y=0,Z=0),GravityOverride_15_#5=(X=0,Y=-450,Z=-450),JointName_27_#6="root")
    GravityOverrideParams_Riding_Moving_Targeting_58_#22:Gravity Override Setting Struct Structure=(WindFrequency_28_#2=(X=0,Y=0,Z=0),WindAmplitude_31_#3=(X=0,Y=0,Z=0),WindOffset_32_#4=(X=0,Y=0,Z=0),GravityOverride_15_#5=(X=0,Y=-200,Z=-980),JointName_27_#6="root")
    GravityOverrideParams_Tactical_Sprint_57_#23:Gravity Override Setting Struct Structure=(WindFrequency_28_#2=(X=0,Y=0,Z=0),WindAmplitude_31_#3=(X=0,Y=0,Z=0),WindOffset_32_#4=(X=0,Y=0,Z=0),GravityOverride_15_#5=(X=0,Y=-450,Z=-500),JointName_27_#6="root")
    GravityOverrideParams_Rail_Grinding_61_#24:Gravity Override Setting Struct Structure=(WindFrequency_28_#2=(X=0,Y=0,Z=0),WindAmplitude_31_#3=(X=0,Y=0,Z=0),WindOffset_32_#4=(X=0,Y=0,Z=0),GravityOverride_15_#5=(X=0,Y=0,Z=-980),JointName_27_#6="root")
    GravityOverrideParams_DBNO_Tumbling_67_#25:Gravity Override Setting Struct Structure=(WindFrequency_28_#2=(X=0,Y=0,Z=0),WindAmplitude_31_#3=(X=0,Y=0,Z=0),WindOffset_32_#4=(X=0,Y=0,Z=0),GravityOverride_15_#5=(X=-2000,Y=0,Z=0),JointName_27_#6="pelvis")
    GravityOverrideParams_DBNO_RollFwd_68_#26:Gravity Override Setting Struct Structure=(WindFrequency_28_#2=(X=0,Y=0,Z=0),WindAmplitude_31_#3=(X=0,Y=0,Z=0),WindOffset_32_#4=(X=0,Y=0,Z=0),GravityOverride_15_#5=(X=-2000,Y=0,Z=0),JointName_27_#6="pelvis")
    GravityOverrideParams_DBNO_RollBwd_69_#27:Gravity Override Setting Struct Structure=(WindFrequency_28_#2=(X=0,Y=0,Z=0),WindAmplitude_31_#3=(X=0,Y=0,Z=0),WindOffset_32_#4=(X=0,Y=0,Z=0),GravityOverride_15_#5=(X=-4000,Y=0,Z=0),JointName_27_#6="pelvis")
    GravityOverrideParams_DBNO_Impulse_70_#28:Gravity Override Setting Struct Structure=(WindFrequency_28_#2=(X=0,Y=0,Z=0),WindAmplitude_31_#3=(X=0,Y=0,Z=0),WindOffset_32_#4=(X=0,Y=0,Z=0),GravityOverride_15_#5=(X=-2000,Y=0,Z=0),JointName_27_#6="pelvis")
    RBAN Params
    DebugDraw
    DebugJointName
   ) -> (then, GravityOverride, CurrentStateOut, GravityOverrideSettingsOut, RBAN SettingsOut)
   > Set gravityOverride(gravityOverride:Vector=@n8) -> (Output_Get:Vector)
   > Set skyDivingBool(skyDivingBool:bool=OR Boolean(A=Get bIsSkyDiving(), B=Get bIsSkydiveDiveMode())) -> (Output_Get:bool)
   n9 = BreakTransform(InTransform=GetDeltaTransformFromRefPose(self=GetOwnerSkeletalMeshComponent(), BoneName=pelvis, BaseName=thigh_l))
   Set leftLegDelta(execute=@n1.then_2, leftLegDelta:double=MapRangeClamped(Value=BreakRotator(InRot=@n9.Rotation).Yaw, InRangeA=10, InRangeB=108)) -> (Output_Get:double)
   n10 = BreakTransform(InTransform=GetDeltaTransformFromRefPose(self=GetOwnerSkeletalMeshComponent(), BoneName=pelvis, BaseName=thigh_r))
   Set rightLegDelta(execute=@n1.then_3, rightLegDelta:double=MapRangeClamped(Value=BreakRotator(InRot=@n10.Rotation).Yaw, InRangeA=10, InRangeB=108)) -> (Output_Get:double)
   Event OnBeginCrouching()
   > Set isCrouchingAlpha(isCrouchingAlpha:double=1) -> (Output_Get:double)
   n11 = Event OnEndCrouching()
   Set isCrouchingAlpha(execute=[@n11.then, @n12.then], isCrouchingAlpha=0)
   n12 = Event BlueprintInitializeAnimation()
  ]}
]
variables [
 Character Part Anim Instance {type=CustomCharacterPartAnimInstance, default=null}
 dressAlpha {type=number}
 finalGravity {type=Vector}
 gravityOverride {type=Vector}
 isCrouchingAlpha {type=number}
 leftLegDelta {type=number}
 rightLegDelta {type=number}
 skyDivingBool {type=boolean}
 strapAlpha {type=number}
 windVector {type=Vector}
]
