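# Physics asset for the F_MED_Hardcore_Sportz female face accessory mesh, defining bodies and joints for dynamic long back hair and hair strands.
# Tiny MSTR anchor capsules lock angularly to the head, and hair capsules hang from them with limited swing and twist, some continuing to second segments.
# The center back tip locks angularly to its left and right neighbors, tying those three chains together.
# Head and neck capsules provide body volumes, joined with free rotation.
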PhysicsAsset /BRCosmetics/Characters/Player/Female/Medium/Bodies/F_MED_Hardcore_Sportz/Meshes/Parts/F_MED_Hardcore_Sportz_FaceAcc_PhysicsAsset
bodies {
 dyn_hairlong_bk_01 [{shapeType=Capsule, length=5, radius=2, center=<2.3027 0 0>, rotation=<rot -90 0 0>}]
 dyn_hairlong_bk_01_MSTR [{shapeType=Capsule, length=0.01, radius=0.01, rotation=<rot 0 -90 0>}]
 dyn_hairlong_bk_01_l [{shapeType=Capsule, length=5.5, radius=1.5, center=<2.3 0 0.3>, rotation=<rot -90 0 0>}]
 dyn_hairlong_bk_01_l_MSTR [{shapeType=Capsule, length=0.01, radius=0.01, rotation=<rot 0 -90 0>}]
 dyn_hairlong_bk_01_r [{shapeType=Capsule, length=5.5, radius=1.75, center=<2.3027 0 0.0001>, rotation=<rot -90 -1.2658 1.2657>}]
 dyn_hairlong_bk_01_r_MSTR [{shapeType=Capsule, length=0.01, radius=0.01, rotation=<rot 0 -90 0>}]
 dyn_hairlong_bk_02 [{shapeType=Capsule, length=5, radius=2, center=<2.3027 0 0>, rotation=<rot -90 0 0>}]
 dyn_hairlong_bk_02_l [{shapeType=Capsule, length=5.5, radius=1.5, center=<2.3 0 0.2>, rotation=<rot -90 180 180>}]
 dyn_hairlong_bk_02_r [{shapeType=Capsule, length=5.5, radius=1.75, center=<2.3027 0 0.0001>, rotation=<rot -90 -1.2658 1.2657>}]
 dyn_hairstrand_01_l [{shapeType=Capsule, length=5.5, radius=1.75, center=<2.3027 0 0.0001>, rotation=<rot -90 -1.2658 1.2657>}]
 dyn_hairstrand_01_l_MSTR [{shapeType=Capsule, length=0.01, radius=0.01, rotation=<rot 0 -90 0>}]
 dyn_hairstrand_01_r [{shapeType=Capsule, length=5.5, radius=1.75, center=<2.3027 0 0.0001>, rotation=<rot -90 -1.2658 1.2657>}]
 dyn_hairstrand_01_r_MSTR [{shapeType=Capsule, length=0.01, radius=0.01, rotation=<rot 0 -90 0>}]
 dyn_hairstrand_02_l [{shapeType=Capsule, length=5.5, radius=1.75, center=<2.3027 0 0.0001>, rotation=<rot -90 -1.2658 1.2657>}]
 dyn_hairstrand_04_l [{shapeType=Capsule, length=5.5, radius=1.75, center=<2.3027 0 0.0001>, rotation=<rot -90 -1.2658 1.2657>}]
 dyn_hairstrand_bk_01_l_MSTR [{shapeType=Capsule, length=0.01, radius=0.01, rotation=<rot 0 -90 0>}]
 dyn_hairstrand_fr_01_l [{shapeType=Capsule, length=5.5, radius=1.75, center=<2.3027 0 0.0001>, rotation=<rot -90 -1.2658 1.2657>}]
 dyn_hairstrand_fr_01_l_MSTR [{shapeType=Capsule, length=0.01, radius=0.01, rotation=<rot 0 -90 0>}]
 dyn_hairstrand_fr_01_r [{shapeType=Capsule, length=5.5, radius=1.75, center=<2.3 0 0>, rotation=<rot -90 -1.2658 1.2657>}]
 dyn_hairstrand_fr_01_r_MSTR [{shapeType=Capsule, length=0.01, radius=0.01, rotation=<rot 0 -90 0>}]
 head [{shapeType=Capsule, length=10, radius=5.5, center=<0 0.7825 0>, rotation=<rot 90 0 -2.7842>}]
 neck_01 [{shapeType=Capsule, length=12.01, radius=5.25, center=<5.1375 1.427 0>, rotation=<rot 0 -85.0011 90.0001>}]
}
constraints [
 {bone1Name=dyn_hairstrand_fr_01_l, bone2Name=dyn_hairstrand_fr_01_l_MSTR, swing1LimitDegrees=15, swing2LimitDegrees=15, twistLimitDegrees=5}
 {bone1Name=dyn_hairstrand_fr_01_r, bone2Name=dyn_hairstrand_fr_01_r_MSTR, swing1LimitDegrees=15, swing2LimitDegrees=15, twistLimitDegrees=5}
 {bone1Name=dyn_hairstrand_01_r, bone2Name=dyn_hairstrand_01_r_MSTR, swing1LimitDegrees=15, swing2LimitDegrees=15, twistLimitDegrees=5}
 {
  bone1Name=dyn_hairstrand_01_l
  bone2Name=dyn_hairstrand_01_l_MSTR
  swing1LimitDegrees=20
  swing2LimitDegrees=20
  twistLimitDegrees=5
  twistMotion=Locked
 }
 {bone1Name=dyn_hairstrand_02_l, bone2Name=dyn_hairstrand_01_l, swing1LimitDegrees=45, swing2LimitDegrees=20, twistLimitDegrees=5}
 {bone1Name=dyn_hairstrand_04_l, bone2Name=dyn_hairstrand_01_l, swing1LimitDegrees=45, swing2LimitDegrees=20, twistLimitDegrees=5}
 {
  bone1Name=dyn_hairlong_bk_01_r
  bone2Name=dyn_hairlong_bk_01_r_MSTR
  swing1LimitDegrees=20
  swing2LimitDegrees=20
  twistLimitDegrees=5
  twistMotion=Locked
 }
 {bone1Name=dyn_hairlong_bk_02_r, bone2Name=dyn_hairlong_bk_01_r, swing1LimitDegrees=45, swing2LimitDegrees=20, twistLimitDegrees=5}
 {bone1Name=dyn_hairlong_bk_01, bone2Name=dyn_hairlong_bk_01_MSTR, swing1LimitDegrees=20, swing2LimitDegrees=20, twistLimitDegrees=2}
 {bone1Name=dyn_hairlong_bk_02, bone2Name=dyn_hairlong_bk_01, swing1LimitDegrees=45, swing2LimitDegrees=20, twistLimitDegrees=5}
 {bone1Name=dyn_hairlong_bk_01_l, bone2Name=dyn_hairlong_bk_01_l_MSTR, swing1LimitDegrees=20, swing2LimitDegrees=20, twistLimitDegrees=2}
 {bone1Name=dyn_hairlong_bk_02_l, bone2Name=dyn_hairlong_bk_01_l, swing1LimitDegrees=45, swing2LimitDegrees=20, twistLimitDegrees=5}
 {
  bone1Name=dyn_hairstrand_fr_01_r_MSTR
  bone2Name=head
  swing1LimitDegrees=45
  swing1Motion=Locked
  swing2LimitDegrees=45
  swing2Motion=Locked
  twistLimitDegrees=45
  twistMotion=Locked
 }
 {
  bone1Name=dyn_hairstrand_01_r_MSTR
  bone2Name=head
  swing1LimitDegrees=45
  swing1Motion=Locked
  swing2LimitDegrees=45
  swing2Motion=Locked
  twistLimitDegrees=45
  twistMotion=Locked
 }
 {
  bone1Name=dyn_hairstrand_01_l_MSTR
  bone2Name=head
  swing1LimitDegrees=45
  swing1Motion=Locked
  swing2LimitDegrees=45
  swing2Motion=Locked
  twistLimitDegrees=45
  twistMotion=Locked
 }
 {
  bone1Name=dyn_hairlong_bk_01_r_MSTR
  bone2Name=head
  swing1LimitDegrees=45
  swing1Motion=Locked
  swing2LimitDegrees=45
  swing2Motion=Locked
  twistLimitDegrees=45
  twistMotion=Locked
 }
 {
  bone1Name=dyn_hairlong_bk_01_MSTR
  bone2Name=head
  swing1LimitDegrees=45
  swing1Motion=Locked
  swing2LimitDegrees=45
  swing2Motion=Locked
  twistLimitDegrees=45
  twistMotion=Locked
 }
 {
  bone1Name=dyn_hairlong_bk_01_l_MSTR
  bone2Name=head
  swing1LimitDegrees=45
  swing1Motion=Locked
  swing2LimitDegrees=45
  swing2Motion=Locked
  twistLimitDegrees=45
  twistMotion=Locked
 }
 {
  bone1Name=dyn_hairstrand_bk_01_l_MSTR
  bone2Name=head
  swing1LimitDegrees=45
  swing1Motion=Locked
  swing2LimitDegrees=45
  swing2Motion=Locked
  twistLimitDegrees=45
  twistMotion=Locked
 }
 {
  bone1Name=dyn_hairstrand_fr_01_l_MSTR
  bone2Name=head
  swing1LimitDegrees=45
  swing1Motion=Locked
  swing2LimitDegrees=45
  swing2Motion=Locked
  twistLimitDegrees=45
  twistMotion=Locked
 }
 {
  bone1Name=dyn_hairlong_bk_02
  bone2Name=dyn_hairlong_bk_02_l
  swing1LimitDegrees=45
  swing1Motion=Locked
  swing2LimitDegrees=45
  swing2Motion=Locked
  twistLimitDegrees=45
  twistMotion=Locked
 }
 {
  bone1Name=dyn_hairlong_bk_02
  bone2Name=dyn_hairlong_bk_02_r
  swing1LimitDegrees=45
  swing1Motion=Locked
  swing2LimitDegrees=45
  swing2Motion=Locked
  twistLimitDegrees=45
  twistMotion=Locked
 }
 {
  bone1Name=head
  bone2Name=neck_01
  swing1LimitDegrees=45
  swing1Motion=Free
  swing2LimitDegrees=45
  swing2Motion=Free
  twistLimitDegrees=45
  twistMotion=Free
 }
]
