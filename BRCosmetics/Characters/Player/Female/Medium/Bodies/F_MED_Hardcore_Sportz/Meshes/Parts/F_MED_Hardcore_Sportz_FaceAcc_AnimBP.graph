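# Anim Blueprint for the Hardcore Sportz female face accessory, adding secondary hair motion to the pose copied from the parent mesh.
# A RigidBody sim relative to spine_05 feeds rotation constraints that average long-hair bones onto two left back strands.
# Half-weight spring AnimDynamics then adds jiggle to both strands.
# The EventGraph computes headAlpha, which drops as head twist from spine_05 increases while targeting, and is 1 otherwise.
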AnimBlueprint /BRCosmetics/Characters/Player/Female/Medium/Bodies/F_MED_Hardcore_Sportz/Meshes/Parts/F_MED_Hardcore_Sportz_FaceAcc_AnimBP
target_skeleton=&/F_MED_Hardcore_Sportz_FaceAcc_Skeleton
graphs [
 AnimGraph {nodes [
   n1 = RigidBody!(
    ComponentPose=Local To Component(LocalPose=Copy Pose From Mesh() {bUseAttachedParent=true})
    Alpha
    OverrideWorldGravity:Vector="0,0,-980"
    ExternalForce
   ) -> (Pose) {
    bClampLinearTranslationLimitToRefPose=true
    bOverrideWorldGravity=true
    overridePhysicsAsset=&/F_MED_Hardcore_Sportz_FaceAcc_PhysicsAsset
    simulationSpace=BaseBoneSpace
    alphaBoolBlend {blendOption=Linear}
    alphaScaleBiasClamp {bInterpResult=true}
    baseBoneRef {boneName=spine_05}
    componentAppliedLinearAccClamp=<3000 3000 3000>
    componentLinearAccScale=<0.4 0.4 0.3>
    componentLinearVelScale=<0.4 0.4 0.3>
    overrideWorldGravity {z=-980}
   }
   n2 = "Constraint - dyn_hairstrand_bk_01_l (RR)"!(ComponentPose=@n1, Alpha=1, ConstraintWeights_0=0.5, ConstraintWeights_1:float=0.5) -> (Pose) {
    alphaBoolBlend {blendOption=Linear}
    boneToModify {boneName=dyn_hairstrand_bk_01_l}
    constraintSetup [{transformType=Rotation, targetBone {boneName=dyn_hairlong_bk_01_l}}, {transformType=Rotation, targetBone {boneName=dyn_hairlong_bk_01}}]
    constraintWeights [1,1]
   }
   n3 = "Constraint - dyn_hairstrand_bk_02_l (RR)"!(ComponentPose=@n2, Alpha=1, ConstraintWeights_0=0.5, ConstraintWeights_1:float=0.5) -> (Pose) {
    alphaBoolBlend {blendOption=Linear}
    boneToModify {boneName=dyn_hairstrand_bk_02_l}
    constraintSetup [{transformType=Rotation, targetBone {boneName=dyn_hairlong_bk_02_l}}, {transformType=Rotation, targetBone {boneName=dyn_hairlong_bk_02}}]
    constraintWeights [1,1]
   }
   n4 = Anim Dynamics - Bone: dyn_hairstrand_bk_01_l(ComponentPose=@n3, Alpha=0.5, ExternalForce="0,0,0") {
    angularDampingOverride=1
    angularSpringConstant=2000
    bAngularSpring=true
    bGravityOverrideInSimSpace=true
    bOverrideAngularDamping=true
    bUseGravityOverride=true
    simulationSpace=BoneRelative
    alphaBoolBlend {blendOption=Linear}
    boundBone {boneName=dyn_hairstrand_bk_01_l}
    componentAppliedLinearAccClamp=<3000 3000 3000>
    componentLinearAccScale=<0.4 0.3 0.3>
    componentLinearVelScale=<0.4 0.3 0.3>
    gravityOverride {z=-980}
    physicsBodyDefinitions [{
      sphereCollisionRadius=0
      boundBone {boneName=dyn_hairstrand_bk_01_l}
      boxExtents=<5 5 5>
      constraintSetup {angularLimitsMax=<15 15 0>, angularLimitsMin=<-15 -15 -30>, angularTarget=<0 0 0>}
      localJointOffset=<1 0 0>
     }]
    relativeSpaceBone {boneName=root}
   }
   n5 = Anim Dynamics - Bone: dyn_hairstrand_bk_02_l(ComponentPose=@n4, Alpha=0.5, ExternalForce="0,0,0") {
    angularDampingOverride=1
    angularSpringConstant=2000
    bAngularSpring=true
    bGravityOverrideInSimSpace=true
    bOverrideAngularDamping=true
    bUseGravityOverride=true
    simulationSpace=BoneRelative
    alphaBoolBlend {blendOption=Linear}
    boundBone {boneName=dyn_hairstrand_bk_02_l}
    componentAppliedLinearAccClamp=<3000 3000 3000>
    componentLinearAccScale=<0.4 0.3 0.3>
    componentLinearVelScale=<0.4 0.3 0.3>
    gravityOverride {z=-980}
    physicsBodyDefinitions [{
      sphereCollisionRadius=0
      boundBone {boneName=dyn_hairstrand_bk_02_l}
      boxExtents=<5 5 5>
      constraintSetup {angularLimitsMax=<15 15 0>, angularLimitsMin=<-15 -15 -30>, angularTarget=<0 0 0>}
      localJointOffset=<1 0 0>
     }]
    relativeSpaceBone {boneName=root}
   }
   Output Pose(Result=Component To Local(ComponentPose=@n5)) {layerGroup=None}
   Get headAlpha() -> (headAlpha:double)
  ]}
 EventGraph {nodes [
   Event BlueprintUpdateAnimation()
   > Sequence()
   n1 = GetPartSkeletalMeshComponent()
   > Set headAlpha(headAlpha:double=SelectFloat(
     A=GetAlphaFromReferenceTransformDelta(
      SkeletalMeshComponent=@n1
      BoneName=head
      BaseName=spine_05
      DeltaComponentType=QuaternionTwist
      InputRange=(X=0,Y=22)
      OutputRange=(X=1,Y=0.1)
      TwistAxis="0,0,1"
     )
     B=1
     bPickA=Get bIsTargeting()
    )) -> (Output_Get:double)
   TryGetPawnOwner()
  ]}
]
variables [headAlpha {type=number}]
